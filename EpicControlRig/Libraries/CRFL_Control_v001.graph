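# Control Rig function library of reusable helpers for building and configuring animation controls.
# Space-null functions create one null per space, register it on the driven control, then drive it by its initial offset from that space, optionally rotation-only.
# Other helpers spawn gimbal child controls with visibility channels, scale shapes, and write mirror metadata.
# Shape names and colors resolve from connectors, metadata or overrides, with missing shapes falling back to Circle_Thick.
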ControlRigBlueprint /EpicControlRig/Libraries/CRFL_Control_v001
graphs [
 Concat_1_1_1_ContainedGraph {nodes [
   n1 = RigVMFunctionEntryNode()
   RigVMFunctionReturnNode(Result=/Script/RigVM.RigVMFunction_NameConcat(A=&RigVMFunction_NameConcat(A=@n1.A, B=@n1.B) {template="Concat::Execute(in A,in B,out Result)"}.Result, B=@n1.C) {is=struct RigVMUnitNode, template="Concat::Execute(in A,in B,out Result)"}.Result)
  ]}
 Concat_1_ContainedGraph {nodes [
   n1 = RigVMFunctionEntryNode()
   Concat = &RigVMFunction_NameConcat(A=@n1.A, B=@n1.B) {template="Concat::Execute(in A,in B,out Result)"}
   RigVMFunctionReturnNode(Result=&RigVMFunction_NameConcat(A=@Concat.Result, B=@n1.C) {template="Concat::Execute(in A,in B,out Result)"}.Result)
  ]}
 ContainedGraph {nodes [
   n1 = RigVMFunctionEntryNode()
   > DISPATCH_RigDispatch_SetMetadata = RigVMDispatchNode(Name=Mirror Axis, NameSpace=None, Item=@n1.Item, Value=@n1.Value) {template="DISPATCH_RigDispatch_SetMetadata(in Item,in Name,in NameSpace,in Value,out Success)"}
   > RigVMFunctionReturnNode()
  ]}
 ContainedGraph {nodes [
   n1 = RigVMFunctionEntryNode()
   > RigVMDispatchNode(Name=Mirror Behavioral, NameSpace=None, Item=@n1.Item, Value=@n1.Value) {template="DISPATCH_RigDispatch_SetMetadata(in Item,in Name,in NameSpace,in Value,out Success)"}
   > RigVMFunctionReturnNode()
  ]}
 ContainedGraph {nodes [
   n1 = RigVMFunctionEntryNode()
   > Set Mirror Axis_1 = RigVMFunctionReferenceNode(Value=@n1.Axis, Item=@n1.Item)
   > Set Mirror Behavior_1 = RigVMFunctionReferenceNode(Item=@n1.Item, Value=@n1.Behavioral)
   > RigVMFunctionReturnNode()
  ]}
 Multiply_1_2_ContainedGraph {nodes [
   n1 = RigVMFunctionEntryNode()
   RigVMFunctionReturnNode(Result=&RigVMFunction_MathDoubleMul(B=@n1.C, A=&RigVMFunction_MathDoubleMul(A=@n1.A, B=@n1.B).Result).Result)
  ]}
 New Function_ContainedGraph {nodes [
   n1 = RigVMFunctionEntryNode()
   Concat_1_1 = RigVMAggregateNode(
    B=.
    C=/Script/ControlRig.RigUnit_ResolveConnector(SkipSocket=False, Connector=RigVMVariableNode() {variable=Item}.Value) {is=struct RigVMUnitNode}.Result.Name
    A=@n1.ShapeLib Namespace
   )
   n2 = RigVMDispatchNode(
    True=@Concat_1_1.Result
    Condition=RigVMDispatchNode(B=Default, A=@n1.Default Shape) {template="DISPATCH_RigVMDispatch_CoreEquals(in A,in B,out Result)"}.Result
    False=@n1.Default Shape
   ) {template="DISPATCH_RigVMDispatch_If(in Condition,in True,in False,out Result)"}
   > RigVMFunctionReturnNode(Result=RigVMDispatchNode(False=Circle_Thick, True=@n2.Result, Condition=/Script/ControlRig.RigUnit_ShapeExists(ShapeName=@n2.Result) {is=struct RigVMUnitNode}.Result) {template="DISPATCH_RigVMDispatch_If(in Condition,in True,in False,out Result)"}.Result)
  ]}
 New Function_ContainedGraph {nodes [
   n1 = RigVMFunctionEntryNode()
   Get Control Color From Metadata = RigVMFunctionReferenceNode(Color Override=@n1.Module Color)
   RigVMFunction_MathBoolAnd = &RigVMFunction_MathBoolAnd(
    A=RigVMDispatchNode(B=@n1.Color Override.G, A=@n1.Color Override.R) {template="DISPATCH_RigVMDispatch_CoreEquals(in A,in B,out Result)"}.Result
    B=RigVMDispatchNode(B=@n1.Color Override.B, A=@n1.Color Override.G) {template="DISPATCH_RigVMDispatch_CoreEquals(in A,in B,out Result)"}.Result
   )
   > RigVMFunctionReturnNode(Result=RigVMDispatchNode(False=@n1.Color Override, True=@"Get Control Color From Metadata".Result, Condition=@RigVMFunction_MathBoolAnd.Result) {template="DISPATCH_RigVMDispatch_If(in Condition,in True,in False,out Result)"}.Result)
  ]}
 New Function_ContainedGraph {nodes [
   n1 = RigVMFunctionEntryNode()
   > For_Each = RigVMDispatchNode(Array=@n1.Controls) {template="DISPATCH_RigVMDispatch_ArrayIterator(in Array,out Element,out Index,out Count,out Ratio)"}
   n2 = /Script/ControlRig.RigUnit_GetShapeTransform(Control=@For_Each.Element.Name) {is=struct RigVMUnitNode}
   Multiply_1 = RigVMAggregateNode(B=@n1.Scale, A=@n1.Global Scale, C=@n1.Auto Scale Factor)
   RigVMFunction_MathVectorFromDouble = /Script/RigVM.RigVMFunction_MathVectorFromDouble(Value=@Multiply_1.Result) {is=struct RigVMUnitNode, template="Cast::Execute(in Value,out Result)"}
   Multiply = /Script/RigVM.RigVMFunction_MathVectorMul(A=@n2.Transform.Scale3D, B=@RigVMFunction_MathVectorFromDouble.Result) {is=struct RigVMUnitNode, template="Multiply::Execute(in A,in B,out Result)"}
   > /Script/ControlRig.RigUnit_SetShapeTransform(
    Transform=(Rotation=(X=0,Y=0,Z=0,W=1),Translation=(X=0,Y=0,Z=0),Scale3D=(X=1,Y=1,Z=1))
    Transform.Rotation=@n2.Transform.Rotation
    Transform.Scale3D=@Multiply.Result
    Transform.Translation=&RigVMFunction_MathVectorMul(A=@n2.Transform.Translation, B=@RigVMFunction_MathVectorFromDouble.Result) {template="Multiply::Execute(in A,in B,out Result)"}.Result
    Control=@For_Each.Element.Name
   ) {is=struct RigVMUnitNode}
   RigVMFunctionReturnNode(ExecuteContext=@For_Each.Completed)
  ]}
 New Function_ContainedGraph {nodes [
   n1 = RigVMFunctionEntryNode()
   n2 = RigVMVariableNode() {variable=Source Control}
   Concat = &RigVMFunction_NameConcat(B=" Gimbal", A=@n2.Value.Name) {template="Concat::Execute(in A,in B,out Result)"}
   n3 = /Script/ControlRig.RigUnit_HierarchyGetShapeSettings(Item=@n1.Source Control) {is=struct RigVMUnitNode}
   n4 = RigVMVariableNode() {variable=Shape Settings Override}
   Not_Equals = RigVMDispatchNode(B=None, A=@n4.Value.Name) {template="DISPATCH_RigVMDispatch_CoreNotEquals(in A,in B,out Result)"}
   n5 = RigVMDispatchNode(
    False=(bVisible=True,Name="Default",Color=(R=1,G=0,B=0,A=1),Transform=(Rotation=(X=0,Y=0,Z=0,W=1),Translation=(X=0,Y=0,Z=0),Scale3D=(X=1,Y=1,Z=1)))
    False.Color=/Script/RigVM.RigVMFunction_MathColorAdd(B=(R=0.2,G=0.2,B=0.2,A=1), A=@n3.Settings.Color) {is=struct RigVMUnitNode, template="Add::Execute(in A,in B,out Result)"}.Result
    False.Name=@n3.Settings.Name
    False.Transform.Rotation=@n3.Settings.Transform.Rotation
    False.Transform.Translation=@n3.Settings.Transform.Translation
    False.bVisible=@n3.Settings.bVisible
    False.Transform.Scale3D=&RigVMFunction_MathVectorMul(B=(X=0.8,Y=0.8,Z=0.8), A=@n3.Settings.Transform.Scale3D) {template="Multiply::Execute(in A,in B,out Result)"}.Result
    Condition=@Not_Equals.Result
    True=@n4.Value
   ) {template="DISPATCH_RigVMDispatch_If(in Condition,in True,in False,out Result)"}
   > n6 = /Script/ControlRig.RigUnit_HierarchyAddControlRotator(
    InitialValue=(Pitch=0,Yaw=0,Roll=0)
    OffsetSpace=LocalSpace
    OffsetTransform=(Rotation=(X=0,Y=0,Z=0,W=1),Translation=(X=0,Y=0,Z=0),Scale3D=(X=1,Y=1,Z=1))
    Settings="(InitialSpace=LocalSpace,Limits=(LimitPitch=(bMinimum=False,bMaximum=False),LimitYaw=(bMinimum=False,bMaximum=False),LimitRoll=(bMinimum=False,bMaximum=False),MinValue=(Pitch=-180,Yaw=-180,Roll=-180),MaxValue=(Pitch=180,Yaw=180,Roll=180),bDrawLimits=True), ...(truncated)"
    Name=@Concat.Result
    Settings.Shape=@n5.Result
    Parent=@n2.Value
    Settings.PreferredRotationOrder=RigVMVariableNode() {variable=Rotation Order}.Value
    Settings.DisplayName=RigVMDispatchNode(
     False=&RigVMFunction_NameConcat(B=" Gimbal", A=@n1.Display Name) {template="Concat::Execute(in A,in B,out Result)"}.Result
     True=@n1.Display Name
     Condition=RigVMDispatchNode(B=None, A=@n1.Display Name) {template="DISPATCH_RigVMDispatch_CoreEquals(in A,in B,out Result)"}.Result
    ) {template="DISPATCH_RigVMDispatch_If(in Condition,in True,in False,out Result)"}.Result
   ) {
    is=struct RigVMUnitNode
    template="SpawnControl::Execute(in InitialValue,in Settings,in OffsetTransform,in OffsetSpace,in Parent,in Name,out Item)"
   }
   > Set_Metadata = RigVMDispatchNode(Name=Gimbal Control, NameSpace=Self, Value=@n6.Item, Item=RigVMVariableNode() {variable=Source Control}.Value) {template="DISPATCH_RigDispatch_SetMetadata(in Item,in Name,in NameSpace,in Value,out Success)"}
   > SpawnAnimationChannel = /Script/ControlRig.RigUnit_HierarchyAddAnimationChannelBool(InitialValue=False, LimitsEnabled=(), MaximumValue=True, MinimumValue=False, Name=Gimbal Control Vis, Parent=@n6.Item) {
    is=struct RigVMUnitNode
    template="SpawnAnimationChannel::Execute(in InitialValue,in MinimumValue,in MaximumValue,in LimitsEnabled,in Parent,in Name,out Item)"
   }
   Make_Array = RigVMDispatchNode(Values=((Type=None,Name="None")), Values.0=RigVMVariableNode() {variable=Source Control}.Value) {template="DISPATCH_RigVMDispatch_ArrayMake(in Values,out Array)"}
   > /Script/ControlRig.RigUnit_SetChannelHosts(Hosts=@Make_Array.Array, Channel=@SpawnAnimationChannel.Item) {is=struct RigVMUnitNode}
   > RigVMFunctionReturnNode(Visibility Channel=@SpawnAnimationChannel.Item, Gimbal Control=@n6.Item)
  ]}
 New Function_ContainedGraph {nodes [
   n1 = RigVMFunctionEntryNode()
   RigVMFunction_Sequence = &RigVMFunction_Sequence(ExecuteContext=@n1.ExecuteContext)
   For_Each_4 = RigVMDispatchNode(ExecuteContext=@RigVMFunction_Sequence.A, Array=RigVMVariableNode() {variable=Spaces}.Value) {template="DISPATCH_RigVMDispatch_ArrayIterator(in Array,out Element,out Index,out Count,out Ratio)"}
   n2 = RigVMVariableNode() {variable=Space Nulls}
   Num = RigVMDispatchNode(Array=@n2.Value) {template="DISPATCH_RigVMDispatch_ArrayGetNum(in Array,out Num)"}
   Less = /Script/RigVM.RigVMFunction_MathIntLess(A=@For_Each_4.Index, B=@Num.Num) {is=struct RigVMUnitNode, template="Less::Execute(in A,in B,out Result)"}
   n3 = &RigVMFunction_ControlFlowBranch(ExecuteContext=@For_Each_4.ExecuteContext, Condition=@Less.Result)
   RigVMDispatchNode(
    Enabled=True
    Prefix="Error: "
    ScreenColor=(R=1,G=1,B=1,A=1)
    ScreenDuration=0.05
    Value=Space array is larger than Space Nulls array
    ExecuteContext=@n3.False
   ) {template="DISPATCH_RigVMDispatch_Print(in Prefix,in Value,in Enabled,in ScreenDuration,in ScreenColor)"}
   RigVMFunction_ControlFlowBranch = &RigVMFunction_ControlFlowBranch(ExecuteContext=@n3.True, Condition=RigVMVariableNode() {variable=Orient Only}.Value)
   n4 = RigVMDispatchNode(Index=@For_Each_4.Index, Array=@n2.Value) {template="DISPATCH_RigVMDispatch_ArrayGetAtIndex(in Array,in Index,out Element)"}
   Multiply = &RigVMFunction_MathTransformMul(
    B=&RigUnit_GetTransform(bInitial=False, Item=@For_Each_4.Element).Transform
    A=&RigVMFunction_MathTransformMakeRelative(
     Parent=&RigUnit_GetTransform(bInitial=true, Item=@For_Each_4.Element).Transform
     Global=&RigUnit_GetTransform(bInitial=true, Item=@n4.Element).Transform
    ).Local
   ) {template="Multiply::Execute(in A,in B,out Result)"}
   &RigUnit_SetTransform(Weight=1, bInitial=False, Value=@Multiply.Result, ExecutePin=@RigVMFunction_ControlFlowBranch.False, Item=@n4.Element)
   Set Transform_4 = &RigUnit_SetRotation(
    Space=GlobalSpace
    Weight=1
    bInitial=False
    Value=@Multiply.Result.Rotation
    ExecutePin=@RigVMFunction_ControlFlowBranch.True
    Item=@n4.Element
   )
   RigVMFunctionReturnNode(ExecuteContext=@RigVMFunction_Sequence.B)
  ]}
 New Function_ContainedGraph {nodes [
   n1 = RigVMFunctionEntryNode()
   > For_Each_5 = RigVMDispatchNode(Array=@n1.Spaces) {template="DISPATCH_RigVMDispatch_ArrayIterator(in Array,out Element,out Index,out Count,out Ratio)"}
   n2 = /Script/ControlRig.RigUnit_GetItemShortName(Item=@For_Each_5.Element) {is=struct RigVMUnitNode}
   n3 = RigVMDispatchNode(
    True=&RigVMFunction_NameConcat(B=" Orient", A=@n2.ShortName) {template="Concat::Execute(in A,in B,out Result)"}.Result
    Condition=RigVMVariableNode() {variable=Orient Only}.Value
    False=@n2.ShortName
   ) {template="DISPATCH_RigVMDispatch_If(in Condition,in True,in False,out Result)"}
   n4 = RigVMVariableNode() {variable=Driven Control}
   Concat = RigVMAggregateNode(B=" ", C=@n3.Result, A=@n4.Value.Name)
   > n5 = &RigUnit_HierarchyAddNull(
    Space=GlobalSpace
    Name=&RigVMFunction_NameConcat(B=" Space", A=@Concat.Result) {template="Concat::Execute(in A,in B,out Result)"}.Result
    Transform=&RigUnit_GetTransform(bInitial=False, Item=/Script/ControlRig.RigUnit_HierarchyGetParent(bDefaultParent=True, Child=@n4.Value) {is=struct RigVMUnitNode}.Parent).Transform
    Parent=RigVMVariableNode() {variable=Space Null Parent}.Value
   )
   > /Script/ControlRig.RigUnit_AddAvailableSpaces(
    Spaces=((Key=(Type=Control,Name="None"),Label="None"))
    Control=RigVMVariableNode() {variable=Driven Control}.Value
    Spaces.0.Label=@n3.Result
    Spaces.0.Key=@n5.Item
   ) {is=struct RigVMUnitNode}
   > Add = RigVMDispatchNode(Array=RigVMVariableNode() {variable=Space Nulls}.Value, Element=@n5.Item) {template="DISPATCH_RigVMDispatch_ArrayAdd(io Array,in Element,out Index)"}
   RigVMFunctionReturnNode(ExecuteContext=@For_Each_5.Completed, Output Space Nulls=RigVMVariableNode() {variable=Space Nulls}.Value)
  ]}
 New Function_ContainedGraph {nodes [
   n1 = RigVMFunctionEntryNode()
   Get Control Color From Metadata v02 = RigVMFunctionReferenceNode(Bone=@n1.Bone, Color Override=@n1.Module Color)
   RigVMFunction_MathBoolAnd = &RigVMFunction_MathBoolAnd(
    A=RigVMDispatchNode(B=@n1.Color Override.G, A=@n1.Color Override.R) {template="DISPATCH_RigVMDispatch_CoreEquals(in A,in B,out Result)"}.Result
    B=RigVMDispatchNode(B=@n1.Color Override.B, A=@n1.Color Override.G) {template="DISPATCH_RigVMDispatch_CoreEquals(in A,in B,out Result)"}.Result
   )
   > RigVMFunctionReturnNode(Result=RigVMDispatchNode(False=@n1.Color Override, True=@"Get Control Color From Metadata v02".Result, Condition=@RigVMFunction_MathBoolAnd.Result) {template="DISPATCH_RigVMDispatch_If(in Condition,in True,in False,out Result)"}.Result)
  ]}
 RigVMFunctionLibrary {nodes [
   Add Mirror Tag = RigVMCollapseNode(Axis=(X=0,Y=1,Z=1), Behavioral=false, ExecuteContext=(), Item=(Type=Bone))
   Construct Space Nulls v01 = RigVMCollapseNode(Driven Control=(Type=Control,Name="None"), ExecuteContext=(), Space Null Parent=(Type=Bone,Name="None"), Spaces=())
   Control Color Override From Metadata v01 = RigVMCollapseNode(Color Override=(R=0,G=0,B=0,A=0), Module Color=(R=0,G=0,B=0,A=0))
   Control Color Override From Metadata v02 = RigVMCollapseNode(Bone=(Type=None,Name="None"), Color Override=(R=0,G=0,B=0,A=0), Module Color=(R=0,G=0,B=0,A=0))
   Create Gimbal Control = RigVMCollapseNode(
    ExecuteContext=()
    Rotation Order=YZX
    Shape Settings Override=(bVisible=True,Name="None",Color=(R=1,G=0,B=0,A=1),Transform=(Rotation=(X=0,Y=0,Z=0,W=1),Translation=(X=0,Y=0,Z=0),Scale3D=(X=1,Y=1,Z=1)))
    Source Control=(Type=Bone,Name="None")
   )
   Evaluate Space Nulls v01 = RigVMCollapseNode(ExecuteContext=(), Orient Only=false, Space Nulls=(), Spaces=())
   Get Control Shape Name From Item v01 = RigVMCollapseNode(Default Shape=Sphere_Solid, Item=(Type=None,Name="None"), ShapeLib Namespace=CRSL)
   Scale Control Shape v01 = RigVMCollapseNode(Auto Scale Factor=1, Controls=(), ExecuteContext=(), Global Scale=1, Scale=1)
   Set Mirror Axis = RigVMCollapseNode(ExecuteContext=(), Item=(Type=Bone,Name="None"), Value=(X=0,Y=1,Z=1))
   Set Mirror Behavior = RigVMCollapseNode(ExecuteContext=(), Item=(Type=Bone,Name="None"), Value=false)
  ]}
 RigVMModel {nodes [
   RigVMDispatchNode(Enabled=True, ExecuteContext=(), Prefix=a:, ScreenColor=(R=1,G=1,B=1,A=1), ScreenDuration=0.05, Value=0) {template="DISPATCH_RigVMDispatch_Print(in Prefix,in Value,in Enabled,in ScreenDuration,in ScreenColor)"}
   RigVMDispatchNode(Enabled=True, ExecuteContext=(), Prefix=b, ScreenColor=(R=1,G=1,B=1,A=1), ScreenDuration=0.05, Value=0) {template="DISPATCH_RigVMDispatch_Print(in Prefix,in Value,in Enabled,in ScreenDuration,in ScreenColor)"}
   RigUnit_BeginExecution = &RigUnit_BeginExecution()
   &RigUnit_ItemArray(Items=((Type=Bone,Name="clavicle_l"),(Type=Bone,Name="upperarm_l"),(Type=Bone,Name="lowerarm_l")))
  ]}
 RigVMModel Rig {nodes [&RigUnit_PrepareForExecution()]}
]
hierarchy [
 root {loc=<0 0 0>, rot=<0 0 0 1>}
 pelvis {parent=root, loc=<0 2.2809 95.8968>, rot=<-0.0224 0.7068 0.0224 -0.7068>}
 spine_01 {parent=pelvis, loc=<3.6771 0 0>, rot=<0 0 -0.1258 -0.9921>}
 spine_02 {parent=spine_01, loc=<6.7951 0 0>, rot=<0 0 0.0302 -0.9995>}
 spine_03 {parent=spine_02, loc=<7.2382 0 0>, rot=<0 0 0.0954 -0.9954>}
 spine_04 {parent=spine_03, loc=<8.5239 0 0>, rot=<0 0 0.0512 -0.9987>}
 spine_05 {parent=spine_04, loc=<19.4398 0 0>, rot=<0 0 0.0059 -1>}
 neck_01 {parent=spine_05, loc=<11.8878 0 0>, rot=<0 0 -0.2073 -0.9783>}
 neck_02 {parent=neck_01, loc=<5.1103 0 0>, rot=<0 0 0.0167 -0.9999>}
 head {parent=neck_02, loc=<4.913 0 0>, rot=<0 0 0.1035 -0.9946>}
 clavicle_l {parent=spine_05, loc=<5.5163 -1.3148 -1.4279>, rot=<0.0833 -0.7546 -0.0413 -0.6496>}
 upperarm_l {parent=clavicle_l, loc=<17.8095 0 0>, rot=<-0.0199 -0.3917 -0.0202 -0.9196>}
 lowerarm_l {parent=upperarm_l, loc=<27.7711 0 0>, rot=<0 0 -0.3336 -0.9427>}
 lowerarm_twist_02_l {parent=lowerarm_l, loc=<9.0837 0 0>, rot=<0.0115 0.0149 0.0175 -0.9997>}
 lowerarm_twist_01_l {parent=lowerarm_l, loc=<18.1674 0 0>, rot=<0.0115 0.0149 0.0175 -0.9997>}
 hand_l {parent=lowerarm_l, loc=<27.2511 0 0>, rot=<-0.5576 -0.0017 0.0206 -0.8299>}
 middle_metacarpal_l {parent=hand_l, loc=<3.3758 0.7536 -0.1829>, rot=<-0.0373 0.0004 -0.0203 -0.9991>}
 middle_01_l {parent=middle_metacarpal_l, loc=<6.0982 0 0>, rot=<0 0 0.2721 -0.9623>}
 middle_02_l {parent=middle_01_l, loc=<5.169 0 0>, rot=<0 0 0.1803 -0.9836>}
 middle_03_l {parent=middle_02_l, loc=<2.474 0 0>, rot=<0 0 0.0872 -0.9962>}
 pinky_metacarpal_l {parent=hand_l, loc=<3.3144 0.3059 2.3911>, rot=<-0.2182 0.1882 0.0583 -0.9558>}
 pinky_01_l {parent=pinky_metacarpal_l, loc=<4.9576 0.1431 -0.1988>, rot=<0.09 -0.017 0.1281 -0.9875>}
 pinky_02_l {parent=pinky_01_l, loc=<3.816 0 0>, rot=<0 0 0.1847 -0.9828>}
 pinky_03_l {parent=pinky_02_l, loc=<2.04 0 0>, rot=<0 0 0.0429 -0.9991>}
 ring_metacarpal_l {parent=hand_l, loc=<3.3743 0.5425 1.0918>, rot=<-0.1166 0.1006 -0.0257 -0.9877>}
 ring_01_l {parent=ring_metacarpal_l, loc=<5.6455 0.0416 -0.0207>, rot=<0.0542 -0.0132 0.2535 -0.9657>}
 ring_02_l {parent=ring_01_l, loc=<4.977 0 0>, rot=<0 0 0.1647 -0.9863>}
 ring_03_l {parent=ring_02_l, loc=<2.265 0 0>, rot=<0 0 0.0799 -0.9968>}
 thumb_01_l {parent=hand_l, loc=<1.9924 -1.3566 -2.5815>, rot=<0.5052 -0.3691 -0.067 -0.7772>}
 thumb_02_l {parent=thumb_01_l, loc=<4.378 0 0>, rot=<0.0336 0.0103 0.2019 -0.9788>}
 thumb_03_l {parent=thumb_02_l, loc=<3.086 0 0>, rot=<0 0 0.0872 -0.9962>}
 index_metacarpal_l {parent=hand_l, loc=<3.4445 0.3847 -2.3793>, rot=<0.029 -0.0637 -0.0071 -0.9975>}
 index_01_l {parent=index_metacarpal_l, loc=<5.8771 -0.0432 0.2409>, rot=<0 0 0.2026 -0.9793>}
 index_02_l {parent=index_01_l, loc=<4.08 0 0>, rot=<0 0 0.1296 -0.9916>}
 index_03_l {parent=index_02_l, loc=<2.595 0 0>, rot=<0 0 0.109 -0.994>}
 upperarm_twist_01_l {parent=upperarm_l, loc=<9.257 0 0>, rot=<-0.0001 0.0188 -0.0028 -0.9998>}
 upperarm_twist_02_l {parent=upperarm_l, loc=<18.5141 0 0>, rot=<0 0 0 1>}
 clavicle_r {parent=spine_05, loc=<5.5162 -1.3148 1.4279>, rot=<0.7546 0.0833 -0.6496 0.0413>}
 upperarm_r {parent=clavicle_r, loc=<-17.8096 0 0.0004>, rot=<-0.0199 -0.3917 -0.0202 -0.9196>}
 lowerarm_r {parent=upperarm_r, loc=<-27.7707 0 0>, rot=<0 0 -0.3336 -0.9427>}
 lowerarm_twist_02_r {parent=lowerarm_r, loc=<-9.0837 0 0>, rot=<0.0115 0.0149 0.0175 -0.9997>}
 lowerarm_twist_01_r {parent=lowerarm_r, loc=<-18.1673 0 0>, rot=<0.0115 0.0149 0.0175 -0.9997>}
 hand_r {parent=lowerarm_r, loc=<-27.251 0 0>, rot=<-0.5576 -0.0017 0.0206 -0.8299>}
 middle_metacarpal_r {parent=hand_r, loc=<-3.3758 -0.754 0.1828>, rot=<-0.0373 0.0004 -0.0203 -0.9991>}
 middle_01_r {parent=middle_metacarpal_r, loc=<-6.0984 0.0001 0>, rot=<0 0 0.2721 -0.9623>}
 middle_02_r {parent=middle_01_r, loc=<-5.169 0.0001 0>, rot=<0 0 0.1803 -0.9836>}
 middle_03_r {parent=middle_02_r, loc=<-2.474 0 0>, rot=<0 0 0.0872 -0.9962>}
 pinky_metacarpal_r {parent=hand_r, loc=<-3.3147 -0.3059 -2.3913>, rot=<-0.2182 0.1882 0.0583 -0.9558>}
 pinky_01_r {parent=pinky_metacarpal_r, loc=<-4.9573 -0.1433 0.1989>, rot=<0.09 -0.017 0.1281 -0.9875>}
 pinky_02_r {parent=pinky_01_r, loc=<-3.816 0 0>, rot=<0 0 0.1847 -0.9828>}
 pinky_03_r {parent=pinky_02_r, loc=<-2.04 0 0>, rot=<0 0 0.0429 -0.9991>}
 ring_metacarpal_r {parent=hand_r, loc=<-3.3742 -0.543 -1.0918>, rot=<-0.1166 0.1006 -0.0257 -0.9877>}
 ring_01_r {parent=ring_metacarpal_r, loc=<-5.6457 -0.0414 0.0207>, rot=<0.0542 -0.0132 0.2535 -0.9657>}
 ring_02_r {parent=ring_01_r, loc=<-4.9771 0 0>, rot=<0 0 0.1647 -0.9863>}
 ring_03_r {parent=ring_02_r, loc=<-2.265 -0.0001 0>, rot=<0 0 0.0799 -0.9968>}
 thumb_01_r {parent=hand_r, loc=<-1.9928 1.3567 2.5813>, rot=<0.5052 -0.3691 -0.067 -0.7772>}
 thumb_02_r {parent=thumb_01_r, loc=<-4.3778 -0.0004 -0.0001>, rot=<0.0336 0.0103 0.2019 -0.9788>}
 thumb_03_r {parent=thumb_02_r, loc=<-3.086 0 0>, rot=<0 0 0.0872 -0.9962>}
 index_metacarpal_r {parent=hand_r, loc=<-3.4445 -0.3852 2.3793>, rot=<0.029 -0.0637 -0.0071 -0.9975>}
 index_01_r {parent=index_metacarpal_r, loc=<-5.8772 0.0434 -0.241>, rot=<0 0 0.2026 -0.9793>}
 index_02_r {parent=index_01_r, loc=<-4.0799 0 0>, rot=<0 0 0.1296 -0.9916>}
 index_03_r {parent=index_02_r, loc=<-2.5951 0 0>, rot=<0 0 0.109 -0.994>}
 upperarm_twist_01_r {parent=upperarm_r, loc=<-9.2569 0 -0.0001>, rot=<-0.0001 0.0188 -0.0028 -0.9998>}
 upperarm_twist_02_r {parent=upperarm_r, loc=<-18.5138 0 -0.0002>, rot=<0 0 0 1>}
 thigh_r {parent=pelvis, loc=<-2.3657 0.1195 9.9691>, rot=<0.0295 -0.0724 0.9965 -0.029>}
 calf_r {parent=thigh_r, loc=<43.3413 0 0>, rot=<0 0 -0.0437 -0.999>}
 foot_r {parent=calf_r, loc=<42.2179 0 0>, rot=<-0.0007 -0.0269 0.0232 -0.9994>}
 ball_r {parent=foot_r, loc=<7.0094 -15.2376 -0.5389>, rot=<0 0 -0.7071 -0.7071>}
 calf_twist_02_r {parent=calf_r, loc=<14.0726 0 0.0377>, rot=<-0.0001 -0.0023 0.0232 -0.9997>}
 calf_twist_01_r {parent=calf_r, loc=<28.1453 0 0.0755>, rot=<-0.0001 -0.0023 0.0232 -0.9997>}
 thigh_twist_01_r {parent=thigh_r, loc=<14.4471 0 0>, rot=<0 -0.0023 -0.0114 -0.9999>}
 thigh_twist_02_r {parent=thigh_r, loc=<28.8942 0 0>, rot=<0 -0.0023 -0.0114 -0.9999>}
 thigh_l {parent=pelvis, loc=<-2.3657 0.11 -9.9692>, rot=<0.0724 0.0295 -0.029 -0.9965>}
 calf_l {parent=thigh_l, loc=<-43.3413 0 0>, rot=<0 0 -0.0437 -0.999>}
 foot_l {parent=calf_l, loc=<-42.2179 0 0>, rot=<-0.0007 -0.0269 0.0232 -0.9994>}
 ball_l {parent=foot_l, loc=<-7.0094 15.2376 0.5389>, rot=<0 0 -0.7071 -0.7071>}
 calf_twist_02_l {parent=calf_l, loc=<-14.0726 0 0>, rot=<-0.0001 -0.0023 0.0232 -0.9997>}
 calf_twist_01_l {parent=calf_l, loc=<-28.1453 0 -0.0754>, rot=<-0.0001 -0.0023 0.0232 -0.9997>}
 thigh_twist_01_l {parent=thigh_l, loc=<-14.4471 0 0>, rot=<0 -0.0023 -0.0114 -0.9999>}
 thigh_twist_02_l {parent=thigh_l, loc=<-28.8942 0 0>, rot=<0 -0.0023 -0.0114 -0.9999>}
 ik_hand_root {parent=root, loc=<0 0 0>, rot=<0 0 0 1>}
 ik_hand_gun {parent=ik_hand_root, loc=<-47.7683 15.6988 104.4664>, rot=<0.5985 0.0096 0.5206 -0.6089>}
 ik_hand_r {parent=ik_hand_gun, loc=<0 0 0>, rot=<0 0 0 1>}
 ik_hand_l {parent=ik_hand_gun, loc=<43.7336 61.6698 58.4081>, rot=<-0.4578 -0.6455 -0.6114 0>}
 ik_foot_root {parent=root, loc=<0 0 0>, rot=<0 0 0 1>}
 ik_foot_l {parent=ik_foot_root, loc=<14.0877 -0.9943 8.2376>, rot=<0.0636 0.7042 0.0414 -0.7059>}
 ik_foot_r {parent=ik_foot_root, loc=<-14.0876 -0.9848 8.2382>, rot=<0.7059 0.0414 -0.7042 0.0636>}
 interaction {parent=root, loc=<0 0 0>, rot=<0 0 0 1>}
 center_of_mass {parent=root, loc=<0 0 0>, rot=<0 0 0 1>}
 MoveData_Speed {type=Curve, loc=<0 0 0>, rot=<0 0 0 1>}
 base_pose {type=Curve, loc=<0 0 0>, rot=<0 0 0 1>}
 calf_l_back_50 {type=Curve, loc=<0 0 0>, rot=<0 0 0 1>}
 calf_l_back_120 {type=Curve, loc=<0 0 0>, rot=<0 0 0 1>}
 calf_l_back_90 {type=Curve, loc=<0 0 0>, rot=<0 0 0 1>}
 calf_l_back_150 {type=Curve, loc=<0 0 0>, rot=<0 0 0 1>}
 calf_r_back_50 {type=Curve, loc=<0 0 0>, rot=<0 0 0 1>}
 calf_r_back_120 {type=Curve, loc=<0 0 0>, rot=<0 0 0 1>}
 calf_r_back_90 {type=Curve, loc=<0 0 0>, rot=<0 0 0 1>}
 calf_r_back_150 {type=Curve, loc=<0 0 0>, rot=<0 0 0 1>}
 clavicle_l_fwd_30 {type=Curve, loc=<0 0 0>, rot=<0 0 0 1>}
 clavicle_l_down_20 {type=Curve, loc=<0 0 0>, rot=<0 0 0 1>}
 clavicle_l_back_30 {type=Curve, loc=<0 0 0>, rot=<0 0 0 1>}
 clavicle_l_up_40 {type=Curve, loc=<0 0 0>, rot=<0 0 0 1>}
 clavicle_r_fwd_30 {type=Curve, loc=<0 0 0>, rot=<0 0 0 1>}
 clavicle_r_down_20 {type=Curve, loc=<0 0 0>, rot=<0 0 0 1>}
 clavicle_r_back_30 {type=Curve, loc=<0 0 0>, rot=<0 0 0 1>}
 clavicle_r_up_40 {type=Curve, loc=<0 0 0>, rot=<0 0 0 1>}
 clavicle_r_back_310 {type=Curve, loc=<0 0 0>, rot=<0 0 0 1>}
 clavicle_r_up_401 {type=Curve, loc=<0 0 0>, rot=<0 0 0 1>}
 foot_l_down_60 {type=Curve, loc=<0 0 0>, rot=<0 0 0 1>}
 foot_l_up_35 {type=Curve, loc=<0 0 0>, rot=<0 0 0 1>}
 foot_r_up_35 {type=Curve, loc=<0 0 0>, rot=<0 0 0 1>}
 foot_r_down_60 {type=Curve, loc=<0 0 0>, rot=<0 0 0 1>}
 hand_l_down_90 {type=Curve, loc=<0 0 0>, rot=<0 0 0 1>}
 hand_l_up_20 {type=Curve, loc=<0 0 0>, rot=<0 0 0 1>}
 hand_l_up_90 {type=Curve, loc=<0 0 0>, rot=<0 0 0 1>}
 hand_r_up_90 {type=Curve, loc=<0 0 0>, rot=<0 0 0 1>}
 hand_r_down_90 {type=Curve, loc=<0 0 0>, rot=<0 0 0 1>}
 hand_r_up_20 {type=Curve, loc=<0 0 0>, rot=<0 0 0 1>}
 Pose_0 {type=Curve, loc=<0 0 0>, rot=<0 0 0 1>}
 Pose_1 {type=Curve, loc=<0 0 0>, rot=<0 0 0 1>}
 Pose_2 {type=Curve, loc=<0 0 0>, rot=<0 0 0 1>}
 Pose_3 {type=Curve, loc=<0 0 0>, rot=<0 0 0 1>}
 Pose_4 {type=Curve, loc=<0 0 0>, rot=<0 0 0 1>}
 Pose_5 {type=Curve, loc=<0 0 0>, rot=<0 0 0 1>}
 Pose_6 {type=Curve, loc=<0 0 0>, rot=<0 0 0 1>}
 Pose_7 {type=Curve, loc=<0 0 0>, rot=<0 0 0 1>}
 Pose_8 {type=Curve, loc=<0 0 0>, rot=<0 0 0 1>}
 lowerarm_l_in_110 {type=Curve, loc=<0 0 0>, rot=<0 0 0 1>}
 lowerarm_l_in_35 {type=Curve, loc=<0 0 0>, rot=<0 0 0 1>}
 lowerarm_l_out_35 {type=Curve, loc=<0 0 0>, rot=<0 0 0 1>}
 lowerarm_l_in_50 {type=Curve, loc=<0 0 0>, rot=<0 0 0 1>}
 lowerarm_l_in_75 {type=Curve, loc=<0 0 0>, rot=<0 0 0 1>}
 lowerarm_l_in_10 {type=Curve, loc=<0 0 0>, rot=<0 0 0 1>}
 lowerarm_l_out_10 {type=Curve, loc=<0 0 0>, rot=<0 0 0 1>}
 lowerarm_l_in_90 {type=Curve, loc=<0 0 0>, rot=<0 0 0 1>}
 lowerarm_r_in_10 {type=Curve, loc=<0 0 0>, rot=<0 0 0 1>}
 lowerarm_r_out_10 {type=Curve, loc=<0 0 0>, rot=<0 0 0 1>}
 lowerarm_r_out_35 {type=Curve, loc=<0 0 0>, rot=<0 0 0 1>}
 lowerarm_r_in_75 {type=Curve, loc=<0 0 0>, rot=<0 0 0 1>}
 lowerarm_r_in_50 {type=Curve, loc=<0 0 0>, rot=<0 0 0 1>}
 lowerarm_r_in_110 {type=Curve, loc=<0 0 0>, rot=<0 0 0 1>}
 lowerarm_r_in_35 {type=Curve, loc=<0 0 0>, rot=<0 0 0 1>}
 lowerarm_r_in_90 {type=Curve, loc=<0 0 0>, rot=<0 0 0 1>}
 thigh_l_bck_10 {type=Curve, loc=<0 0 0>, rot=<0 0 0 1>}
 thigh_l_fwd_10 {type=Curve, loc=<0 0 0>, rot=<0 0 0 1>}
 thigh_l_in_45_out_90 {type=Curve, loc=<0 0 0>, rot=<0 0 0 1>}
 thigh_l_fwd_90 {type=Curve, loc=<0 0 0>, rot=<0 0 0 1>}
 thigh_l_bck_90 {type=Curve, loc=<0 0 0>, rot=<0 0 0 1>}
 thigh_l_bck_50 {type=Curve, loc=<0 0 0>, rot=<0 0 0 1>}
 thigh_l_out_110 {type=Curve, loc=<0 0 0>, rot=<0 0 0 1>}
 thigh_l_out_10 {type=Curve, loc=<0 0 0>, rot=<0 0 0 1>}
 thigh_l_fwd_45 {type=Curve, loc=<0 0 0>, rot=<0 0 0 1>}
 thigh_l_in_50 {type=Curve, loc=<0 0 0>, rot=<0 0 0 1>}
 thigh_l_fwd_110 {type=Curve, loc=<0 0 0>, rot=<0 0 0 1>}
 thigh_l_out_55 {type=Curve, loc=<0 0 0>, rot=<0 0 0 1>}
 thigh_l_out_85 {type=Curve, loc=<0 0 0>, rot=<0 0 0 1>}
 thigh_r_fwd_10 {type=Curve, loc=<0 0 0>, rot=<0 0 0 1>}
 thigh_r_out_55 {type=Curve, loc=<0 0 0>, rot=<0 0 0 1>}
 thigh_r_fwd_90 {type=Curve, loc=<0 0 0>, rot=<0 0 0 1>}
 thigh_r_fwd_45 {type=Curve, loc=<0 0 0>, rot=<0 0 0 1>}
 thigh_r_bck_10 {type=Curve, loc=<0 0 0>, rot=<0 0 0 1>}
 thigh_r_out_85 {type=Curve, loc=<0 0 0>, rot=<0 0 0 1>}
 thigh_r_in_50 {type=Curve, loc=<0 0 0>, rot=<0 0 0 1>}
 thigh_r_out_10 {type=Curve, loc=<0 0 0>, rot=<0 0 0 1>}
 thigh_r_out_110 {type=Curve, loc=<0 0 0>, rot=<0 0 0 1>}
 thigh_r_bck_50 {type=Curve, loc=<0 0 0>, rot=<0 0 0 1>}
 thigh_r_bck_90 {type=Curve, loc=<0 0 0>, rot=<0 0 0 1>}
 thigh_r_in_45_out_90 {type=Curve, loc=<0 0 0>, rot=<0 0 0 1>}
 thigh_r_fwd_110 {type=Curve, loc=<0 0 0>, rot=<0 0 0 1>}
 upperarm_l_back_45 {type=Curve, loc=<0 0 0>, rot=<0 0 0 1>}
 upperarm_l_out_55 {type=Curve, loc=<0 0 0>, rot=<0 0 0 1>}
 upperarm_l_out110_twist_in_70 {type=Curve, loc=<0 0 0>, rot=<0 0 0 1>}
 upperarm_l_out_85 {type=Curve, loc=<0 0 0>, rot=<0 0 0 1>}
 upperarm_l_fwd_110 {type=Curve, loc=<0 0 0>, rot=<0 0 0 1>}
 upperarm_l_in_10 {type=Curve, loc=<0 0 0>, rot=<0 0 0 1>}
 upperarm_l_out110_twist_out_70 {type=Curve, loc=<0 0 0>, rot=<0 0 0 1>}
 upperarm_l_out_twist_a {type=Curve, loc=<0 0 0>, rot=<0 0 0 1>}
 upperarm_l_fwd_15 {type=Curve, loc=<0 0 0>, rot=<0 0 0 1>}
 upperarm_l_out_15 {type=Curve, loc=<0 0 0>, rot=<0 0 0 1>}
 upperarm_l_back_10 {type=Curve, loc=<0 0 0>, rot=<0 0 0 1>}
 upperarm_l_out110 {type=Curve, loc=<0 0 0>, rot=<0 0 0 1>}
 upperarm_l_fwd_twist_in_a {type=Curve, loc=<0 0 0>, rot=<0 0 0 1>}
 upperarm_l_fwd_90 {type=Curve, loc=<0 0 0>, rot=<0 0 0 1>}
 upperarm_r_out_15 {type=Curve, loc=<0 0 0>, rot=<0 0 0 1>}
 upperarm_r_fwd_15 {type=Curve, loc=<0 0 0>, rot=<0 0 0 1>}
 upperarm_r_out_twist_a {type=Curve, loc=<0 0 0>, rot=<0 0 0 1>}
 upperarm_r_out110_twist_in_70 {type=Curve, loc=<0 0 0>, rot=<0 0 0 1>}
 upperarm_r_fwd_90 {type=Curve, loc=<0 0 0>, rot=<0 0 0 1>}
 upperarm_r_out110 {type=Curve, loc=<0 0 0>, rot=<0 0 0 1>}
 upperarm_r_back_10 {type=Curve, loc=<0 0 0>, rot=<0 0 0 1>}
 upperarm_r_out_85 {type=Curve, loc=<0 0 0>, rot=<0 0 0 1>}
 upperarm_r_in_10 {type=Curve, loc=<0 0 0>, rot=<0 0 0 1>}
 upperarm_r_back_45 {type=Curve, loc=<0 0 0>, rot=<0 0 0 1>}
 upperarm_r_out110_twist_out_70 {type=Curve, loc=<0 0 0>, rot=<0 0 0 1>}
 upperarm_r_out_55 {type=Curve, loc=<0 0 0>, rot=<0 0 0 1>}
 upperarm_r_fwd_twist_in_a {type=Curve, loc=<0 0 0>, rot=<0 0 0 1>}
 upperarm_r_fwd_110 {type=Curve, loc=<0 0 0>, rot=<0 0 0 1>}
 Distance {type=Curve, loc=<0 0 0>, rot=<0 0 0 1>}
 RemainingTurnYaw {type=Curve, loc=<0 0 0>, rot=<0 0 0 1>}
 TurnYawWeight {type=Curve, loc=<0 0 0>, rot=<0 0 0 1>}
 GroundDistance {type=Curve, loc=<0 0 0>, rot=<0 0 0 1>}
 DisableLHandIK {type=Curve, loc=<0 0 0>, rot=<0 0 0 1>}
 ScaleDownWeaponR {type=Curve, loc=<0 0 0>, rot=<0 0 0 1>}
 DisableLegIK {type=Curve, loc=<0 0 0>, rot=<0 0 0 1>}
 blendParent1 {type=Curve, loc=<0 0 0>, rot=<0 0 0 1>}
 DisableRHandIK {type=Curve, loc=<0 0 0>, rot=<0 0 0 1>}
 DisableLeftHandPoseOverride {type=Curve, loc=<0 0 0>, rot=<0 0 0 1>}
 applyHipfireOverridePose {type=Curve, loc=<0 0 0>, rot=<0 0 0 1>}
 DisableHandIKRetargeting {type=Curve, loc=<0 0 0>, rot=<0 0 0 1>}
 lowerarm_r_in_80 {type=Curve, loc=<0 0 0>, rot=<0 0 0 1>}
 lowerarm_l_in_80 {type=Curve, loc=<0 0 0>, rot=<0 0 0 1>}
 calf_r_back_130 {type=Curve, loc=<0 0 0>, rot=<0 0 0 1>}
 calf_l_back_130 {type=Curve, loc=<0 0 0>, rot=<0 0 0 1>}
 Pose_9 {type=Curve, loc=<0 0 0>, rot=<0 0 0 1>}
]
