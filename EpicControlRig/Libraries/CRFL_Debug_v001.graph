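# Control Rig function library for debugging, holding one function, Draw Axis, that draws transform axes on rig items.
# When Enable is true, it loops over the Item array, gets each transform and draws foreground axes using the Scale and Thickness inputs.
# A sequence node continues execution to the return afterwards.
# The rig's own event graphs contain only entry events, alongside a humanoid skeleton hierarchy with animation curves.
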ControlRigBlueprint /EpicControlRig/Libraries/CRFL_Debug_v001
graphs [
 ContainedGraph {nodes [
   n1 = RigVMFunctionEntryNode()
   RigUnit_SequenceAggregate = &RigVMFunction_Sequence(ExecuteContext=@n1.ExecuteContext)
   n2 = &RigVMFunction_ControlFlowBranch(ExecuteContext=@RigUnit_SequenceAggregate.A, Condition=RigVMVariableNode() {variable=Enable}.Value)
   For_Each = RigVMDispatchNode(ExecuteContext=@n2.True, Array=RigVMVariableNode() {variable=Item}.Value) {template="DISPATCH_RigVMDispatch_ArrayIterator(in Array,out Element,out Index,out Count,out Ratio)"}
   > DebugTransformMutableItemSpace = /Script/RigVM.RigVMFunction_DebugTransformMutableNoSpace(
    Color=(R=1,G=0,B=0,A=1)
    DebugDrawSettings=(DepthPriority=SDPG_Foreground,Lifetime=-1)
    Mode=Axes
    WorldOffset=(Rotation=(X=0,Y=0,Z=0,W=1),Translation=(X=0,Y=0,Z=0),Scale3D=(X=1,Y=1,Z=1))
    bEnabled=True
    Transform=&RigUnit_GetTransform(Item=@For_Each.Element).Transform
    Thickness=RigVMVariableNode() {variable=Thickness}.Value
    Scale=RigVMVariableNode() {variable=Scale}.Value
   ) {is=struct RigVMUnitNode}
   RigVMFunctionReturnNode(ExecuteContext=@RigUnit_SequenceAggregate.B)
  ]}
 RigVMFunctionLibrary {nodes [Draw Axis = RigVMCollapseNode(Enable=true, ExecuteContext=(), Item=(), Scale=10, Thickness=0.2)]}
 RigVMModel {nodes [RigUnit_BeginExecution = &RigUnit_BeginExecution()]}
 RigVMModel Rig {nodes [&RigUnit_PrepareForExecution()]}
]
hierarchy [
 root {loc=<0 0 0>, rot=<0 0 0 1>}
 pelvis {parent=root, loc=<0 2.2809 95.8968>, rot=<-0.0224 0.7068 0.0224 -0.7068>}
 spine_01 {parent=pelvis, loc=<3.6771 0 0>, rot=<0 0 -0.1258 -0.9921>}
 spine_02 {parent=spine_01, loc=<6.7951 0 0>, rot=<0 0 0.0302 -0.9995>}
 spine_03 {parent=spine_02, loc=<7.2382 0 0>, rot=<0 0 0.0954 -0.9954>}
 spine_04 {parent=spine_03, loc=<8.5239 0 0>, rot=<0 0 0.0512 -0.9987>}
 spine_05 {parent=spine_04, loc=<19.4398 0 0>, rot=<0 0 0.0059 -1>}
 neck_01 {parent=spine_05, loc=<11.8878 0 0>, rot=<0 0 -0.2073 -0.9783>}
 neck_02 {parent=neck_01, loc=<5.1103 0 0>, rot=<0 0 0.0167 -0.9999>}
 head {parent=neck_02, loc=<4.913 0 0>, rot=<0 0 0.1035 -0.9946>}
 clavicle_l {parent=spine_05, loc=<5.5163 -1.3148 -1.4279>, rot=<0.0833 -0.7546 -0.0413 -0.6496>}
 upperarm_l {parent=clavicle_l, loc=<17.8095 0 0>, rot=<-0.0199 -0.3917 -0.0202 -0.9196>}
 lowerarm_l {parent=upperarm_l, loc=<27.7711 0 0>, rot=<0 0 -0.3336 -0.9427>}
 lowerarm_twist_02_l {parent=lowerarm_l, loc=<9.0837 0 0>, rot=<0.0115 0.0149 0.0175 -0.9997>}
 lowerarm_twist_01_l {parent=lowerarm_l, loc=<18.1674 0 0>, rot=<0.0115 0.0149 0.0175 -0.9997>}
 hand_l {parent=lowerarm_l, loc=<27.2511 0 0>, rot=<-0.5576 -0.0017 0.0206 -0.8299>}
 middle_metacarpal_l {parent=hand_l, loc=<3.3758 0.7536 -0.1829>, rot=<-0.0373 0.0004 -0.0203 -0.9991>}
 middle_01_l {parent=middle_metacarpal_l, loc=<6.0982 0 0>, rot=<0 0 0.2721 -0.9623>}
 middle_02_l {parent=middle_01_l, loc=<5.169 0 0>, rot=<0 0 0.1803 -0.9836>}
 middle_03_l {parent=middle_02_l, loc=<2.474 0 0>, rot=<0 0 0.0872 -0.9962>}
 pinky_metacarpal_l {parent=hand_l, loc=<3.3144 0.3059 2.3911>, rot=<-0.2182 0.1882 0.0583 -0.9558>}
 pinky_01_l {parent=pinky_metacarpal_l, loc=<4.9576 0.1431 -0.1988>, rot=<0.09 -0.017 0.1281 -0.9875>}
 pinky_02_l {parent=pinky_01_l, loc=<3.816 0 0>, rot=<0 0 0.1847 -0.9828>}
 pinky_03_l {parent=pinky_02_l, loc=<2.04 0 0>, rot=<0 0 0.0429 -0.9991>}
 ring_metacarpal_l {parent=hand_l, loc=<3.3743 0.5425 1.0918>, rot=<-0.1166 0.1006 -0.0257 -0.9877>}
 ring_01_l {parent=ring_metacarpal_l, loc=<5.6455 0.0416 -0.0207>, rot=<0.0542 -0.0132 0.2535 -0.9657>}
 ring_02_l {parent=ring_01_l, loc=<4.977 0 0>, rot=<0 0 0.1647 -0.9863>}
 ring_03_l {parent=ring_02_l, loc=<2.265 0 0>, rot=<0 0 0.0799 -0.9968>}
 thumb_01_l {parent=hand_l, loc=<1.9924 -1.3566 -2.5815>, rot=<0.5052 -0.3691 -0.067 -0.7772>}
 thumb_02_l {parent=thumb_01_l, loc=<4.378 0 0>, rot=<0.0336 0.0103 0.2019 -0.9788>}
 thumb_03_l {parent=thumb_02_l, loc=<3.086 0 0>, rot=<0 0 0.0872 -0.9962>}
 index_metacarpal_l {parent=hand_l, loc=<3.4445 0.3847 -2.3793>, rot=<0.029 -0.0637 -0.0071 -0.9975>}
 index_01_l {parent=index_metacarpal_l, loc=<5.8771 -0.0432 0.2409>, rot=<0 0 0.2026 -0.9793>}
 index_02_l {parent=index_01_l, loc=<4.08 0 0>, rot=<0 0 0.1296 -0.9916>}
 index_03_l {parent=index_02_l, loc=<2.595 0 0>, rot=<0 0 0.109 -0.994>}
 upperarm_twist_01_l {parent=upperarm_l, loc=<9.257 0 0>, rot=<-0.0001 0.0188 -0.0028 -0.9998>}
 upperarm_twist_02_l {parent=upperarm_l, loc=<18.5141 0 0>, rot=<0 0 0 1>}
 clavicle_r {parent=spine_05, loc=<5.5162 -1.3148 1.4279>, rot=<0.7546 0.0833 -0.6496 0.0413>}
 upperarm_r {parent=clavicle_r, loc=<-17.8096 0 0.0004>, rot=<-0.0199 -0.3917 -0.0202 -0.9196>}
 lowerarm_r {parent=upperarm_r, loc=<-27.7707 0 0>, rot=<0 0 -0.3336 -0.9427>}
 lowerarm_twist_02_r {parent=lowerarm_r, loc=<-9.0837 0 0>, rot=<0.0115 0.0149 0.0175 -0.9997>}
 lowerarm_twist_01_r {parent=lowerarm_r, loc=<-18.1673 0 0>, rot=<0.0115 0.0149 0.0175 -0.9997>}
 hand_r {parent=lowerarm_r, loc=<-27.251 0 0>, rot=<-0.5576 -0.0017 0.0206 -0.8299>}
 middle_metacarpal_r {parent=hand_r, loc=<-3.3758 -0.754 0.1828>, rot=<-0.0373 0.0004 -0.0203 -0.9991>}
 middle_01_r {parent=middle_metacarpal_r, loc=<-6.0984 0.0001 0>, rot=<0 0 0.2721 -0.9623>}
 middle_02_r {parent=middle_01_r, loc=<-5.169 0.0001 0>, rot=<0 0 0.1803 -0.9836>}
 middle_03_r {parent=middle_02_r, loc=<-2.474 0 0>, rot=<0 0 0.0872 -0.9962>}
 pinky_metacarpal_r {parent=hand_r, loc=<-3.3147 -0.3059 -2.3913>, rot=<-0.2182 0.1882 0.0583 -0.9558>}
 pinky_01_r {parent=pinky_metacarpal_r, loc=<-4.9573 -0.1433 0.1989>, rot=<0.09 -0.017 0.1281 -0.9875>}
 pinky_02_r {parent=pinky_01_r, loc=<-3.816 0 0>, rot=<0 0 0.1847 -0.9828>}
 pinky_03_r {parent=pinky_02_r, loc=<-2.04 0 0>, rot=<0 0 0.0429 -0.9991>}
 ring_metacarpal_r {parent=hand_r, loc=<-3.3742 -0.543 -1.0918>, rot=<-0.1166 0.1006 -0.0257 -0.9877>}
 ring_01_r {parent=ring_metacarpal_r, loc=<-5.6457 -0.0414 0.0207>, rot=<0.0542 -0.0132 0.2535 -0.9657>}
 ring_02_r {parent=ring_01_r, loc=<-4.9771 0 0>, rot=<0 0 0.1647 -0.9863>}
 ring_03_r {parent=ring_02_r, loc=<-2.265 -0.0001 0>, rot=<0 0 0.0799 -0.9968>}
 thumb_01_r {parent=hand_r, loc=<-1.9928 1.3567 2.5813>, rot=<0.5052 -0.3691 -0.067 -0.7772>}
 thumb_02_r {parent=thumb_01_r, loc=<-4.3778 -0.0004 -0.0001>, rot=<0.0336 0.0103 0.2019 -0.9788>}
 thumb_03_r {parent=thumb_02_r, loc=<-3.086 0 0>, rot=<0 0 0.0872 -0.9962>}
 index_metacarpal_r {parent=hand_r, loc=<-3.4445 -0.3852 2.3793>, rot=<0.029 -0.0637 -0.0071 -0.9975>}
 index_01_r {parent=index_metacarpal_r, loc=<-5.8772 0.0434 -0.241>, rot=<0 0 0.2026 -0.9793>}
 index_02_r {parent=index_01_r, loc=<-4.0799 0 0>, rot=<0 0 0.1296 -0.9916>}
 index_03_r {parent=index_02_r, loc=<-2.5951 0 0>, rot=<0 0 0.109 -0.994>}
 upperarm_twist_01_r {parent=upperarm_r, loc=<-9.2569 0 -0.0001>, rot=<-0.0001 0.0188 -0.0028 -0.9998>}
 upperarm_twist_02_r {parent=upperarm_r, loc=<-18.5138 0 -0.0002>, rot=<0 0 0 1>}
 thigh_r {parent=pelvis, loc=<-2.3657 0.1195 9.9691>, rot=<0.0295 -0.0724 0.9965 -0.029>}
 calf_r {parent=thigh_r, loc=<43.3413 0 0>, rot=<0 0 -0.0437 -0.999>}
 foot_r {parent=calf_r, loc=<42.2179 0 0>, rot=<-0.0007 -0.0269 0.0232 -0.9994>}
 ball_r {parent=foot_r, loc=<7.0094 -15.2376 -0.5389>, rot=<0 0 -0.7071 -0.7071>}
 calf_twist_02_r {parent=calf_r, loc=<14.0726 0 0.0377>, rot=<-0.0001 -0.0023 0.0232 -0.9997>}
 calf_twist_01_r {parent=calf_r, loc=<28.1453 0 0.0755>, rot=<-0.0001 -0.0023 0.0232 -0.9997>}
 thigh_twist_01_r {parent=thigh_r, loc=<14.4471 0 0>, rot=<0 -0.0023 -0.0114 -0.9999>}
 thigh_twist_02_r {parent=thigh_r, loc=<28.8942 0 0>, rot=<0 -0.0023 -0.0114 -0.9999>}
 thigh_l {parent=pelvis, loc=<-2.3657 0.11 -9.9692>, rot=<0.0724 0.0295 -0.029 -0.9965>}
 calf_l {parent=thigh_l, loc=<-43.3413 0 0>, rot=<0 0 -0.0437 -0.999>}
 foot_l {parent=calf_l, loc=<-42.2179 0 0>, rot=<-0.0007 -0.0269 0.0232 -0.9994>}
 ball_l {parent=foot_l, loc=<-7.0094 15.2376 0.5389>, rot=<0 0 -0.7071 -0.7071>}
 calf_twist_02_l {parent=calf_l, loc=<-14.0726 0 0>, rot=<-0.0001 -0.0023 0.0232 -0.9997>}
 calf_twist_01_l {parent=calf_l, loc=<-28.1453 0 -0.0754>, rot=<-0.0001 -0.0023 0.0232 -0.9997>}
 thigh_twist_01_l {parent=thigh_l, loc=<-14.4471 0 0>, rot=<0 -0.0023 -0.0114 -0.9999>}
 thigh_twist_02_l {parent=thigh_l, loc=<-28.8942 0 0>, rot=<0 -0.0023 -0.0114 -0.9999>}
 ik_hand_root {parent=root, loc=<0 0 0>, rot=<0 0 0 1>}
 ik_hand_gun {parent=ik_hand_root, loc=<-47.7683 15.6988 104.4664>, rot=<0.5985 0.0096 0.5206 -0.6089>}
 ik_hand_r {parent=ik_hand_gun, loc=<0 0 0>, rot=<0 0 0 1>}
 ik_hand_l {parent=ik_hand_gun, loc=<43.7336 61.6698 58.4081>, rot=<-0.4578 -0.6455 -0.6114 0>}
 ik_foot_root {parent=root, loc=<0 0 0>, rot=<0 0 0 1>}
 ik_foot_l {parent=ik_foot_root, loc=<14.0877 -0.9943 8.2376>, rot=<0.0636 0.7042 0.0414 -0.7059>}
 ik_foot_r {parent=ik_foot_root, loc=<-14.0876 -0.9848 8.2382>, rot=<0.7059 0.0414 -0.7042 0.0636>}
 interaction {parent=root, loc=<0 0 0>, rot=<0 0 0 1>}
 center_of_mass {parent=root, loc=<0 0 0>, rot=<0 0 0 1>}
 MoveData_Speed {type=Curve, loc=<0 0 0>, rot=<0 0 0 1>}
 base_pose {type=Curve, loc=<0 0 0>, rot=<0 0 0 1>}
 calf_l_back_50 {type=Curve, loc=<0 0 0>, rot=<0 0 0 1>}
 calf_l_back_120 {type=Curve, loc=<0 0 0>, rot=<0 0 0 1>}
 calf_l_back_90 {type=Curve, loc=<0 0 0>, rot=<0 0 0 1>}
 calf_l_back_150 {type=Curve, loc=<0 0 0>, rot=<0 0 0 1>}
 calf_r_back_50 {type=Curve, loc=<0 0 0>, rot=<0 0 0 1>}
 calf_r_back_120 {type=Curve, loc=<0 0 0>, rot=<0 0 0 1>}
 calf_r_back_90 {type=Curve, loc=<0 0 0>, rot=<0 0 0 1>}
 calf_r_back_150 {type=Curve, loc=<0 0 0>, rot=<0 0 0 1>}
 clavicle_l_fwd_30 {type=Curve, loc=<0 0 0>, rot=<0 0 0 1>}
 clavicle_l_down_20 {type=Curve, loc=<0 0 0>, rot=<0 0 0 1>}
 clavicle_l_back_30 {type=Curve, loc=<0 0 0>, rot=<0 0 0 1>}
 clavicle_l_up_40 {type=Curve, loc=<0 0 0>, rot=<0 0 0 1>}
 clavicle_r_fwd_30 {type=Curve, loc=<0 0 0>, rot=<0 0 0 1>}
 clavicle_r_down_20 {type=Curve, loc=<0 0 0>, rot=<0 0 0 1>}
 clavicle_r_back_30 {type=Curve, loc=<0 0 0>, rot=<0 0 0 1>}
 clavicle_r_up_40 {type=Curve, loc=<0 0 0>, rot=<0 0 0 1>}
 clavicle_r_back_310 {type=Curve, loc=<0 0 0>, rot=<0 0 0 1>}
 clavicle_r_up_401 {type=Curve, loc=<0 0 0>, rot=<0 0 0 1>}
 foot_l_down_60 {type=Curve, loc=<0 0 0>, rot=<0 0 0 1>}
 foot_l_up_35 {type=Curve, loc=<0 0 0>, rot=<0 0 0 1>}
 foot_r_up_35 {type=Curve, loc=<0 0 0>, rot=<0 0 0 1>}
 foot_r_down_60 {type=Curve, loc=<0 0 0>, rot=<0 0 0 1>}
 hand_l_down_90 {type=Curve, loc=<0 0 0>, rot=<0 0 0 1>}
 hand_l_up_20 {type=Curve, loc=<0 0 0>, rot=<0 0 0 1>}
 hand_l_up_90 {type=Curve, loc=<0 0 0>, rot=<0 0 0 1>}
 hand_r_up_90 {type=Curve, loc=<0 0 0>, rot=<0 0 0 1>}
 hand_r_down_90 {type=Curve, loc=<0 0 0>, rot=<0 0 0 1>}
 hand_r_up_20 {type=Curve, loc=<0 0 0>, rot=<0 0 0 1>}
 Pose_0 {type=Curve, loc=<0 0 0>, rot=<0 0 0 1>}
 Pose_1 {type=Curve, loc=<0 0 0>, rot=<0 0 0 1>}
 Pose_2 {type=Curve, loc=<0 0 0>, rot=<0 0 0 1>}
 Pose_3 {type=Curve, loc=<0 0 0>, rot=<0 0 0 1>}
 Pose_4 {type=Curve, loc=<0 0 0>, rot=<0 0 0 1>}
 Pose_5 {type=Curve, loc=<0 0 0>, rot=<0 0 0 1>}
 Pose_6 {type=Curve, loc=<0 0 0>, rot=<0 0 0 1>}
 Pose_7 {type=Curve, loc=<0 0 0>, rot=<0 0 0 1>}
 Pose_8 {type=Curve, loc=<0 0 0>, rot=<0 0 0 1>}
 lowerarm_l_in_110 {type=Curve, loc=<0 0 0>, rot=<0 0 0 1>}
 lowerarm_l_in_35 {type=Curve, loc=<0 0 0>, rot=<0 0 0 1>}
 lowerarm_l_out_35 {type=Curve, loc=<0 0 0>, rot=<0 0 0 1>}
 lowerarm_l_in_50 {type=Curve, loc=<0 0 0>, rot=<0 0 0 1>}
 lowerarm_l_in_75 {type=Curve, loc=<0 0 0>, rot=<0 0 0 1>}
 lowerarm_l_in_10 {type=Curve, loc=<0 0 0>, rot=<0 0 0 1>}
 lowerarm_l_out_10 {type=Curve, loc=<0 0 0>, rot=<0 0 0 1>}
 lowerarm_l_in_90 {type=Curve, loc=<0 0 0>, rot=<0 0 0 1>}
 lowerarm_r_in_10 {type=Curve, loc=<0 0 0>, rot=<0 0 0 1>}
 lowerarm_r_out_10 {type=Curve, loc=<0 0 0>, rot=<0 0 0 1>}
 lowerarm_r_out_35 {type=Curve, loc=<0 0 0>, rot=<0 0 0 1>}
 lowerarm_r_in_75 {type=Curve, loc=<0 0 0>, rot=<0 0 0 1>}
 lowerarm_r_in_50 {type=Curve, loc=<0 0 0>, rot=<0 0 0 1>}
 lowerarm_r_in_110 {type=Curve, loc=<0 0 0>, rot=<0 0 0 1>}
 lowerarm_r_in_35 {type=Curve, loc=<0 0 0>, rot=<0 0 0 1>}
 lowerarm_r_in_90 {type=Curve, loc=<0 0 0>, rot=<0 0 0 1>}
 thigh_l_bck_10 {type=Curve, loc=<0 0 0>, rot=<0 0 0 1>}
 thigh_l_fwd_10 {type=Curve, loc=<0 0 0>, rot=<0 0 0 1>}
 thigh_l_in_45_out_90 {type=Curve, loc=<0 0 0>, rot=<0 0 0 1>}
 thigh_l_fwd_90 {type=Curve, loc=<0 0 0>, rot=<0 0 0 1>}
 thigh_l_bck_90 {type=Curve, loc=<0 0 0>, rot=<0 0 0 1>}
 thigh_l_bck_50 {type=Curve, loc=<0 0 0>, rot=<0 0 0 1>}
 thigh_l_out_110 {type=Curve, loc=<0 0 0>, rot=<0 0 0 1>}
 thigh_l_out_10 {type=Curve, loc=<0 0 0>, rot=<0 0 0 1>}
 thigh_l_fwd_45 {type=Curve, loc=<0 0 0>, rot=<0 0 0 1>}
 thigh_l_in_50 {type=Curve, loc=<0 0 0>, rot=<0 0 0 1>}
 thigh_l_fwd_110 {type=Curve, loc=<0 0 0>, rot=<0 0 0 1>}
 thigh_l_out_55 {type=Curve, loc=<0 0 0>, rot=<0 0 0 1>}
 thigh_l_out_85 {type=Curve, loc=<0 0 0>, rot=<0 0 0 1>}
 thigh_r_fwd_10 {type=Curve, loc=<0 0 0>, rot=<0 0 0 1>}
 thigh_r_out_55 {type=Curve, loc=<0 0 0>, rot=<0 0 0 1>}
 thigh_r_fwd_90 {type=Curve, loc=<0 0 0>, rot=<0 0 0 1>}
 thigh_r_fwd_45 {type=Curve, loc=<0 0 0>, rot=<0 0 0 1>}
 thigh_r_bck_10 {type=Curve, loc=<0 0 0>, rot=<0 0 0 1>}
 thigh_r_out_85 {type=Curve, loc=<0 0 0>, rot=<0 0 0 1>}
 thigh_r_in_50 {type=Curve, loc=<0 0 0>, rot=<0 0 0 1>}
 thigh_r_out_10 {type=Curve, loc=<0 0 0>, rot=<0 0 0 1>}
 thigh_r_out_110 {type=Curve, loc=<0 0 0>, rot=<0 0 0 1>}
 thigh_r_bck_50 {type=Curve, loc=<0 0 0>, rot=<0 0 0 1>}
 thigh_r_bck_90 {type=Curve, loc=<0 0 0>, rot=<0 0 0 1>}
 thigh_r_in_45_out_90 {type=Curve, loc=<0 0 0>, rot=<0 0 0 1>}
 thigh_r_fwd_110 {type=Curve, loc=<0 0 0>, rot=<0 0 0 1>}
 upperarm_l_back_45 {type=Curve, loc=<0 0 0>, rot=<0 0 0 1>}
 upperarm_l_out_55 {type=Curve, loc=<0 0 0>, rot=<0 0 0 1>}
 upperarm_l_out110_twist_in_70 {type=Curve, loc=<0 0 0>, rot=<0 0 0 1>}
 upperarm_l_out_85 {type=Curve, loc=<0 0 0>, rot=<0 0 0 1>}
 upperarm_l_fwd_110 {type=Curve, loc=<0 0 0>, rot=<0 0 0 1>}
 upperarm_l_in_10 {type=Curve, loc=<0 0 0>, rot=<0 0 0 1>}
 upperarm_l_out110_twist_out_70 {type=Curve, loc=<0 0 0>, rot=<0 0 0 1>}
 upperarm_l_out_twist_a {type=Curve, loc=<0 0 0>, rot=<0 0 0 1>}
 upperarm_l_fwd_15 {type=Curve, loc=<0 0 0>, rot=<0 0 0 1>}
 upperarm_l_out_15 {type=Curve, loc=<0 0 0>, rot=<0 0 0 1>}
 upperarm_l_back_10 {type=Curve, loc=<0 0 0>, rot=<0 0 0 1>}
 upperarm_l_out110 {type=Curve, loc=<0 0 0>, rot=<0 0 0 1>}
 upperarm_l_fwd_twist_in_a {type=Curve, loc=<0 0 0>, rot=<0 0 0 1>}
 upperarm_l_fwd_90 {type=Curve, loc=<0 0 0>, rot=<0 0 0 1>}
 upperarm_r_out_15 {type=Curve, loc=<0 0 0>, rot=<0 0 0 1>}
 upperarm_r_fwd_15 {type=Curve, loc=<0 0 0>, rot=<0 0 0 1>}
 upperarm_r_out_twist_a {type=Curve, loc=<0 0 0>, rot=<0 0 0 1>}
 upperarm_r_out110_twist_in_70 {type=Curve, loc=<0 0 0>, rot=<0 0 0 1>}
 upperarm_r_fwd_90 {type=Curve, loc=<0 0 0>, rot=<0 0 0 1>}
 upperarm_r_out110 {type=Curve, loc=<0 0 0>, rot=<0 0 0 1>}
 upperarm_r_back_10 {type=Curve, loc=<0 0 0>, rot=<0 0 0 1>}
 upperarm_r_out_85 {type=Curve, loc=<0 0 0>, rot=<0 0 0 1>}
 upperarm_r_in_10 {type=Curve, loc=<0 0 0>, rot=<0 0 0 1>}
 upperarm_r_back_45 {type=Curve, loc=<0 0 0>, rot=<0 0 0 1>}
 upperarm_r_out110_twist_out_70 {type=Curve, loc=<0 0 0>, rot=<0 0 0 1>}
 upperarm_r_out_55 {type=Curve, loc=<0 0 0>, rot=<0 0 0 1>}
 upperarm_r_fwd_twist_in_a {type=Curve, loc=<0 0 0>, rot=<0 0 0 1>}
 upperarm_r_fwd_110 {type=Curve, loc=<0 0 0>, rot=<0 0 0 1>}
 Distance {type=Curve, loc=<0 0 0>, rot=<0 0 0 1>}
 RemainingTurnYaw {type=Curve, loc=<0 0 0>, rot=<0 0 0 1>}
 TurnYawWeight {type=Curve, loc=<0 0 0>, rot=<0 0 0 1>}
 GroundDistance {type=Curve, loc=<0 0 0>, rot=<0 0 0 1>}
 DisableLHandIK {type=Curve, loc=<0 0 0>, rot=<0 0 0 1>}
 ScaleDownWeaponR {type=Curve, loc=<0 0 0>, rot=<0 0 0 1>}
 DisableLegIK {type=Curve, loc=<0 0 0>, rot=<0 0 0 1>}
 blendParent1 {type=Curve, loc=<0 0 0>, rot=<0 0 0 1>}
 DisableRHandIK {type=Curve, loc=<0 0 0>, rot=<0 0 0 1>}
 DisableLeftHandPoseOverride {type=Curve, loc=<0 0 0>, rot=<0 0 0 1>}
 applyHipfireOverridePose {type=Curve, loc=<0 0 0>, rot=<0 0 0 1>}
 DisableHandIKRetargeting {type=Curve, loc=<0 0 0>, rot=<0 0 0 1>}
 lowerarm_r_in_80 {type=Curve, loc=<0 0 0>, rot=<0 0 0 1>}
 lowerarm_l_in_80 {type=Curve, loc=<0 0 0>, rot=<0 0 0 1>}
 calf_r_back_130 {type=Curve, loc=<0 0 0>, rot=<0 0 0 1>}
 calf_l_back_130 {type=Curve, loc=<0 0 0>, rot=<0 0 0 1>}
 Pose_9 {type=Curve, loc=<0 0 0>, rot=<0 0 0 1>}
]
