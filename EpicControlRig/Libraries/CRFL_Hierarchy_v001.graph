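# Control Rig function library of reusable hierarchy-building and rig utility functions.
# Control stack functions spawn controls with optional secondary controls and nulls, linked through metadata.
# Movable-pivot proxy functions use interaction state and a buffer null to rotate a driven control around the proxy.
# Other helpers color controls by side or module, scale shapes, toggle visibility, search children, build plane-aligned virtual bones and set up twist-following pole vectors.
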ControlRigBlueprint /EpicControlRig/Libraries/CRFL_Hierarchy_v001
graphs [
 Concat_2_ContainedGraph {nodes [
   n1 = RigVMFunctionEntryNode()
   RigVMFunctionReturnNode(Result=/Script/RigVM.RigVMFunction_NameConcat(A=&RigVMFunction_NameConcat(A=@n1.A, B=@n1.B) {template="Concat::Execute(in A,in B,out Result)"}.Result, B=@n1.C) {is=struct RigVMUnitNode, template="Concat::Execute(in A,in B,out Result)"}.Result)
  ]}
 ContainedGraph {nodes [
   n1 = RigVMFunctionEntryNode()
   Replace = /Script/RigVM.RigVMFunction_NameReplace(New=None, Old=@n1.Old Suffix, Name=@n1.Item.Name) {is=struct RigVMUnitNode, template="Replace::Execute(in Name,in Old,in New,out Result)"}
   Concat = &RigVMFunction_NameConcat(A=@Replace.Result, B=@n1.New Suffix) {template="Concat::Execute(in A,in B,out Result)"}
   > n2 = &RigUnit_HierarchyAddNull(
    Space=GlobalSpace
    Name=@Concat.Result
    Transform=&RigUnit_GetTransform(bInitial=true, Item=RigVMVariableNode() {variable=Item}.Value).Transform
    Parent=/Script/ControlRig.RigUnit_HierarchyGetParent(bDefaultParent=True, Child=RigVMVariableNode() {variable=Item}.Value) {is=struct RigVMUnitNode}.Parent
   )
   > /Script/ControlRig.RigUnit_SetDefaultParent(Child=RigVMVariableNode() {variable=Item}.Value, Parent=@n2.Item) {is=struct RigVMUnitNode}
   > RigVMFunctionReturnNode(Null=@n2.Item)
  ]}
 ContainedGraph {nodes [
   RigVMFunctionEntryNode()
   > RigVMFunction_Sequence = RigVMAggregateNode()
   BranchNode_2 = &RigVMFunction_ControlFlowBranch(Condition=RigVMVariableNode() {variable=Add Secondary Control}.Value, ExecuteContext=@RigVMFunction_Sequence.B)
   n1 = RigVMVariableNode() {variable=Name}
   n2 = /Script/ControlRig.RigUnit_HierarchyAddControlTransform(
    InitialValue=(Rotation=(X=0,Y=0,Z=0,W=1),Translation=(X=0,Y=0,Z=0),Scale3D=(X=1,Y=1,Z=1))
    OffsetSpace=LocalSpace
    OffsetTransform=(Rotation=(X=0,Y=0,Z=0,W=1),Translation=(X=0,Y=0,Z=0),Scale3D=(X=1,Y=1,Z=1))
    ExecutePin=@BranchNode_2.True
    Name=&RigVMFunction_NameConcat(
     A=&RigVMFunction_NameConcat(B=_sec, A=@n1.Value) {template="Concat::Execute(in A,in B,out Result)"}.Result
     B=RigVMVariableNode() {variable=Control Suffix}.Value
    ) {template="Concat::Execute(in A,in B,out Result)"}.Result
    Parent=RigVMVariableNode() {variable=the control}.Value
    Settings=RigVMVariableNode() {variable=Secondary Settings}.Value
   ) {
    is=struct RigVMUnitNode
    template="SpawnControl::Execute(in InitialValue,in Settings,in OffsetTransform,in OffsetSpace,in Parent,in Name,out Item)"
   }
   > RigVMVariableNode(Value=@n2.Item) {variable=the secondary ctrl}
   > RigVMDispatchNode(
    Name=Secondary
    NameSpace=Self
    Item=RigVMVariableNode() {variable=the control}.Value
    Value=RigVMVariableNode() {variable=the secondary ctrl}.Value
   ) {template="DISPATCH_RigDispatch_SetMetadata(in Item,in Name,in NameSpace,in Value,out Success)"}
   n3 = RigVMVariableNode() {variable=Global Position}
   Concat = &RigVMFunction_NameConcat(A=RigVMVariableNode() {variable=Name}.Value, B=RigVMVariableNode() {variable=Control Suffix}.Value) {template="Concat::Execute(in A,in B,out Result)"}
   SpawnControl = &RigUnit_HierarchyAddControlTransform(
    OffsetSpace=LocalSpace
    Parent=(Type=Bone,Name="None")
    InitialValue=&RigVMFunction_MathTransformMakeRelative(Global=@n3.Value, Parent=@n3.Value).Local
    Settings=RigVMVariableNode() {variable=Settings}.Value
    OffsetTransform=@n3.Value
    Name=@Concat.Result
    ExecutePin=@RigVMFunction_Sequence.A
   ) {template="SpawnControl::Execute(in InitialValue,in Settings,in OffsetTransform,in OffsetSpace,in Parent,in Name,out Item)"}
   > &RigUnit_SetDefaultParent(Parent=RigVMVariableNode() {variable=Parent}.Value, Child=@SpawnControl.Item)
   > RigVMVariableNode(Value=@SpawnControl.Item) {variable=the control}
   > n4 = &RigUnit_SetTransform(
    Weight=1
    bInitial=true
    Value=&RigUnit_GetTransform(bInitial=true, Item=RigVMVariableNode() {variable=the control}.Value).Transform
    Item=RigVMVariableNode() {variable=the control}.Value
   )
   n5 = &RigVMFunction_ControlFlowBranch(ExecuteContext=@n4.ExecutePin, Condition=RigVMVariableNode() {variable=Orient to World}.Value)
   n6 = RigVMVariableNode() {variable=the control}
   n7 = &RigUnit_GetTransform(bInitial=true, Item=@n6.Value)
   &RigUnit_SetTransform(
    Value=(Rotation=(X=0,Y=0,Z=0,W=1),Translation=(X=0,Y=0,Z=0),Scale3D=(X=1,Y=1,Z=1))
    Weight=1
    bInitial=true
    Value.Scale3D=@n7.Transform.Scale3D
    Value.Translation=@n7.Transform.Translation
    ExecutePin=@n5.True
    Item=@n6.Value
   )
   > /Script/ControlRig.RigUnit_SetMetadataTag(NameSpace=Self, Tag=Controls, Item=RigVMVariableNode() {variable=the control}.Value) {is=struct RigVMUnitNode}
   n8 = &RigVMFunction_ControlFlowBranch(Condition=RigVMVariableNode() {variable=Add Null}.Value, ExecuteContext=@RigVMFunction_Sequence.C)
   n9 = RigVMVariableNode() {variable=the control}
   Add Null Above = RigVMFunctionReferenceNode(
    ExecuteContext=@n8.True
    Item=@n9.Value
    Old Suffix=RigVMVariableNode() {variable=Control Suffix}.Value
    New Suffix=RigVMVariableNode() {variable=Null Suffix}.Value
   )
   > DISPATCH_RigDispatch_SetMetadata = RigVMDispatchNode(Name=Null, NameSpace=Self, Value=@"Add Null Above".Null, Item=@n9.Value) {template="DISPATCH_RigDispatch_SetMetadata(in Item,in Name,in NameSpace,in Value,out Success)"}
   > RigVMVariableNode(Value=@"Add Null Above".Null) {variable=the null above}
   > &RigUnit_SetTransform(
    Weight=1
    bInitial=true
    Value=&RigUnit_GetTransform(bInitial=true, Item=RigVMVariableNode() {variable=the null above}.Value).Transform
    Item=RigVMVariableNode() {variable=the control}.Value
   )
   BranchNode_1 = &RigVMFunction_ControlFlowBranch(Condition=RigVMVariableNode() {variable=Add Bottom Null}.Value, ExecuteContext=@RigVMFunction_Sequence.D)
   n10 = &RigUnit_HierarchyAddNull(
    Space=LocalSpace
    Transform=(Rotation=(X=0,Y=0,Z=0,W=1),Translation=(X=0,Y=0,Z=0),Scale3D=(X=1,Y=1,Z=1))
    ExecutePin=@BranchNode_1.True
    Name=&RigVMFunction_NameConcat(B=_ctrl_null, A=RigVMVariableNode() {variable=Name}.Value) {template="Concat::Execute(in A,in B,out Result)"}.Result
    Parent=RigVMDispatchNode(
     False=RigVMVariableNode() {variable=the control}.Value
     Condition=RigVMVariableNode() {variable=Add Secondary Control}.Value
     True=RigVMVariableNode() {variable=the secondary ctrl}.Value
    ) {template="DISPATCH_RigVMDispatch_If(in Condition,in True,in False,out Result)"}.Result
   )
   > RigVMDispatchNode(Name=Bottom Null, NameSpace=Self, Item=RigVMVariableNode() {variable=the control}.Value, Value=@n10.Item) {template="DISPATCH_RigDispatch_SetMetadata(in Item,in Name,in NameSpace,in Value,out Success)"}
   > RigVMVariableNode(Value=@n10.Item) {variable=the null below}
   RigVMFunctionReturnNode(
    Control=RigVMVariableNode() {variable=the control}.Value
    Secondary=RigVMVariableNode() {variable=the secondary ctrl}.Value
    Top Null=RigVMVariableNode() {variable=the null above}.Value
    Bottom Null=RigVMVariableNode() {variable=the null below}.Value
    ExecuteContext=@RigVMFunction_Sequence.E
   )
  ]}
 ContainedGraph {nodes [
   n1 = RigVMFunctionEntryNode()
   n2 = RigVMVariableNode() {variable=Name}
   DISPATCH_RigDispatch_CoreNotEquals = RigVMDispatchNode(B=None, A=@n2.Value) {template="DISPATCH_RigVMDispatch_CoreNotEquals(in A,in B,out Result)"}
   > Control Stack at Position = RigVMFunctionReferenceNode(
    Add Bottom Null=@n1.Add Bottom Null
    Add Null=@n1.Add Null
    Add Secondary Control=@n1.Add Secondary Control
    Control Suffix=@n1.Control Suffix
    Null Suffix=@n1.Null Suffix
    Orient to World=@n1.Orient to World
    Parent=@n1.Parent
    Secondary Settings=@n1.Secondary Settings
    Settings=@n1.Settings
    Global Position=&RigUnit_GetTransform(bInitial=true, Item=RigVMVariableNode() {variable=Item}.Value).Transform
    Name=RigVMDispatchNode(Condition=@DISPATCH_RigDispatch_CoreNotEquals.Result, True=@n2.Value, False=@n1.Item.Name) {template="DISPATCH_RigVMDispatch_If(in Condition,in True,in False,out Result)"}.Result
   )
   > RigVMDispatchNode(Name=Item, NameSpace=Self, Item=@"Control Stack at Position".Control, Value=RigVMVariableNode() {variable=Item}.Value) {template="DISPATCH_RigDispatch_SetMetadata(in Item,in Name,in NameSpace,in Value,out Success)"}
   > RigVMFunctionReturnNode(
    Bottom Null=@"Control Stack at Position".Bottom Null
    Control=@"Control Stack at Position".Control
    Secondary=@"Control Stack at Position".Secondary
    Top Null=@"Control Stack at Position".Top Null
   )
  ]}
 ContainedGraph {nodes [
   n1 = RigVMFunctionEntryNode()
   n2 = &RigUnit_GetTransform(bInitial=true, Item=RigVMVariableNode() {variable=Proxy Snap To}.Value)
   > n3 = &RigUnit_HierarchyAddNull(
    Space=GlobalSpace
    Transform=(Rotation=(X=0,Y=0,Z=0,W=1),Translation=(X=0,Y=0,Z=0),Scale3D=(X=1,Y=1,Z=1))
    Name=&RigVMFunction_NameConcat(B=" Null", A=RigVMVariableNode() {variable=Proxy Name}.Value) {template="Concat::Execute(in A,in B,out Result)"}.Result
    Parent=@n1.Proxy Parent Control
    Transform.Rotation=&RigUnit_GetTransform(bInitial=true, Item=RigVMVariableNode() {variable=Driven Control}.Value).Transform.Rotation
    Transform.Scale3D=@n2.Transform.Scale3D
    Transform.Translation=@n2.Transform.Translation
   )
   > SpawnControl = &RigUnit_HierarchyAddControlTransform(
    InitialValue=(Rotation=(X=0,Y=0,Z=0,W=1),Translation=(X=0,Y=0,Z=0),Scale3D=(X=1,Y=1,Z=1))
    OffsetSpace=LocalSpace
    OffsetTransform=(Rotation=(X=0,Y=0,Z=0,W=1),Translation=(X=0,Y=0,Z=0),Scale3D=(X=1,Y=1,Z=1))
    Parent=@n3.Item
    Name=RigVMVariableNode() {variable=Proxy Name}.Value
    Settings=RigVMVariableNode() {variable=Proxy Shape Settings}.Value
   ) {template="SpawnControl::Execute(in InitialValue,in Settings,in OffsetTransform,in OffsetSpace,in Parent,in Name,out Item)"}
   > SpawnAnimationChannel = /Script/ControlRig.RigUnit_HierarchyAddAnimationChannelBool(
    LimitsEnabled=()
    MaximumValue=True
    MinimumValue=False
    Name=Movable Pivot Vis
    InitialValue=@n1.Proxy Vis
    Parent=RigVMVariableNode() {variable=Driven Control}.Value
   ) {
    is=struct RigVMUnitNode
    template="SpawnAnimationChannel::Execute(in InitialValue,in MinimumValue,in MaximumValue,in LimitsEnabled,in Parent,in Name,out Item)"
   }
   > SetControlDrivenList = /Script/ControlRig.RigUnit_SetControlDrivenList(Driven=((Type=None,Name="None")), Control=@SpawnControl.Item.Name, Driven.0=RigVMVariableNode() {variable=Driven Control}.Value) {is=struct RigVMUnitNode, template="SetControlDrivenList::Execute(in Control,in Driven)"}
   Concat = &RigVMFunction_NameConcat(B=" Bfr", A=RigVMVariableNode() {variable=Proxy Name}.Value) {template="Concat::Execute(in A,in B,out Result)"}
   > n4 = &RigUnit_HierarchyAddNull(
    Space=GlobalSpace
    Transform=(Rotation=(X=0,Y=0,Z=0,W=1),Translation=(X=0,Y=0,Z=0),Scale3D=(X=1,Y=1,Z=1))
    Name=@Concat.Result
    Parent=@SpawnControl.Item
   )
   > RigVMDispatchNode(Name=IsSet, NameSpace=Self, Value=False, Item=@n4.Item) {template="DISPATCH_RigDispatch_SetMetadata(in Item,in Name,in NameSpace,in Value,out Success)"}
   > RigVMFunctionReturnNode(Proxy Control=@SpawnControl.Item, Proxy Buffer=@n4.Item, Proxy Vis Channel=@SpawnAnimationChannel.Item)
  ]}
 ContainedGraph {nodes [
   n1 = RigVMFunctionEntryNode()
   SequenceAggregate = &RigVMFunction_Sequence(ExecuteContext=@n1.ExecuteContext)
   GetAnimationChannelFromItem = /Script/ControlRig.RigUnit_GetBoolAnimationChannelFromItem(bInitial=False, Item=RigVMVariableNode() {variable=Proxy Vis Channel}.Value) {is=struct RigVMUnitNode, template="GetAnimationChannelFromItem::Execute(out Value,in Item,in bInitial)"}
   SetControlVisibility = /Script/ControlRig.RigUnit_SetControlVisibility(bVisible=@GetAnimationChannelFromItem.Value, ExecutePin=@SequenceAggregate.A, Item=RigVMVariableNode() {variable=Proxy Control}.Value) {is=struct RigVMUnitNode, template="SetControlVisibility::Execute(in Item,in Pattern,in bVisible)"}
   n2 = &RigVMFunction_ControlFlowBranch(ExecuteContext=@SetControlVisibility.ExecutePin, Condition=@GetAnimationChannelFromItem.Value)
   RigVMFunction_Sequence = &RigVMFunction_Sequence(ExecuteContext=@n2.True)
   n3 = /Script/ControlRig.RigUnit_IsInteracting() {is=struct RigVMUnitNode}
   n4 = RigVMDispatchNode(Array=@n3.Items, Element=RigVMVariableNode() {variable=Proxy Control}.Value) {template="DISPATCH_RigVMDispatch_ArrayFind(in Array,in Element,out Index,out Success)"}
   n5 = &RigVMFunction_ControlFlowBranch(Condition=&RigVMFunction_MathBoolAnd(A=@n4.Success, B=@n3.bIsRotating).Result, ExecuteContext=@RigVMFunction_Sequence.B)
   RigVMDispatchNode(Name=IsSet, NameSpace=Self, Value=False, ExecuteContext=@n5.False, Item=RigVMVariableNode() {variable=Proxy Buffer}.Value) {template="DISPATCH_RigDispatch_SetMetadata(in Item,in Name,in NameSpace,in Value,out Success)"}
   n6 = &RigUnit_GetTransform(bInitial=False, Item=RigVMVariableNode() {variable=Proxy Snap To}.Value)
   > &RigUnit_SetTransform(
    Value=(Rotation=(X=0,Y=0,Z=0,W=1),Translation=(X=0,Y=0,Z=0),Scale3D=(X=1,Y=1,Z=1))
    Weight=1
    bInitial=False
    Value.Rotation=&RigUnit_GetTransform(bInitial=False, Item=RigVMVariableNode() {variable=Driven Control}.Value).Transform.Rotation
    Value.Scale3D=@n6.Transform.Scale3D
    Value.Translation=@n6.Transform.Translation
    Item=&RigUnit_HierarchyGetParent(bDefaultParent=True, Child=RigVMVariableNode() {variable=Proxy Control}.Value).Parent
   )
   > Set Transform_1 = &RigUnit_SetRotation(Value=(X=0,Y=0,Z=0,W=1), Weight=1, bInitial=False, Item=RigVMVariableNode() {variable=Proxy Control}.Value)
   SequenceAggregate_5_1 = &RigVMFunction_Sequence(ExecuteContext=@n5.True)
   RigVMFunction_MathBoolNot = &RigVMFunction_MathBoolNot(Value=RigVMDispatchNode(Default=False, Name=IsSet, NameSpace=Self, Item=RigVMVariableNode() {variable=Proxy Buffer}.Value) {template="DISPATCH_RigDispatch_GetMetadata(in Item,in Name,in NameSpace,in Default,out Value,out Found)"}.Value)
   n7 = &RigVMFunction_ControlFlowBranch(Condition=@RigVMFunction_MathBoolNot.Result, ExecuteContext=@SequenceAggregate_5_1.A)
   n8 = RigVMVariableNode() {variable=Proxy Buffer}
   &RigUnit_SetTransform(
    Weight=1
    bInitial=False
    ExecutePin=@n7.True
    Value=&RigUnit_GetTransform(Item=RigVMVariableNode() {variable=Driven Control}.Value).Transform
    Item=@n8.Value
   )
   > RigVMDispatchNode(Name=IsSet, NameSpace=Self, Value=true, Item=@n8.Value) {template="DISPATCH_RigDispatch_SetMetadata(in Item,in Name,in NameSpace,in Value,out Success)"}
   &RigUnit_SetTransform(
    Weight=1
    bInitial=False
    Value=&RigUnit_GetTransform(Item=RigVMVariableNode() {variable=Proxy Buffer}.Value).Transform
    ExecutePin=@SequenceAggregate_5_1.B
    Item=RigVMVariableNode() {variable=Driven Control}.Value
   )
   RigVMFunction_MathBoolAnd = &RigVMFunction_MathBoolAnd(B=@n4.Success, A=@n3.bIsTranslating)
   RigVMFunction_ControlFlowBranch = &RigVMFunction_ControlFlowBranch(Condition=@RigVMFunction_MathBoolAnd.Result, ExecuteContext=@RigVMFunction_Sequence.A)
   n9 = RigVMVariableNode() {variable=Proxy Buffer}
   &RigUnit_SetTransform(
    Weight=1
    bInitial=False
    Value=&RigUnit_GetTransform(Item=RigVMVariableNode() {variable=Driven Control}.Value).Transform
    ExecutePin=@RigVMFunction_ControlFlowBranch.True
    Item=@n9.Value
   )
   > RigVMDispatchNode(Name=IsSet, NameSpace=Self, Value=true, Item=@n9.Value) {template="DISPATCH_RigDispatch_SetMetadata(in Item,in Name,in NameSpace,in Value,out Success)"}
   RigVMFunctionReturnNode(ExecuteContext=@SequenceAggregate.B)
  ]}
 Multiply_1_2_ContainedGraph {nodes [
   n1 = RigVMFunctionEntryNode()
   RigVMFunctionReturnNode(Result=&RigVMFunction_MathDoubleMul(B=@n1.C, A=&RigVMFunction_MathDoubleMul(A=@n1.A, B=@n1.B).Result).Result)
  ]}
 New Function_ContainedGraph {nodes [
   n1 = RigVMFunctionEntryNode()
   Replace = &RigVMFunction_NameReplace(New=None, Name=@n1.Item.Name, Old=@n1.Replace Suffix) {template="Replace::Execute(in Name,in Old,in New,out Result)"}
   Concat = &RigVMFunction_NameConcat(A=@Replace.Result, B=RigVMVariableNode() {variable=New Suffix}.Value) {template="Concat::Execute(in A,in B,out Result)"}
   > n2 = &RigUnit_HierarchyAddNull(
    Space=GlobalSpace
    Transform=&RigUnit_GetTransform(bInitial=true, Item=RigVMVariableNode() {variable=Item}.Value).Transform
    Parent=RigVMVariableNode() {variable=Item}.Value
    Name=@Concat.Result
   )
   > RigVMFunctionReturnNode(Null=@n2.Item)
  ]}
 New Function_ContainedGraph {nodes [
   RigVMFunctionEntryNode()
   > RigVMFunction_Sequence = RigVMAggregateNode()
   For_Each = RigVMDispatchNode(ExecuteContext=@RigVMFunction_Sequence.A, Array=RigVMVariableNode() {variable=Controls}.Value) {template="DISPATCH_RigVMDispatch_ArrayIterator(in Array,out Element,out Index,out Count,out Ratio)"}
   > RigVMVariableNode(Value=@For_Each.Element) {variable=loop item}
   > RigVMFunction_Sequence_1 = RigVMAggregateNode()
   n1 = &RigVMFunction_ControlFlowBranch(Condition=RigVMVariableNode() {variable=Flip}.Value, ExecuteContext=@RigVMFunction_Sequence_1.B)
   Multiply = &RigVMFunction_MathDoubleMul(B=-1, A=RigVMVariableNode() {variable=position}.Value)
   RigVMVariableNode(Value=@Multiply.Result, ExecuteContext=@n1.True) {variable=position}
   n2 = RigVMVariableNode() {variable=position}
   MathFloatEquals = RigVMDispatchNode(B=0, A=@n2.Value) {template="DISPATCH_RigVMDispatch_CoreEquals(in A,in B,out Result)"}
   n3 = &RigVMFunction_ControlFlowBranch(
    Condition=&RigVMFunction_MathBoolOr(
     B=@MathFloatEquals.Result
     A=/Script/RigVM.RigVMFunction_MathFloatIsNearlyZero(Tolerance=0.001, Value=@n2.Value) {is=struct RigVMUnitNode, template="IsNearlyZero::Execute(in Value,in Tolerance,out Result)"}.Result
    ).Result
    ExecuteContext=@RigVMFunction_Sequence_1.C
   )
   Greater = /Script/RigVM.RigVMFunction_MathDoubleGreater(B=0, A=RigVMVariableNode() {variable=position}.Value) {is=struct RigVMUnitNode, template="Greater::Execute(in A,in B,out Result)"}
   n4 = &RigVMFunction_ControlFlowBranch(ExecuteContext=@n3.False, Condition=@Greater.Result)
   Less = /Script/RigVM.RigVMFunction_MathDoubleLess(B=0, A=RigVMVariableNode() {variable=position}.Value) {is=struct RigVMUnitNode, template="Less::Execute(in A,in B,out Result)"}
   n5 = &RigVMFunction_ControlFlowBranch(ExecuteContext=@n4.False, Condition=@Less.Result)
   /Script/ControlRig.RigUnit_SetControlColor(ExecutePin=@n5.True, Color=RigVMVariableNode() {variable=Right Color}.Value, Control=RigVMVariableNode() {variable=loop item}.Value.Name) {is=struct RigVMUnitNode, template="SetControlColor::Execute(in Control,in Color)"}
   &RigUnit_SetControlColor(ExecutePin=@n4.True, Color=RigVMVariableNode() {variable=Left Color}.Value, Control=RigVMVariableNode() {variable=loop item}.Value.Name) {template="SetControlColor::Execute(in Control,in Color)"}
   SetControlColor = &RigUnit_SetControlColor(ExecutePin=@n3.True, Color=RigVMVariableNode() {variable=Center Color}.Value, Control=RigVMVariableNode() {variable=loop item}.Value.Name) {template="SetControlColor::Execute(in Control,in Color)"}
   n6 = RigVMDispatchNode(Value=RigVMVariableNode() {variable=Axis}.Value) {template="DISPATCH_RigVMDispatch_CastEnumToInt(in Value,out Result)"}
   RigVMFunction_ControlFlowBranch = &RigVMFunction_ControlFlowBranch(
    Condition=RigVMDispatchNode(B=0, A=@n6.Result) {template="DISPATCH_RigVMDispatch_CoreEquals(in A,in B,out Result)"}.Result
    ExecuteContext=@RigVMFunction_Sequence_1.A
   )
   n7 = &RigVMFunction_ControlFlowBranch(
    Condition=RigVMDispatchNode(B=1, A=@n6.Result) {template="DISPATCH_RigVMDispatch_CoreEquals(in A,in B,out Result)"}.Result
    ExecuteContext=@RigVMFunction_ControlFlowBranch.False
   )
   n8 = &RigVMFunction_ControlFlowBranch(
    Condition=RigVMDispatchNode(B=2, A=@n6.Result) {template="DISPATCH_RigVMDispatch_CoreEquals(in A,in B,out Result)"}.Result
    ExecuteContext=@n7.False
   )
   n9 = &RigUnit_GetTransform(bInitial=False, Item=RigVMVariableNode() {variable=loop item}.Value)
   RigVMVariableNode(Value=@n9.Transform.Translation.Z, ExecuteContext=@n8.True) {variable=position}
   RigVMVariableNode(Value=@n9.Transform.Translation.Y, ExecuteContext=@n7.True) {variable=position}
   RigVMVariableNode(Value=@n9.Transform.Translation.X, ExecuteContext=@RigVMFunction_ControlFlowBranch.True) {variable=position}
   RigVMFunctionReturnNode(ExecuteContext=@RigVMFunction_Sequence.B)
  ]}
 New Function_ContainedGraph {nodes [
   n1 = RigVMFunctionEntryNode()
   > For_Each = RigVMDispatchNode(Array=@n1.Controls) {template="DISPATCH_RigVMDispatch_ArrayIterator(in Array,out Element,out Index,out Count,out Ratio)"}
   n2 = /Script/ControlRig.RigUnit_GetShapeTransform(Control=@For_Each.Element.Name) {is=struct RigVMUnitNode}
   Multiply_1 = RigVMAggregateNode(B=@n1.Scale, A=@n1.Global Scale, C=@n1.Auto Scale Factor)
   RigVMFunction_MathVectorFromDouble = /Script/RigVM.RigVMFunction_MathVectorFromDouble(Value=@Multiply_1.Result) {is=struct RigVMUnitNode, template="Cast::Execute(in Value,out Result)"}
   Multiply = /Script/RigVM.RigVMFunction_MathVectorMul(A=@n2.Transform.Scale3D, B=@RigVMFunction_MathVectorFromDouble.Result) {is=struct RigVMUnitNode, template="Multiply::Execute(in A,in B,out Result)"}
   > /Script/ControlRig.RigUnit_SetShapeTransform(
    Transform=(Rotation=(X=0,Y=0,Z=0,W=1),Translation=(X=0,Y=0,Z=0),Scale3D=(X=1,Y=1,Z=1))
    Transform.Rotation=@n2.Transform.Rotation
    Transform.Scale3D=@Multiply.Result
    Transform.Translation=&RigVMFunction_MathVectorMul(A=@n2.Transform.Translation, B=@RigVMFunction_MathVectorFromDouble.Result) {template="Multiply::Execute(in A,in B,out Result)"}.Result
    Control=@For_Each.Element.Name
   ) {is=struct RigVMUnitNode}
   RigVMFunctionReturnNode(ExecuteContext=@For_Each.Completed)
  ]}
 New Function_ContainedGraph {nodes [
   n1 = RigVMFunctionEntryNode()
   RigVMFunction_Sequence = &RigVMFunction_Sequence(ExecuteContext=@n1.ExecuteContext)
   RigVMDispatchNode(ExecuteContext=@RigVMFunction_Sequence.A, Array=RigVMVariableNode() {variable=parents_var}.Value) {template="DISPATCH_RigVMDispatch_ArrayReset(io Array)"}
   > For_Each = RigVMDispatchNode(Array=RigVMVariableNode() {variable=Items}.Value) {template="DISPATCH_RigVMDispatch_ArrayIterator(in Array,out Element,out Index,out Count,out Ratio)"}
   > Add = RigVMDispatchNode(
    Element=&RigUnit_HierarchyGetParent(bDefaultParent=True, Child=@For_Each.Element).Parent
    Array=RigVMVariableNode() {variable=parents_var}.Value
   ) {template="DISPATCH_RigVMDispatch_ArrayAdd(io Array,in Element,out Index)"}
   RigVMFunctionReturnNode(ExecuteContext=@RigVMFunction_Sequence.B, Parents=RigVMVariableNode() {variable=parents_var}.Value)
  ]}
 New Function_ContainedGraph {nodes [
   n1 = RigVMFunctionEntryNode()
   RigVMFunction_Sequence = &RigVMFunction_Sequence(ExecuteContext=@n1.ExecuteContext)
   Project Middle Bone to IK Plane = RigVMFunctionReferenceNode(Debug=false, Secondary Axis=(X=0,Y=1,Z=0), ExecuteContext=@RigVMFunction_Sequence.A, End=@n1.End, Mid=@n1.Mid, Start=@n1.Start)
   n2 = RigVMVariableNode() {variable=Axis}
   Negate = /Script/RigVM.RigVMFunction_MathVectorNegate(Value=@n2.Value) {is=struct RigVMUnitNode, template="Negate::Execute(in Value,out Result)"}
   > RigVMVariableNode(Value=RigVMDispatchNode(True=@Negate.Result, Condition=RigVMVariableNode() {variable=Negative Side}.Value, False=@n2.Value) {template="DISPATCH_RigVMDispatch_If(in Condition,in True,in False,out Result)"}.Result) {variable=axis var}
   RigVMFunction_MathTransformArrayToSRT = /Script/RigVM.RigVMFunction_MathTransformArrayToSRT(Transforms=@"Project Middle Bone to IK Plane".outputs) {is=struct RigVMUnitNode}
   n3 = RigVMDispatchNode(Index=0, Array=@RigVMFunction_MathTransformArrayToSRT.Translations) {template="DISPATCH_RigVMDispatch_ArrayGetAtIndex(in Array,in Index,out Element)"}
   n4 = RigVMDispatchNode(Index=1, Array=@RigVMFunction_MathTransformArrayToSRT.Translations) {template="DISPATCH_RigVMDispatch_ArrayGetAtIndex(in Array,in Index,out Element)"}
   Subtract = /Script/RigVM.RigVMFunction_MathVectorSub(B=@n3.Element, A=@n4.Element) {is=struct RigVMUnitNode, template="Subtract::Execute(in A,in B,out Result)"}
   n5 = RigVMDispatchNode(Index=2, Array=@RigVMFunction_MathTransformArrayToSRT.Translations) {template="DISPATCH_RigVMDispatch_ArrayGetAtIndex(in Array,in Index,out Element)"}
   RigVMFunction_MathVectorCross = /Script/RigVM.RigVMFunction_MathVectorCross(
    A=@Subtract.Result
    B=&RigVMFunction_MathVectorSub(B=@n4.Element, A=@n5.Element) {template="Subtract::Execute(in A,in B,out Result)"}.Result
   ) {is=struct RigVMUnitNode, template="Cross::Execute(in A,in B,out Result)"}
   n6 = /Script/ControlRig.RigUnit_AimBoneMath(
    DebugSettings=(bEnabled=False,Scale=10,WorldOffset=(Rotation=(X=0,Y=0,Z=0,W=1),Translation=(X=0,Y=0,Z=0),Scale3D=(X=1,Y=1,Z=1)))
    InputTransform=(Rotation=(X=0,Y=0,Z=0,W=1),Translation=(X=0,Y=0,Z=0),Scale3D=(X=1,Y=1,Z=1))
    Primary=(Weight=1,Axis=(X=1,Y=0,Z=0),Target=(X=1,Y=0,Z=0),Kind=Location,Space=(Type=Bone,Name="None"))
    Secondary=(Weight=1,Axis=(X=0,Y=0,Z=1),Target=(X=1,Y=0,Z=0),Kind=Direction,Space=(Type=Bone,Name="None"))
    Weight=1
    InputTransform.Translation=@n3.Element
    Secondary.Target=@RigVMFunction_MathVectorCross.Result
    Primary.Axis=RigVMVariableNode() {variable=axis var}.Value
    Primary.Target=@n4.Element
   ) {is=struct RigVMUnitNode}
   Concat = &RigVMFunction_NameConcat(B=Bone A, A=RigVMVariableNode() {variable=Name}.Value) {template="Concat::Execute(in A,in B,out Result)"}
   > n7 = /Script/ControlRig.RigUnit_HierarchyAddBone(
    Space=GlobalSpace
    Transform=@n6.Result
    Name=@Concat.Result
    Parent=&RigUnit_HierarchyGetParent(bDefaultParent=True, Child=@n1.Start).Parent
   ) {is=struct RigVMUnitNode}
   n8 = &RigUnit_AimBoneMath(
    DebugSettings=(bEnabled=False,Scale=10,WorldOffset=(Rotation=(X=0,Y=0,Z=0,W=1),Translation=(X=0,Y=0,Z=0),Scale3D=(X=1,Y=1,Z=1)))
    InputTransform=(Rotation=(X=0,Y=0,Z=0,W=1),Translation=(X=0,Y=0,Z=0),Scale3D=(X=1,Y=1,Z=1))
    Primary=(Weight=1,Axis=(X=1,Y=0,Z=0),Target=(X=1,Y=0,Z=0),Kind=Location,Space=(Type=Bone,Name="None"))
    Secondary=(Weight=1,Axis=(X=0,Y=0,Z=1),Target=(X=1,Y=0,Z=0),Kind=Direction,Space=(Type=Bone,Name="None"))
    Weight=1
    Primary.Axis=RigVMVariableNode() {variable=axis var}.Value
    Primary.Target=@n5.Element
    Secondary.Target=@RigVMFunction_MathVectorCross.Result
    InputTransform.Translation=@n4.Element
   )
   > n9 = &RigUnit_HierarchyAddBone(
    Space=GlobalSpace
    Transform=@n8.Result
    Name=&RigVMFunction_NameConcat(B=Bone B, A=RigVMVariableNode() {variable=Name}.Value) {template="Concat::Execute(in A,in B,out Result)"}.Result
    Parent=@n7.Item
   )
   n10 = &RigUnit_GetTransform(bInitial=False, Item=RigVMVariableNode() {variable=End}.Value)
   RigVMFunction_MathTransformTransformVector = /Script/RigVM.RigVMFunction_MathTransformTransformVector(
    Location=(X=2,Y=0,Z=0)
    Transform=@n10.Transform
    Location.X=RigVMDispatchNode(False=2, True=-2, Condition=RigVMVariableNode() {variable=Negative Side}.Value) {template="DISPATCH_RigVMDispatch_If(in Condition,in True,in False,out Result)"}.Result
   ) {is=struct RigVMUnitNode}
   n11 = &RigUnit_AimBoneMath(
    DebugSettings=(bEnabled=False,Scale=10,WorldOffset=(Rotation=(X=0,Y=0,Z=0,W=1),Translation=(X=0,Y=0,Z=0),Scale3D=(X=1,Y=1,Z=1)))
    InputTransform=(Rotation=(X=0,Y=0,Z=0,W=1),Translation=(X=0,Y=0,Z=0),Scale3D=(X=1,Y=1,Z=1))
    Primary=(Weight=1,Axis=(X=1,Y=0,Z=0),Target=(X=1,Y=0,Z=0),Kind=Location,Space=(Type=Bone,Name="None"))
    Secondary=(Weight=1,Axis=(X=0,Y=-1,Z=0),Target=(X=1,Y=0,Z=0),Kind=Location,Space=(Type=Bone,Name="None"))
    Weight=1
    Primary.Target=@RigVMFunction_MathTransformTransformVector.Result
    Secondary.Target=&RigVMFunction_MathTransformTransformVector(Location=(X=0,Y=0,Z=2), Transform=@n10.Transform).Result
    Primary.Axis=RigVMVariableNode() {variable=axis var}.Value
    InputTransform.Translation=@n5.Element
   )
   > n12 = &RigUnit_HierarchyAddBone(
    Space=GlobalSpace
    Transform=@n11.Result
    Name=&RigVMFunction_NameConcat(B=Bone C, A=RigVMVariableNode() {variable=Name}.Value) {template="Concat::Execute(in A,in B,out Result)"}.Result
    Parent=@n9.Item
   )
   > RigVMVariableNode(Value=((Type=None,Name="None"),(Type=None,Name="None"),(Type=None,Name="None")), Value.2=@n12.Item, Value.1=@n9.Item, Value.0=@n7.Item) {variable=virtual bones}
   RigVMFunctionReturnNode(Bones=RigVMVariableNode() {variable=virtual bones}.Value, ExecuteContext=@RigVMFunction_Sequence.B)
   RigVMRerouteNode(Value=@n4.Element)
  ]}
 New Function_ContainedGraph {nodes [
   n1 = RigVMFunctionEntryNode()
   Negate = &RigVMFunction_MathVectorNegate(Value=@n1.Axis) {template="Negate::Execute(in Value,out Result)"}
   Multiply = &RigVMFunction_MathTransformMul(B=(Rotation=(X=0,Y=0,Z=0,W=1),Translation=(X=0,Y=0,Z=0),Scale3D=(X=1,Y=1,Z=1)), A=@n1.Transform, B.Scale3D=@Negate.Result) {template="Multiply::Execute(in A,in B,out Result)"}
   > RigVMFunctionReturnNode(Mirror Transform=@Multiply.Result)
  ]}
 New Function_ContainedGraph {nodes [
   n1 = RigVMFunctionEntryNode()
   RigVMFunction_Sequence = &RigVMFunction_Sequence(ExecuteContext=@n1.ExecuteContext)
   RigVMDispatchNode(ExecuteContext=@RigVMFunction_Sequence.A, Array=RigVMVariableNode() {variable=outputs}.Value) {template="DISPATCH_RigVMDispatch_ArrayReset(io Array)"}
   > For_Each_3 = RigVMDispatchNode(Array=/Script/ControlRig.RigUnit_CollectionChildrenArray(bDefaultChildren=True, bIncludeParent=@n1.Include Item, Parent=@n1.Item, bRecursive=@n1.Recursive, TypeToSearch=@n1.Type to Search) {is=struct RigVMUnitNode}.Items) {template="DISPATCH_RigVMDispatch_ArrayIterator(in Array,out Element,out Index,out Count,out Ratio)"}
   > RigVMVariableNode(Value=true) {variable=result}
   RigVMFunction_StringSplit = /Script/RigVM.RigVMFunction_StringSplit(Value=RigVMVariableNode() {variable=Search String}.Value, Separator=RigVMVariableNode() {variable=Seperator}.Value) {is=struct RigVMUnitNode}
   > For_Each_4 = RigVMDispatchNode(Array=@RigVMFunction_StringSplit.Result) {template="DISPATCH_RigVMDispatch_ArrayIterator(in Array,out Element,out Index,out Count,out Ratio)"}
   To_String = RigVMDispatchNode(Value=@For_Each_3.Element.Name) {template="DISPATCH_RigDispatch_ToString(in Value,out Result)"}
   Contains = /Script/RigVM.RigVMFunction_StringContains(Search=@For_Each_4.Element, Name=@To_String.Result) {is=struct RigVMUnitNode, template="Contains::Execute(in Name,in Search,out Result)"}
   RigVMFunction_ControlFlowBranch = &RigVMFunction_ControlFlowBranch(Condition=@Contains.Result, ExecuteContext=@For_Each_4.ExecuteContext)
   RigVMVariableNode(ExecuteContext=@RigVMFunction_ControlFlowBranch.False) {variable=result}
   n2 = &RigVMFunction_ControlFlowBranch(ExecuteContext=@For_Each_4.Completed, Condition=RigVMVariableNode() {variable=result}.Value)
   Add = RigVMDispatchNode(ExecuteContext=@n2.True, Array=RigVMVariableNode() {variable=outputs}.Value, Element=@For_Each_3.Element) {template="DISPATCH_RigVMDispatch_ArrayAdd(io Array,in Element,out Index)"}
   RigVMFunctionReturnNode(ExecuteContext=@RigVMFunction_Sequence.B, Results=RigVMVariableNode() {variable=outputs}.Value)
  ]}
 New Function_ContainedGraph {nodes [
   n1 = RigVMFunctionEntryNode()
   > RigVMFunction_Sequence = RigVMAggregateNode()
   For_Each = RigVMDispatchNode(ExecuteContext=@RigVMFunction_Sequence.A, Array=@n1.On Controls) {template="DISPATCH_RigVMDispatch_ArrayIterator(in Array,out Element,out Index,out Count,out Ratio)"}
   > SetControlVisibility = &RigUnit_SetControlVisibility(Item=@For_Each.Element, bVisible=RigVMVariableNode() {variable=Switch}.Value) {template="SetControlVisibility::Execute(in Item,in Pattern,in bVisible)"}
   For_Each_1 = RigVMDispatchNode(ExecuteContext=@RigVMFunction_Sequence.B, Array=@n1.Off Controls) {template="DISPATCH_RigVMDispatch_ArrayIterator(in Array,out Element,out Index,out Count,out Ratio)"}
   RigVMFunction_MathBoolNot = &RigVMFunction_MathBoolNot(Value=RigVMVariableNode() {variable=Switch}.Value)
   > &RigUnit_SetControlVisibility(Item=@For_Each_1.Element, bVisible=@RigVMFunction_MathBoolNot.Result) {template="SetControlVisibility::Execute(in Item,in Pattern,in bVisible)"}
   RigVMFunctionReturnNode(ExecuteContext=@RigVMFunction_Sequence.C)
  ]}
 New Function_ContainedGraph {nodes [
   RigVMFunctionEntryNode()
   From_String_2 = RigVMDispatchNode(String=Global Center Control Color) {template="DISPATCH_RigDispatch_FromString(in String,out Result)"}
   Get_Module_Metadata = RigVMDispatchNode(Default=(R=1,G=1,B=0,A=1), NameSpace=Root, Name=@From_String_2.Result) {template="DISPATCH_RigDispatch_GetModuleMetadata(in Name,in NameSpace,in Default,out Value,out Found)"}
   n1 = /Script/ControlRig.RigUnit_GetModuleName() {is=struct RigVMUnitNode}
   EndsWith = /Script/RigVM.RigVMFunction_StringEndsWith(Ending=" R", Name=@n1.Module) {is=struct RigVMUnitNode, template="EndsWith::Execute(in Name,in Ending,out Result)"}
   From_String_1 = RigVMDispatchNode(String=Global Left Control Color) {template="DISPATCH_RigDispatch_FromString(in String,out Result)"}
   Get_Module_Metadata_2 = RigVMDispatchNode(Default=(R=1,G=1,B=0,A=1), NameSpace=Root, Name=@From_String_1.Result) {template="DISPATCH_RigDispatch_GetModuleMetadata(in Name,in NameSpace,in Default,out Value,out Found)"}
   From_String = RigVMDispatchNode(String=Global Right Control Color) {template="DISPATCH_RigDispatch_FromString(in String,out Result)"}
   Get_Module_Metadata_2_1 = RigVMDispatchNode(Default=(R=1,G=1,B=0,A=1), NameSpace=Root, Name=@From_String.Result) {template="DISPATCH_RigDispatch_GetModuleMetadata(in Name,in NameSpace,in Default,out Value,out Found)"}
   n2 = RigVMDispatchNode(Condition=@EndsWith.Result, False=@Get_Module_Metadata_2.Value, True=@Get_Module_Metadata_2_1.Value) {template="DISPATCH_RigVMDispatch_If(in Condition,in True,in False,out Result)"}
   RigVMFunction_MathBoolOr = &RigVMFunction_MathBoolOr(
    A=@EndsWith.Result
    B=&RigVMFunction_StringEndsWith(Ending=" L", Name=@n1.Module) {template="EndsWith::Execute(in Name,in Ending,out Result)"}.Result
   )
   n3 = RigVMVariableNode() {variable=Color Override}
   RigVMFunction_MathBoolAnd = &RigVMFunction_MathBoolAnd(
    A=RigVMDispatchNode(B=@n3.Value.G, A=@n3.Value.R) {template="DISPATCH_RigVMDispatch_CoreEquals(in A,in B,out Result)"}.Result
    B=RigVMDispatchNode(B=@n3.Value.B, A=@n3.Value.G) {template="DISPATCH_RigVMDispatch_CoreEquals(in A,in B,out Result)"}.Result
   )
   > RigVMFunctionReturnNode(Result=RigVMDispatchNode(
     True=RigVMDispatchNode(False=@Get_Module_Metadata.Value, True=@n2.Result, Condition=@RigVMFunction_MathBoolOr.Result) {template="DISPATCH_RigVMDispatch_If(in Condition,in True,in False,out Result)"}.Result
     Condition=@RigVMFunction_MathBoolAnd.Result
     False=@n3.Value
    ) {template="DISPATCH_RigVMDispatch_If(in Condition,in True,in False,out Result)"}.Result)
  ]}
 New Function_ContainedGraph {nodes [
   n1 = RigVMFunctionEntryNode()
   n2 = RigVMVariableNode() {variable=PV Control}
   n3 = &RigUnit_GetTransform(bInitial=@n1.Initial, Item=@n1.PV Root Space)
   n4 = RigVMVariableNode() {variable=IK Control}
   n5 = &RigUnit_AimBoneMath(
    DebugSettings=(Scale=10,WorldOffset=(Rotation=(X=0,Y=0,Z=0,W=1),Translation=(X=0,Y=0,Z=0),Scale3D=(X=1,Y=1,Z=1)))
    Primary=(Weight=1,Axis=(X=1,Y=0,Z=0),Target=(X=0,Y=0,Z=0),Kind=Location,Space=(Type=Bone))
    Secondary=(Weight=1,Axis=(X=0,Y=0,Z=1),Target=(X=1,Y=0,Z=0),Kind=Direction,Space=(Type=Bone))
    Weight=1
    InputTransform=@n3.Transform
    Primary.Space=@n4.Value
    Secondary.Space=@n4.Value
   )
   n6 = &RigUnit_AimBoneMath(
    DebugSettings=(Scale=10,WorldOffset=(Rotation=(X=0,Y=0,Z=0,W=1),Translation=(X=0,Y=0,Z=0),Scale3D=(X=1,Y=1,Z=1)))
    Primary=(Weight=1,Axis=(X=1,Y=0,Z=0),Target=(X=0,Y=0,Z=0),Kind=Location,Space=(Type=Bone))
    Secondary=(Weight=1,Axis=(X=0,Y=1,Z=0),Target=(X=0,Y=1,Z=0),Kind=Direction,Space=(Type=Bone))
    Weight=1
    Secondary.Space=@n1.PV Root Space
    InputTransform=@n3.Transform
    Primary.Space=RigVMVariableNode() {variable=IK Control}.Value
   )
   Interpolate = /Script/RigVM.RigVMFunction_MathTransformLerp(B=@n5.Result, A=@n6.Result, T=RigVMVariableNode() {variable=PV Twist Follow}.Value) {is=struct RigVMUnitNode, template="Interpolate::Execute(in A,in B,in T,out Result)"}
   > &RigUnit_SetTransform(
    Weight=1
    bInitial=@n1.Initial
    Item=&RigUnit_HierarchyGetParent(bDefaultParent=True, Child=&RigUnit_HierarchyGetParent(bDefaultParent=True, Child=@n2.Value).Parent).Parent
    Value=@Interpolate.Result
   )
   > Set Transform_3 = &RigUnit_SetRotation(
    Space=GlobalSpace
    Weight=1
    bInitial=False
    Value=&RigUnit_GetTransform(bInitial=False, Item=RigVMVariableNode() {variable=PV Orient Space}.Value).Transform.Rotation
    Item=&RigUnit_HierarchyGetParent(bDefaultParent=True, Child=RigVMVariableNode() {variable=PV Control}.Value).Parent
   )
   > RigVMFunctionReturnNode()
  ]}
 New Function_ContainedGraph {nodes [
   n1 = RigVMFunctionEntryNode()
   n2 = RigVMVariableNode() {variable=IK Control}
   n3 = &RigUnit_AimBoneMath(
    DebugSettings=(Scale=10,WorldOffset=(Rotation=(X=0,Y=0,Z=0,W=1),Translation=(X=0,Y=0,Z=0),Scale3D=(X=1,Y=1,Z=1)))
    Primary=(Weight=1,Axis=(X=1,Y=0,Z=0),Target=(X=0,Y=0,Z=0),Kind=Location,Space=(Type=Bone))
    Secondary=(Weight=1,Axis=(X=0,Y=0,Z=1),Target=(X=1,Y=0,Z=0),Kind=Direction,Space=(Type=Bone))
    Weight=1
    InputTransform=&RigUnit_GetTransform(bInitial=@n1.Initial, Item=@n1.PV Root Space).Transform
    Primary.Space=@n2.Value
    Secondary.Space=@n2.Value
   )
   > n4 = &RigUnit_HierarchyAddNull(
    Space=GlobalSpace
    Transform=@n3.Result
    Name=&RigVMFunction_NameConcat(B=RootSpaceNull, A=RigVMVariableNode() {variable=Name}.Value) {template="Concat::Execute(in A,in B,out Result)"}.Result
    Parent=&RigUnit_HierarchyGetParent(bDefaultParent=True, Child=@n1.PV Root Space).Parent
   )
   Concat = &RigVMFunction_NameConcat(B=RootNull, A=RigVMVariableNode() {variable=Name}.Value) {template="Concat::Execute(in A,in B,out Result)"}
   > n5 = &RigUnit_HierarchyAddNull(Space=GlobalSpace, Transform=@n3.Result, Name=@Concat.Result, Parent=RigVMVariableNode() {variable=Parent}.Value)
   Compute Pole Vector From Plane v01_1 = RigVMFunctionReferenceNode(
    Item A=RigVMVariableNode() {variable=Item A}.Value
    Item B=RigVMVariableNode() {variable=Item B}.Value
    Item C=RigVMVariableNode() {variable=Item C}.Value
    OffsetMultiplier=RigVMVariableNode() {variable=Offset Multiplier}.Value
    Initial=RigVMVariableNode() {variable=Initial}.Value
   )
   > n6 = &RigUnit_HierarchyAddNull(
    Space=GlobalSpace
    Transform=(Rotation=(X=0,Y=0,Z=0,W=1),Translation=(X=0,Y=0,Z=0),Scale3D=(X=1,Y=1,Z=1))
    Transform.Translation=@"Compute Pole Vector From Plane v01_1".Result.Translation
    Name=&RigVMFunction_NameConcat(B=Null, A=RigVMVariableNode() {variable=Name}.Value) {template="Concat::Execute(in A,in B,out Result)"}.Result
    Parent=@n5.Item
   )
   > n7 = /Script/ControlRig.RigUnit_HierarchyAddControlVector(
    InitialValue=(X=0,Y=0,Z=0)
    OffsetSpace=LocalSpace
    OffsetTransform=(Rotation=(X=0,Y=0,Z=0,W=1),Translation=(X=0,Y=0,Z=0),Scale3D=(X=1,Y=1,Z=1))
    Settings="(InitialSpace=LocalSpace,bIsPosition=True,Limits=(LimitX=(bMinimum=False,bMaximum=False),LimitY=(bMinimum=False,bMaximum=False),LimitZ=(bMinimum=False,bMaximum=False),MinValue=(X=0,Y=0,Z=0),MaxValue=(X=1,Y=1,Z=1),bDrawLimits=True),Shape=(bVisible=True,Name ...(truncated)"
    Parent=@n6.Item
    Name=RigVMVariableNode() {variable=Name}.Value
    Settings.Shape=RigVMVariableNode() {variable=Shape}.Value
   ) {
    is=struct RigVMUnitNode
    template="SpawnControl::Execute(in InitialValue,in Settings,in OffsetTransform,in OffsetSpace,in Parent,in Name,out Item)"
   }
   > RigVMFunctionReturnNode(Pv Root Space Null=@n4.Item, Pv Control=@n7.Item)
  ]}
 New Function_ContainedGraph {nodes [
   n1 = RigVMFunctionEntryNode()
   n2 = &RigUnit_GetTransform(bInitial=@n1.Initial, Item=@n1.Item B)
   n3 = &RigUnit_GetTransform(bInitial=@n1.Initial, Item=@n1.Item A)
   Subtract = &RigVMFunction_MathVectorSub(B=@n3.Transform.Translation, A=@n2.Transform.Translation) {template="Subtract::Execute(in A,in B,out Result)"}
   n4 = &RigVMFunction_MathVectorSub(B=@n3.Transform.Translation, A=&RigUnit_GetTransform(bInitial=@n1.Initial, Item=@n1.Item C).Transform.Translation) {template="Subtract::Execute(in A,in B,out Result)"}
   n5 = /Script/RigVM.RigVMFunction_VisualDebugVectorNoSpace(
    Color=(R=1,G=0,B=0,A=1)
    DebugDrawSettings=(DepthPriority=SDPG_Foreground,Lifetime=-1)
    Mode=Vector
    Scale=1
    Thickness=1
    bEnabled=True
    Value=&RigVMFunction_MathVectorCross(B=@Subtract.Result, A=@n4.Result) {template="Cross::Execute(in A,in B,out Result)"}.Result
   ) {is=struct RigVMUnitNode, template="VisualDebug::Execute(io Value,in bEnabled,in Mode,in Color,in Thickness,in Scale,in DebugDrawSettings)"}
   Multiply = &RigVMFunction_MathDoubleMul(
    B=@n1.OffsetMultiplier
    A=/Script/RigVM.RigVMFunction_MathVectorLength(Value=@Subtract.Result) {is=struct RigVMUnitNode, template="Length::Execute(in Value,out Result)"}.Result
   )
   Scale = /Script/RigVM.RigVMFunction_MathVectorScale(
    Value=/Script/RigVM.RigVMFunction_MathVectorSetLength(
     Length=1
     Value=&RigVMFunction_MathVectorCross(A=@n5.Value, B=@n4.Result) {template="Cross::Execute(in A,in B,out Result)"}.Result
    ) {is=struct RigVMUnitNode, template="SetLength::Execute(in Value,in Length,out Result)"}.Result
    Factor=@Multiply.Result
   ) {is=struct RigVMUnitNode, template="Scale::Execute(in Value,in Factor,out Result)"}
   Add = /Script/RigVM.RigVMFunction_MathVectorAdd(B=@n2.Transform.Translation, A=@Scale.Result) {is=struct RigVMUnitNode, template="Add::Execute(in A,in B,out Result)"}
   n6 = &RigUnit_AimBoneMath(
    DebugSettings=(Scale=10,WorldOffset=(Rotation=(X=0,Y=0,Z=0,W=1),Translation=(X=0,Y=0,Z=0),Scale3D=(X=1,Y=1,Z=1)))
    InputTransform=(Rotation=(X=0,Y=0,Z=0,W=1),Translation=(X=0,Y=0,Z=0),Scale3D=(X=1,Y=1,Z=1))
    Primary=(Weight=1,Axis=(X=1,Y=0,Z=0),Target=(X=0,Y=0,Z=1),Kind=Location,Space=(Type=Bone))
    Secondary=(Weight=1,Axis=(X=0,Y=0,Z=1),Target=(X=0,Y=0,Z=1),Kind=Direction,Space=(Type=Bone))
    Weight=1
    InputTransform.Translation=@Add.Result
    Primary.Target=&RigUnit_GetTransform(Item=RigVMVariableNode() {variable=Item B}.Value).Transform.Translation
   )
   > RigVMFunctionReturnNode(Result=@n6.Result)
   RigVMFunction_VisualDebugVectorNoSpace = &RigVMFunction_VisualDebugVectorNoSpace(
    Color=(R=1,G=0,B=0,A=1)
    DebugDrawSettings=(DepthPriority=SDPG_Foreground,Lifetime=-1)
    Mode=Vector
    Scale=1
    Thickness=1
    bEnabled=false
    Value=@Add.Result
   ) {template="VisualDebug::Execute(io Value,in bEnabled,in Mode,in Color,in Thickness,in Scale,in DebugDrawSettings)"}
  ]}
 New Function_ContainedGraph {nodes [
   n1 = RigVMFunctionEntryNode()
   n2 = &RigUnit_GetTransform(bInitial=@n1.Initial, Item=@n1.Item A)
   n3 = &RigUnit_GetTransform(bInitial=@n1.Initial, Item=@n1.Item B)
   Subtract = &RigVMFunction_MathVectorSub(B=@n2.Transform.Translation, A=@n3.Transform.Translation) {template="Subtract::Execute(in A,in B,out Result)"}
   Multiply = &RigVMFunction_MathDoubleMul(
    B=@n1.OffsetMultiplier
    A=&RigVMFunction_MathVectorLength(Value=@Subtract.Result) {template="Length::Execute(in Value,out Result)"}.Result
   )
   RerouteNode_3_MakeStruct = RigVMDispatchNode(
    Elements=(Rotation=(X=0,Y=0,Z=0,W=1),Translation=(X=0,Y=0,Z=0),Scale3D=(X=1,Y=1,Z=1))
    Elements.Translation=@n3.Transform.Translation
    Elements.Rotation=/Script/RigVM.RigVMFunction_MathQuaternionSlerp(T=0.5, A=@n2.Transform.Rotation, B=@n3.Transform.Rotation) {is=struct RigVMUnitNode, template="Interpolate::Execute(in A,in B,in T,out Result)"}.Result
   ) {template="DISPATCH_RigVMDispatch_MakeStruct(in Elements,out Struct)"}
   RigVMFunction_MathTransformTransformVector = &RigVMFunction_MathTransformTransformVector(
    Location=&RigVMFunction_MathVectorMul(B=(X=1,Y=1,Z=1), A=RigVMVariableNode() {variable=Up Vector}.Value, B.X=@Multiply.Result, B.Y=@Multiply.Result, B.Z=@Multiply.Result) {template="Multiply::Execute(in A,in B,out Result)"}.Result
    Transform=@RerouteNode_3_MakeStruct.Struct
   )
   > RigVMFunctionReturnNode(Result=@RigVMFunction_MathTransformTransformVector.Result)
  ]}
 New Function_ContainedGraph {nodes [
   n1 = RigVMFunctionEntryNode()
   RigVMFunction_Sequence = &RigVMFunction_Sequence(ExecuteContext=@n1.ExecuteContext)
   Project Middle Bone to IK Plane = RigVMFunctionReferenceNode(
    Debug=false
    ExecuteContext=@RigVMFunction_Sequence.A
    Secondary Axis=RigVMVariableNode() {variable=Secondary Axis}.Value
    End=@n1.End
    Mid=@n1.Mid
    Start=@n1.Start
   )
   RigVMFunction_MathTransformArrayToSRT = &RigVMFunction_MathTransformArrayToSRT(Transforms=@"Project Middle Bone to IK Plane".outputs)
   n2 = RigVMDispatchNode(Index=0, Array=@RigVMFunction_MathTransformArrayToSRT.Translations) {template="DISPATCH_RigVMDispatch_ArrayGetAtIndex(in Array,in Index,out Element)"}
   n3 = RigVMDispatchNode(Index=1, Array=@RigVMFunction_MathTransformArrayToSRT.Translations) {template="DISPATCH_RigVMDispatch_ArrayGetAtIndex(in Array,in Index,out Element)"}
   Subtract = &RigVMFunction_MathVectorSub(B=@n2.Element, A=@n3.Element) {template="Subtract::Execute(in A,in B,out Result)"}
   n4 = RigVMDispatchNode(Index=2, Array=@RigVMFunction_MathTransformArrayToSRT.Translations) {template="DISPATCH_RigVMDispatch_ArrayGetAtIndex(in Array,in Index,out Element)"}
   RigVMFunction_MathVectorCross = &RigVMFunction_MathVectorCross(
    A=@Subtract.Result
    B=&RigVMFunction_MathVectorSub(B=@n3.Element, A=@n4.Element) {template="Subtract::Execute(in A,in B,out Result)"}.Result
   ) {template="Cross::Execute(in A,in B,out Result)"}
   n5 = &RigUnit_AimBoneMath(
    DebugSettings=(bEnabled=False,Scale=10,WorldOffset=(Rotation=(X=0,Y=0,Z=0,W=1),Translation=(X=0,Y=0,Z=0),Scale3D=(X=1,Y=1,Z=1)))
    InputTransform=(Rotation=(X=0,Y=0,Z=0,W=1),Translation=(X=0,Y=0,Z=0),Scale3D=(X=1,Y=1,Z=1))
    Primary=(Weight=1,Axis=(X=1,Y=0,Z=0),Target=(X=1,Y=0,Z=0),Kind=Location,Space=(Type=Bone,Name="None"))
    Secondary=(Weight=1,Axis=(X=0,Y=0,Z=1),Target=(X=1,Y=0,Z=0),Kind=Direction,Space=(Type=Bone,Name="None"))
    Weight=1
    InputTransform.Translation=@n2.Element
    Secondary.Target=@RigVMFunction_MathVectorCross.Result
    Primary.Axis=RigVMVariableNode() {variable=Axis}.Value
    Primary.Target=@n3.Element
   )
   Concat = &RigVMFunction_NameConcat(B=Bone A, A=RigVMVariableNode() {variable=Name}.Value) {template="Concat::Execute(in A,in B,out Result)"}
   > n6 = &RigUnit_HierarchyAddBone(
    Space=GlobalSpace
    Transform=@n5.Result
    Name=@Concat.Result
    Parent=&RigUnit_HierarchyGetParent(bDefaultParent=True, Child=@n1.Start).Parent
   )
   n7 = &RigUnit_AimBoneMath(
    DebugSettings=(bEnabled=False,Scale=10,WorldOffset=(Rotation=(X=0,Y=0,Z=0,W=1),Translation=(X=0,Y=0,Z=0),Scale3D=(X=1,Y=1,Z=1)))
    InputTransform=(Rotation=(X=0,Y=0,Z=0,W=1),Translation=(X=0,Y=0,Z=0),Scale3D=(X=1,Y=1,Z=1))
    Primary=(Weight=1,Axis=(X=1,Y=0,Z=0),Target=(X=1,Y=0,Z=0),Kind=Location,Space=(Type=Bone,Name="None"))
    Secondary=(Weight=1,Axis=(X=0,Y=0,Z=1),Target=(X=1,Y=0,Z=0),Kind=Direction,Space=(Type=Bone,Name="None"))
    Weight=1
    Primary.Axis=RigVMVariableNode() {variable=Axis}.Value
    Primary.Target=@n4.Element
    Secondary.Target=@RigVMFunction_MathVectorCross.Result
    InputTransform.Translation=@n3.Element
   )
   > n8 = &RigUnit_HierarchyAddBone(
    Space=GlobalSpace
    Transform=@n7.Result
    Name=&RigVMFunction_NameConcat(B=Bone B, A=RigVMVariableNode() {variable=Name}.Value) {template="Concat::Execute(in A,in B,out Result)"}.Result
    Parent=@n6.Item
   )
   n9 = &RigUnit_GetTransform(bInitial=False, Item=RigVMVariableNode() {variable=End}.Value)
   RigVMFunction_MathTransformTransformVector = &RigVMFunction_MathTransformTransformVector(Transform=@n9.Transform, Location=RigVMVariableNode() {variable=Axis}.Value)
   n10 = &RigUnit_AimBoneMath(
    DebugSettings=(bEnabled=False,Scale=10,WorldOffset=(Rotation=(X=0,Y=0,Z=0,W=1),Translation=(X=0,Y=0,Z=0),Scale3D=(X=1,Y=1,Z=1)))
    InputTransform=(Rotation=(X=0,Y=0,Z=0,W=1),Translation=(X=0,Y=0,Z=0),Scale3D=(X=1,Y=1,Z=1))
    Primary=(Weight=1,Axis=(X=1,Y=0,Z=0),Target=(X=1,Y=0,Z=0),Kind=Location,Space=(Type=Bone,Name="None"))
    Secondary=(Weight=1,Axis=(X=0,Y=-1,Z=0),Target=(X=1,Y=0,Z=0),Kind=Location,Space=(Type=Bone,Name="None"))
    Weight=1
    Primary.Target=@RigVMFunction_MathTransformTransformVector.Result
    Secondary.Target=&RigVMFunction_MathTransformTransformVector(Location=(X=0,Y=0,Z=2), Transform=@n9.Transform).Result
    Primary.Axis=RigVMVariableNode() {variable=Axis}.Value
    InputTransform.Translation=@n4.Element
   )
   > n11 = &RigUnit_HierarchyAddBone(
    Space=GlobalSpace
    Transform=@n10.Result
    Name=&RigVMFunction_NameConcat(B=Bone C, A=RigVMVariableNode() {variable=Name}.Value) {template="Concat::Execute(in A,in B,out Result)"}.Result
    Parent=@n8.Item
   )
   > RigVMVariableNode(Value=((Type=None,Name="None"),(Type=None,Name="None"),(Type=None,Name="None")), Value.2=@n11.Item, Value.1=@n8.Item, Value.0=@n6.Item) {variable=virtual bones}
   RigVMFunctionReturnNode(Bones=RigVMVariableNode() {variable=virtual bones}.Value, ExecuteContext=@RigVMFunction_Sequence.B)
  ]}
 New Function_ContainedGraph {nodes [
   RigVMFunctionEntryNode()
   From_String_2 = RigVMDispatchNode(String=Global Center Control Color) {template="DISPATCH_RigDispatch_FromString(in String,out Result)"}
   Get_Module_Metadata = RigVMDispatchNode(Default=(R=1,G=1,B=0,A=1), NameSpace=Root, Name=@From_String_2.Result) {template="DISPATCH_RigDispatch_GetModuleMetadata(in Name,in NameSpace,in Default,out Value,out Found)"}
   From_String_1 = RigVMDispatchNode(String=Global Left Control Color) {template="DISPATCH_RigDispatch_FromString(in String,out Result)"}
   Get_Module_Metadata_2 = RigVMDispatchNode(Default=(R=1,G=1,B=0,A=1), NameSpace=Root, Name=@From_String_1.Result) {template="DISPATCH_RigDispatch_GetModuleMetadata(in Name,in NameSpace,in Default,out Value,out Found)"}
   From_String = RigVMDispatchNode(String=Global Right Control Color) {template="DISPATCH_RigDispatch_FromString(in String,out Result)"}
   Get_Module_Metadata_2_1 = RigVMDispatchNode(Default=(R=1,G=1,B=0,A=1), NameSpace=Root, Name=@From_String.Result) {template="DISPATCH_RigDispatch_GetModuleMetadata(in Name,in NameSpace,in Default,out Value,out Found)"}
   n1 = &RigUnit_GetModuleName()
   EndsWith = &RigVMFunction_StringEndsWith(Ending=" R", Name=@n1.Module) {template="EndsWith::Execute(in Name,in Ending,out Result)"}
   RigUnit_ItemToName = /Script/ControlRig.RigUnit_ItemToName(Value=RigVMVariableNode() {variable=Bone}.Value) {is=struct RigVMUnitNode, template="Cast::Execute(in Value,out Result)"}
   Has Side = RigVMFunctionReferenceNode(Side to Check=r, Name=@RigUnit_ItemToName.Result)
   RigVMFunction_IsNameValid = /Script/RigVM.RigVMFunction_IsNameValid(Value=@RigUnit_ItemToName.Result) {is=struct RigVMUnitNode, template="IsValid::Execute(in Value,out Result)"}
   n2 = &RigVMFunction_MathBoolOr(B=@EndsWith.Result, A=&RigVMFunction_MathBoolAnd(B=@"Has Side".Result, A=@RigVMFunction_IsNameValid.Result).Result)
   n3 = RigVMDispatchNode(False=@Get_Module_Metadata_2.Value, True=@Get_Module_Metadata_2_1.Value, Condition=@n2.Result) {template="DISPATCH_RigVMDispatch_If(in Condition,in True,in False,out Result)"}
   Has Side_1 = RigVMFunctionReferenceNode(Side to Check=l, Name=@RigUnit_ItemToName.Result)
   RigVMFunction_MathBoolOr = &RigVMFunction_MathBoolOr(
    B=&RigVMFunction_MathBoolOr(
     B=&RigVMFunction_StringEndsWith(Ending=" L", Name=@n1.Module) {template="EndsWith::Execute(in Name,in Ending,out Result)"}.Result
     A=&RigVMFunction_MathBoolAnd(B=@"Has Side_1".Result, A=@RigVMFunction_IsNameValid.Result).Result
    ).Result
    A=@n2.Result
   )
   n4 = RigVMVariableNode() {variable=Color Override}
   RigVMFunction_MathBoolAnd = &RigVMFunction_MathBoolAnd(
    A=RigVMDispatchNode(B=@n4.Value.G, A=@n4.Value.R) {template="DISPATCH_RigVMDispatch_CoreEquals(in A,in B,out Result)"}.Result
    B=RigVMDispatchNode(B=@n4.Value.B, A=@n4.Value.G) {template="DISPATCH_RigVMDispatch_CoreEquals(in A,in B,out Result)"}.Result
   )
   > RigVMFunctionReturnNode(Result=RigVMDispatchNode(
     True=RigVMDispatchNode(False=@Get_Module_Metadata.Value, True=@n3.Result, Condition=@RigVMFunction_MathBoolOr.Result) {template="DISPATCH_RigVMDispatch_If(in Condition,in True,in False,out Result)"}.Result
     Condition=@RigVMFunction_MathBoolAnd.Result
     False=@n4.Value
    ) {template="DISPATCH_RigVMDispatch_If(in Condition,in True,in False,out Result)"}.Result)
  ]}
 New Function_ContainedGraph {nodes [
   RigVMFunctionEntryNode()
   Concat_1 = RigVMAggregateNode(A=_, C=_, B=RigVMVariableNode() {variable=Side to Check}.Value)
   Contains = /Script/RigVM.RigVMFunction_Contains(Search=@Concat_1.Result, Name=RigVMVariableNode() {variable=Name}.Value) {is=struct RigVMUnitNode, template="Contains::Execute(in Name,in Search,out Result)"}
   n1 = RigVMVariableNode() {variable=Side to Check}
   Concat = &RigVMFunction_NameConcat(B=_, A=RigVMVariableNode() {variable=Side to Check}.Value) {template="Concat::Execute(in A,in B,out Result)"}
   StartsWith = /Script/RigVM.RigVMFunction_StartsWith(Start=@Concat.Result, Name=RigVMVariableNode() {variable=Name}.Value) {is=struct RigVMUnitNode, template="StartsWith::Execute(in Name,in Start,out Result)"}
   RigVMFunction_MathBoolOr_1 = RigVMAggregateNode(
    C=@Contains.Result
    E=/Script/RigVM.RigVMFunction_EndsWith(
     Ending=&RigVMFunction_NameConcat(A=_, B=@n1.Value) {template="Concat::Execute(in A,in B,out Result)"}.Result
     Name=RigVMVariableNode() {variable=Name}.Value
    ) {is=struct RigVMUnitNode, template="EndsWith::Execute(in Name,in Ending,out Result)"}.Result
    A=@StartsWith.Result
   )
   > RigVMFunctionReturnNode(Result=@RigVMFunction_MathBoolOr_1.Result)
  ]}
 New Function_ContainedGraph {nodes [
   RigVMFunctionEntryNode()
   n1 = RigVMVariableNode() {variable=Debug}
   n2 = RigVMVariableNode() {variable=PV Offset}
   n3 = &RigUnit_GetTransform(
    bInitial=False
    Item=RigVMDispatchNode(Index=1, Array=RigVMVariableNode() {variable=Bones}.Value) {template="DISPATCH_RigVMDispatch_ArrayGetAtIndex(in Array,in Index,out Element)"}.Element
   )
   n4 = RigVMVariableNode() {variable=Bones}
   n5 = /Script/ControlRig.RigUnit_GetTransformItemArray(Space=GlobalSpace, bInitial=true, Items=@n4.Value) {is=struct RigVMUnitNode}
   n6 = RigVMDispatchNode(Index=1, Array=@n5.Transforms) {template="DISPATCH_RigVMDispatch_ArrayGetAtIndex(in Array,in Index,out Element)"}
   n7 = RigVMVariableNode() {variable=Secondary Axis}
   Compute Pole Vector v01 = RigVMFunctionReferenceNode(Project To Plane=False, Transforms=@n5.Transforms, Secondary Axis=@n7.Value)
   n8 = &RigVMFunction_MathVectorAdd(B=@n6.Element.Translation, A=@"Compute Pole Vector v01".Pole Vector) {template="Add::Execute(in A,in B,out Result)"}
   n9 = &RigVMFunction_MathVectorSub(
    B=@n3.Transform.Translation
    A=&RigVMFunction_MathTransformTransformVector(
     Transform=@n3.Transform
     Location=&RigVMFunction_MathTransformMakeRelative(Global=(Rotation=(X=0,Y=0,Z=0,W=1),Translation=(X=0,Y=0,Z=0),Scale3D=(X=1,Y=1,Z=1)), Global.Translation=@n8.Result, Parent=@n6.Element).Local.Translation
    ).Result
   ) {template="Subtract::Execute(in A,in B,out Result)"}
   Compute Pole Vector v01_1 = RigVMFunctionReferenceNode(
    Project To Plane=False
    Transforms=&RigUnit_GetTransformItemArray(Space=GlobalSpace, bInitial=False, Items=@n4.Value).Transforms
    Secondary Axis=@n7.Value
   )
   n10 = &RigVMFunction_VisualDebugVectorNoSpace(
    Color=(R=0,G=1,B=0,A=1)
    DebugDrawSettings=(DepthPriority=SDPG_Foreground,Lifetime=-1)
    Mode=Vector
    Thickness=1
    bEnabled=@n1.Value
    Scale=@n2.Value
    Value=@"Compute Pole Vector v01_1".Pole Vector
   ) {template="VisualDebug::Execute(io Value,in bEnabled,in Mode,in Color,in Thickness,in Scale,in DebugDrawSettings)"}
   LessEqual = /Script/RigVM.RigVMFunction_MathDoubleLessEqual(
    B=0
    A=/Script/RigVM.RigVMFunction_MathVectorDot(
     B=&RigVMFunction_VisualDebugVectorNoSpace(
      Color=(R=0,G=0,B=1,A=1)
      DebugDrawSettings=(DepthPriority=SDPG_Foreground,Lifetime=-1)
      Mode=Vector
      Thickness=1
      bEnabled=@n1.Value
      Scale=@n2.Value
      Value=@n9.Result
     ) {template="VisualDebug::Execute(io Value,in bEnabled,in Mode,in Color,in Thickness,in Scale,in DebugDrawSettings)"}.Value
     A=@n10.Value
    ) {is=struct RigVMUnitNode}.Result
   ) {is=struct RigVMUnitNode, template="LessEqual::Execute(in A,in B,out Result)"}
   n11 = &RigVMFunction_MathVectorScale(
    Value=&RigVMFunction_MathVectorSetLength(Length=1, Value=@n10.Value) {template="SetLength::Execute(in Value,in Length,out Result)"}.Result
    Factor=&RigVMFunction_MathDoubleMul(B=RigVMVariableNode() {variable=PV Offset}.Value, A=@"Compute Pole Vector v01_1".Upper Segment Length).Result
   ) {template="Scale::Execute(in Value,in Factor,out Result)"}
   Negate = &RigVMFunction_MathVectorNegate(Value=@n11.Result) {template="Negate::Execute(in Value,out Result)"}
   > RigVMFunctionReturnNode(Pole Vector Location=&RigVMFunction_MathVectorAdd(
     B=RigVMDispatchNode(Condition=@LessEqual.Result, True=@Negate.Result, False=@n11.Result) {template="DISPATCH_RigVMDispatch_If(in Condition,in True,in False,out Result)"}.Result
     A=@"Compute Pole Vector v01_1".Elbow Position
    ) {template="Add::Execute(in A,in B,out Result)"}.Result)
  ]}
 New Function_ContainedGraph {nodes [
   n1 = RigVMFunctionEntryNode()
   n2 = RigVMDispatchNode(Index=1, Array=@n1.Transforms) {template="DISPATCH_RigVMDispatch_ArrayGetAtIndex(in Array,in Index,out Element)"}
   n3 = RigVMVariableNode() {variable=Secondary Axis}
   n4 = RigVMDispatchNode(Index=0, Array=@n1.Transforms) {template="DISPATCH_RigVMDispatch_ArrayGetAtIndex(in Array,in Index,out Element)"}
   n5 = &RigVMFunction_MathVectorSub(
    A=RigVMDispatchNode(Index=2, Array=@n1.Transforms) {template="DISPATCH_RigVMDispatch_ArrayGetAtIndex(in Array,in Index,out Element)"}.Element.Translation
    B=@n4.Element.Translation
   ) {template="Subtract::Execute(in A,in B,out Result)"}
   n6 = &RigVMFunction_MathVectorCross(
    A=&RigVMFunction_MathVectorSub(A=&RigVMFunction_MathTransformTransformVector(Transform=@n2.Element, Location=@n3.Value).Result, B=@n2.Element.Translation) {template="Subtract::Execute(in A,in B,out Result)"}.Result
    B=@n5.Result
   ) {template="Cross::Execute(in A,in B,out Result)"}
   n7 = RigVMDispatchNode(
    True=/Script/RigVM.RigVMFunction_MathIntersectPlane(Direction=@n6.Result, PlaneNormal=@n6.Result, Start=@n2.Element.Translation, PlanePoint=@n4.Element.Translation) {is=struct RigVMUnitNode}.Result
    Condition=RigVMVariableNode() {variable=Project To Plane}.Value
    False=@n2.Element.Translation
   ) {template="DISPATCH_RigVMDispatch_If(in Condition,in True,in False,out Result)"}
   n8 = &RigVMFunction_MathVectorSub(A=@n7.Result, B=@n4.Element.Translation) {template="Subtract::Execute(in A,in B,out Result)"}
   Dot___1 = &RigVMFunction_MathVectorDot(A=@n8.Result, B=@n5.Result) {template="Dot,|::Execute(in A,in B,out Result)"}
   n9 = &RigVMFunction_MathDoubleDiv(
    A=@Dot___1.Result
    B=/Script/RigVM.RigVMFunction_MathVectorLengthSquared(Value=@n5.Result) {is=struct RigVMUnitNode, template="LengthSquared::Execute(in Value,out Result)"}.Result
   )
   n10 = &RigVMFunction_MathVectorSub(
    B=&RigVMFunction_MathVectorAdd(
     B=&RigVMFunction_MathVectorScale(Factor=@n9.Result, Value=@n5.Result) {template="Scale::Execute(in Value,in Factor,out Result)"}.Result
     A=@n4.Element.Translation
    ) {template="Add::Execute(in A,in B,out Result)"}.Result
    A=@n7.Result
   ) {template="Subtract::Execute(in A,in B,out Result)"}
   IsNearlyZero = /Script/RigVM.RigVMFunction_MathVectorIsNearlyZero(Tolerance=0.01, Value=@n10.Result) {is=struct RigVMUnitNode, template="IsNearlyZero::Execute(in Value,in Tolerance,out Result)"}
   RigVMFunction_MathTransformTransformVector = &RigVMFunction_MathTransformTransformVector(
    Transform=RigVMDispatchNode(Index=1, Array=RigVMVariableNode() {variable=Transforms}.Value) {template="DISPATCH_RigVMDispatch_ArrayGetAtIndex(in Array,in Index,out Element)"}.Element
    Location=RigVMVariableNode() {variable=Secondary Axis}.Value
   )
   RigVMFunction_MathVectorLength = &RigVMFunction_MathVectorLength(Value=@n8.Result) {template="Length::Execute(in Value,out Result)"}
   > RigVMFunctionReturnNode(
    Pole Vector=RigVMDispatchNode(
     Condition=@IsNearlyZero.Result
     True=&RigVMFunction_MathVectorSub(A=@RigVMFunction_MathTransformTransformVector.Result, B=@n10.Result) {template="Subtract::Execute(in A,in B,out Result)"}.Result
     False=@n10.Result
    ) {template="DISPATCH_RigVMDispatch_If(in Condition,in True,in False,out Result)"}.Result
    Upper Segment Length=@RigVMFunction_MathVectorLength.Result
    Elbow Position=@n7.Result
   )
  ]}
 RigVMFunctionLibrary {nodes [
   Add Null Above = RigVMCollapseNode(ExecuteContext=(), Item=(Name="None"))
   Add Null Below = RigVMCollapseNode(ExecuteContext=(), Item=(Type=None,Name="None"), New Suffix=None, Replace Suffix=None)
   Compute Auto Pole Vector v02 = RigVMCollapseNode(
    ExecuteContext=()
    IK Control=(Type=None,Name="None")
    PV Control=(Type=None,Name="None")
    PV Orient Space=(Type=None,Name="None")
    PV Root Space=(Type=None,Name="None")
   )
   Compute Pole Vector From Plane v01 = RigVMCollapseNode(Initial=true, Item A=(Type=None,Name="None"), Item B=(Type=None,Name="None"), Item C=(Type=None,Name="None"), OffsetMultiplier=1)
   Compute Pole Vector Location v01 = RigVMCollapseNode(Initial=true, Item A=(Type=None,Name="None"), Item B=(Type=None,Name="None"), OffsetMultiplier=1, Up Vector=(X=0,Y=0,Z=0))
   Compute Pole Vector Location v02 = RigVMCollapseNode(Bones=(), PV Offset=10, Secondary Axis=(X=0,Y=0,Z=0))
   Compute Pole Vector v01 = RigVMCollapseNode(Project To Plane=False, Secondary Axis=(X=0,Y=1,Z=0), Transforms=())
   Construct Auto Pole Vector v02 = RigVMCollapseNode(
    ExecuteContext=()
    IK Control=(Type=None,Name="None")
    Item A=(Type=None,Name="None")
    Item B=(Type=None,Name="None")
    Item C=(Type=None,Name="None")
    PV Root Space=(Type=None,Name="None")
    Parent=(Type=None,Name="None")
    Shape=(bVisible=True,Name="Default",Color=(R=1,G=0,B=0,A=1),Transform=(Rotation=(X=0,Y=0,Z=0,W=1),Translation=(X=0,Y=0,Z=0),Scale3D=(X=1,Y=1,Z=1)))
   )
   Construct Movable Proxy v01 = RigVMCollapseNode(
    Driven Control=(Type=Control,Name="body_ctrl")
    ExecuteContext=()
    Proxy Name=proxy
    Proxy Parent Control=(Type=None,Name="leg_l_controls_space")
    Proxy Shape Settings="(DisplayName=\"None\",bSelectable=True,InitialSpace=LocalSpace,bUsePreferredRotationOrder=False,PreferredRotationOrder=YZX,Limits=(LimitTranslationX=(bMinimum=False,bMaximum=False),LimitTranslationY=(bMinimum=False,bMaximum=False),LimitTranslationZ=(bMinimum ...(truncated)"
    Proxy Snap To=(Type=None,Name="None")
    Proxy Vis=true
   )
   Control Stack at Item = RigVMCollapseNode(
    Add Bottom Null=False
    Add Null=true
    Add Secondary Control=False
    Control Suffix=_ctrl
    ExecuteContext=()
    Item=(Name="None")
    Null Suffix=_null
    Parent=(Name="None")
    Secondary Settings="(Shape=(bVisible=True,Name=\"Default\",Color=(R=1,G=0,B=0,A=1),Transform=(Rotation=(X=0,Y=0,Z=0,W=1),Translation=(X=0,Y=0,Z=0),Scale3D=(X=1,Y=1,Z=1))),Proxy=(bIsProxy=False,ShapeVisibility=BasedOnSelection),DisplayName=\"None\",bSelectable=True,InitialSpace=Lo ...(truncated)"
    Settings="(Shape=(bVisible=True,Name=\"Default\",Color=(R=1,G=0,B=0,A=1),Transform=(Rotation=(X=0,Y=0,Z=0,W=1),Translation=(X=0,Y=0,Z=0),Scale3D=(X=1,Y=1,Z=1))),Proxy=(bIsProxy=False,ShapeVisibility=BasedOnSelection),DisplayName=\"None\",bSelectable=True,InitialSpace=Lo ...(truncated)"
   )
   Control Stack at Position = RigVMCollapseNode(
    Add Bottom Null=False
    Add Null=False
    Add Secondary Control=False
    Control Suffix=_ctrl
    ExecuteContext=()
    Global Position=(Rotation=(X=0,Y=0,Z=0,W=1),Translation=(X=0,Y=0,Z=0),Scale3D=(X=1,Y=1,Z=1))
    Null Suffix=_null
    Parent=(Type=None,Name="None")
    Secondary Settings="(Shape=(bVisible=True,Name=\"Default\",Color=(R=1,G=0,B=0,A=1),Transform=(Rotation=(X=0,Y=0,Z=0,W=1),Translation=(X=0,Y=0,Z=0),Scale3D=(X=1,Y=1,Z=1))),Proxy=(bIsProxy=False,ShapeVisibility=BasedOnSelection),DisplayName=\"None\",bSelectable=True,InitialSpace=Lo ...(truncated)"
    Settings="(Shape=(bVisible=True,Name=\"Default\",Color=(R=1,G=0,B=0,A=1),Transform=(Rotation=(X=0,Y=0,Z=0,W=1),Translation=(X=0,Y=0,Z=0),Scale3D=(X=1,Y=1,Z=1))),Proxy=(bIsProxy=False,ShapeVisibility=BasedOnSelection),DisplayName=\"None\",bSelectable=True,InitialSpace=Lo ...(truncated)"
   )
   Create IK Plane Virtual Bones = RigVMCollapseNode(Axis=(X=1,Y=0,Z=0), End=(Type=None,Name="None"), ExecuteContext=(), Mid=(Type=None,Name="None"), Start=(Type=None,Name="None"))
   Create IK Plane Virtual Bones v02 = RigVMCollapseNode(
    Axis=(X=1,Y=0,Z=0)
    End=(Type=None,Name="None")
    ExecuteContext=()
    Mid=(Type=None,Name="None")
    Secondary Axis=(X=0,Y=1,Z=0)
    Start=(Type=None,Name="None")
   )
   Forward Movable Proxy v01 = RigVMCollapseNode(
    Driven Control=(Type=Control,Name="body_ctrl")
    ExecuteContext=()
    Proxy Buffer=(Type=Null,Name="foot_l_proxy_bfr")
    Proxy Control=(Type=Control,Name="foot_l_proxy")
    Proxy Snap To=(Type=Control,Name="body_ctrl")
    Proxy Vis Channel=(Type=Control,Name="body_ctrl")
   )
   Get Array Parents = RigVMCollapseNode(ExecuteContext=(), Items=())
   Get Children by Contained Strings = RigVMCollapseNode(ExecuteContext=(), Item=(Type=None,Name="None"), Recursive=true)
   Get Control Color From Metadata = RigVMCollapseNode(Color Override=(R=1,G=1,B=1,A=1))
   Get Control Color From Metadata v02 = RigVMCollapseNode(Bone=(Type=None,Name="None"), Color Override=(R=1,G=1,B=1,A=1))
   Get Mirror Transform = RigVMCollapseNode(Axis=(X=1,Y=0,Z=0), ExecuteContext=(), Transform=(Rotation=(X=0,Y=0,Z=0,W=1),Translation=(X=0,Y=0,Z=0),Scale3D=(X=1,Y=1,Z=1)))
   Has Side = RigVMCollapseNode()
   Set Contol Color By Position = RigVMCollapseNode(Center Color=(R=1,G=1,B=0,A=1), Controls=(), ExecuteContext=(), Left Color=(R=0,G=1,B=0,A=1), Right Color=(R=1,G=0,B=0,A=1))
   Set Control Scale = RigVMCollapseNode(Auto Scale Factor=1, Controls=(), ExecuteContext=(), Global Scale=1, Scale=1)
   Switch Control Visibility = RigVMCollapseNode(ExecuteContext=(), Off Controls=(), On Controls=())
  ]}
 RigVMFunction_MathBoolOr_2_ContainedGraph {nodes [
   n1 = RigVMFunctionEntryNode()
   RigVMFunctionReturnNode(Result=&RigVMFunction_MathBoolOr(B=@n1.E, A=&RigVMFunction_MathBoolOr(A=@n1.A, B=@n1.C).Result).Result)
  ]}
 RigVMFunction_Sequence_1_1_ContainedGraph {nodes [
   n1 = RigVMFunctionEntryNode()
   n2 = &RigVMFunction_Sequence(ExecuteContext=@n1.ExecuteContext)
   n3 = &RigVMFunction_Sequence(ExecuteContext=@n2.B)
   RigVMFunctionReturnNode(A=@n2.A, B=@n3.A, C=@n3.B)
  ]}
 RigVMFunction_Sequence_1_ContainedGraph {nodes [
   n1 = RigVMFunctionEntryNode()
   RigVMFunction_Sequence = &RigVMFunction_Sequence(ExecuteContext=@n1.ExecuteContext)
   n2 = &RigVMFunction_Sequence(ExecuteContext=@RigVMFunction_Sequence.B)
   n3 = &RigVMFunction_Sequence(ExecuteContext=@n2.B)
   n4 = &RigVMFunction_Sequence(ExecuteContext=@n3.B)
   n5 = &RigVMFunction_Sequence(ExecuteContext=@n4.B)
   RigVMFunctionReturnNode(A=@RigVMFunction_Sequence.A, B=@n2.A, C=@n3.A, D=@n4.A, E=@n5.A, F=@n5.B)
  ]}
 RigVMFunction_Sequence_1_ContainedGraph {nodes [
   n1 = RigVMFunctionEntryNode()
   RigVMFunction_Sequence = &RigVMFunction_Sequence(ExecuteContext=@n1.ExecuteContext)
   n2 = &RigVMFunction_Sequence(ExecuteContext=@RigVMFunction_Sequence.B)
   RigVMFunctionReturnNode(A=@RigVMFunction_Sequence.A, B=@n2.A, C=@n2.B)
  ]}
 RigVMFunction_Sequence_1_ContainedGraph {nodes [
   n1 = RigVMFunctionEntryNode()
   RigVMFunction_Sequence = &RigVMFunction_Sequence(ExecuteContext=@n1.ExecuteContext)
   n2 = &RigVMFunction_Sequence(ExecuteContext=@RigVMFunction_Sequence.B)
   RigVMFunctionReturnNode(A=@RigVMFunction_Sequence.A, B=@n2.A, C=@n2.B)
  ]}
 RigVMModel {nodes [
   RigUnit_BeginExecution = &RigUnit_BeginExecution()
   Get Children by Contained Strings = RigVMFunctionReferenceNode(
    Item=(Type=Bone,Name="upperarm_l")
    Recursive=true
    Search String=lowerarm|twist
    Seperator=|
    Type to Search=Bone
    ExecuteContext=@RigUnit_BeginExecution.ExecutePin
   )
  ]}
 RigVMModel Rig {nodes [&RigUnit_PrepareForExecution()]}
]
hierarchy [
 MoveData_Speed {type=Curve, loc=<0 0 0>, rot=<0 0 0 1>}
 base_pose {type=Curve, loc=<0 0 0>, rot=<0 0 0 1>}
 calf_l_back_50 {type=Curve, loc=<0 0 0>, rot=<0 0 0 1>}
 calf_l_back_120 {type=Curve, loc=<0 0 0>, rot=<0 0 0 1>}
 calf_l_back_90 {type=Curve, loc=<0 0 0>, rot=<0 0 0 1>}
 calf_l_back_150 {type=Curve, loc=<0 0 0>, rot=<0 0 0 1>}
 calf_r_back_50 {type=Curve, loc=<0 0 0>, rot=<0 0 0 1>}
 calf_r_back_120 {type=Curve, loc=<0 0 0>, rot=<0 0 0 1>}
 calf_r_back_90 {type=Curve, loc=<0 0 0>, rot=<0 0 0 1>}
 calf_r_back_150 {type=Curve, loc=<0 0 0>, rot=<0 0 0 1>}
 clavicle_l_fwd_30 {type=Curve, loc=<0 0 0>, rot=<0 0 0 1>}
 clavicle_l_down_20 {type=Curve, loc=<0 0 0>, rot=<0 0 0 1>}
 clavicle_l_back_30 {type=Curve, loc=<0 0 0>, rot=<0 0 0 1>}
 clavicle_l_up_40 {type=Curve, loc=<0 0 0>, rot=<0 0 0 1>}
 clavicle_r_fwd_30 {type=Curve, loc=<0 0 0>, rot=<0 0 0 1>}
 clavicle_r_down_20 {type=Curve, loc=<0 0 0>, rot=<0 0 0 1>}
 clavicle_r_back_30 {type=Curve, loc=<0 0 0>, rot=<0 0 0 1>}
 clavicle_r_up_40 {type=Curve, loc=<0 0 0>, rot=<0 0 0 1>}
 clavicle_r_back_310 {type=Curve, loc=<0 0 0>, rot=<0 0 0 1>}
 clavicle_r_up_401 {type=Curve, loc=<0 0 0>, rot=<0 0 0 1>}
 foot_l_down_60 {type=Curve, loc=<0 0 0>, rot=<0 0 0 1>}
 foot_l_up_35 {type=Curve, loc=<0 0 0>, rot=<0 0 0 1>}
 foot_r_up_35 {type=Curve, loc=<0 0 0>, rot=<0 0 0 1>}
 foot_r_down_60 {type=Curve, loc=<0 0 0>, rot=<0 0 0 1>}
 hand_l_down_90 {type=Curve, loc=<0 0 0>, rot=<0 0 0 1>}
 hand_l_up_20 {type=Curve, loc=<0 0 0>, rot=<0 0 0 1>}
 hand_l_up_90 {type=Curve, loc=<0 0 0>, rot=<0 0 0 1>}
 hand_r_up_90 {type=Curve, loc=<0 0 0>, rot=<0 0 0 1>}
 hand_r_down_90 {type=Curve, loc=<0 0 0>, rot=<0 0 0 1>}
 hand_r_up_20 {type=Curve, loc=<0 0 0>, rot=<0 0 0 1>}
 Pose_0 {type=Curve, loc=<0 0 0>, rot=<0 0 0 1>}
 Pose_1 {type=Curve, loc=<0 0 0>, rot=<0 0 0 1>}
 Pose_2 {type=Curve, loc=<0 0 0>, rot=<0 0 0 1>}
 Pose_3 {type=Curve, loc=<0 0 0>, rot=<0 0 0 1>}
 Pose_4 {type=Curve, loc=<0 0 0>, rot=<0 0 0 1>}
 Pose_5 {type=Curve, loc=<0 0 0>, rot=<0 0 0 1>}
 Pose_6 {type=Curve, loc=<0 0 0>, rot=<0 0 0 1>}
 Pose_7 {type=Curve, loc=<0 0 0>, rot=<0 0 0 1>}
 Pose_8 {type=Curve, loc=<0 0 0>, rot=<0 0 0 1>}
 lowerarm_l_in_110 {type=Curve, loc=<0 0 0>, rot=<0 0 0 1>}
 lowerarm_l_in_35 {type=Curve, loc=<0 0 0>, rot=<0 0 0 1>}
 lowerarm_l_out_35 {type=Curve, loc=<0 0 0>, rot=<0 0 0 1>}
 lowerarm_l_in_50 {type=Curve, loc=<0 0 0>, rot=<0 0 0 1>}
 lowerarm_l_in_75 {type=Curve, loc=<0 0 0>, rot=<0 0 0 1>}
 lowerarm_l_in_10 {type=Curve, loc=<0 0 0>, rot=<0 0 0 1>}
 lowerarm_l_out_10 {type=Curve, loc=<0 0 0>, rot=<0 0 0 1>}
 lowerarm_l_in_90 {type=Curve, loc=<0 0 0>, rot=<0 0 0 1>}
 lowerarm_r_in_10 {type=Curve, loc=<0 0 0>, rot=<0 0 0 1>}
 lowerarm_r_out_10 {type=Curve, loc=<0 0 0>, rot=<0 0 0 1>}
 lowerarm_r_out_35 {type=Curve, loc=<0 0 0>, rot=<0 0 0 1>}
 lowerarm_r_in_75 {type=Curve, loc=<0 0 0>, rot=<0 0 0 1>}
 lowerarm_r_in_50 {type=Curve, loc=<0 0 0>, rot=<0 0 0 1>}
 lowerarm_r_in_110 {type=Curve, loc=<0 0 0>, rot=<0 0 0 1>}
 lowerarm_r_in_35 {type=Curve, loc=<0 0 0>, rot=<0 0 0 1>}
 lowerarm_r_in_90 {type=Curve, loc=<0 0 0>, rot=<0 0 0 1>}
 thigh_l_bck_10 {type=Curve, loc=<0 0 0>, rot=<0 0 0 1>}
 thigh_l_fwd_10 {type=Curve, loc=<0 0 0>, rot=<0 0 0 1>}
 thigh_l_in_45_out_90 {type=Curve, loc=<0 0 0>, rot=<0 0 0 1>}
 thigh_l_fwd_90 {type=Curve, loc=<0 0 0>, rot=<0 0 0 1>}
 thigh_l_bck_90 {type=Curve, loc=<0 0 0>, rot=<0 0 0 1>}
 thigh_l_bck_50 {type=Curve, loc=<0 0 0>, rot=<0 0 0 1>}
 thigh_l_out_110 {type=Curve, loc=<0 0 0>, rot=<0 0 0 1>}
 thigh_l_out_10 {type=Curve, loc=<0 0 0>, rot=<0 0 0 1>}
 thigh_l_fwd_45 {type=Curve, loc=<0 0 0>, rot=<0 0 0 1>}
 thigh_l_in_50 {type=Curve, loc=<0 0 0>, rot=<0 0 0 1>}
 thigh_l_fwd_110 {type=Curve, loc=<0 0 0>, rot=<0 0 0 1>}
 thigh_l_out_55 {type=Curve, loc=<0 0 0>, rot=<0 0 0 1>}
 thigh_l_out_85 {type=Curve, loc=<0 0 0>, rot=<0 0 0 1>}
 thigh_r_fwd_10 {type=Curve, loc=<0 0 0>, rot=<0 0 0 1>}
 thigh_r_out_55 {type=Curve, loc=<0 0 0>, rot=<0 0 0 1>}
 thigh_r_fwd_90 {type=Curve, loc=<0 0 0>, rot=<0 0 0 1>}
 thigh_r_fwd_45 {type=Curve, loc=<0 0 0>, rot=<0 0 0 1>}
 thigh_r_bck_10 {type=Curve, loc=<0 0 0>, rot=<0 0 0 1>}
 thigh_r_out_85 {type=Curve, loc=<0 0 0>, rot=<0 0 0 1>}
 thigh_r_in_50 {type=Curve, loc=<0 0 0>, rot=<0 0 0 1>}
 thigh_r_out_10 {type=Curve, loc=<0 0 0>, rot=<0 0 0 1>}
 thigh_r_out_110 {type=Curve, loc=<0 0 0>, rot=<0 0 0 1>}
 thigh_r_bck_50 {type=Curve, loc=<0 0 0>, rot=<0 0 0 1>}
 thigh_r_bck_90 {type=Curve, loc=<0 0 0>, rot=<0 0 0 1>}
 thigh_r_in_45_out_90 {type=Curve, loc=<0 0 0>, rot=<0 0 0 1>}
 thigh_r_fwd_110 {type=Curve, loc=<0 0 0>, rot=<0 0 0 1>}
 upperarm_l_back_45 {type=Curve, loc=<0 0 0>, rot=<0 0 0 1>}
 upperarm_l_out_55 {type=Curve, loc=<0 0 0>, rot=<0 0 0 1>}
 upperarm_l_out110_twist_in_70 {type=Curve, loc=<0 0 0>, rot=<0 0 0 1>}
 upperarm_l_out_85 {type=Curve, loc=<0 0 0>, rot=<0 0 0 1>}
 upperarm_l_fwd_110 {type=Curve, loc=<0 0 0>, rot=<0 0 0 1>}
 upperarm_l_in_10 {type=Curve, loc=<0 0 0>, rot=<0 0 0 1>}
 upperarm_l_out110_twist_out_70 {type=Curve, loc=<0 0 0>, rot=<0 0 0 1>}
 upperarm_l_out_twist_a {type=Curve, loc=<0 0 0>, rot=<0 0 0 1>}
 upperarm_l_fwd_15 {type=Curve, loc=<0 0 0>, rot=<0 0 0 1>}
 upperarm_l_out_15 {type=Curve, loc=<0 0 0>, rot=<0 0 0 1>}
 upperarm_l_back_10 {type=Curve, loc=<0 0 0>, rot=<0 0 0 1>}
 upperarm_l_out110 {type=Curve, loc=<0 0 0>, rot=<0 0 0 1>}
 upperarm_l_fwd_twist_in_a {type=Curve, loc=<0 0 0>, rot=<0 0 0 1>}
 upperarm_l_fwd_90 {type=Curve, loc=<0 0 0>, rot=<0 0 0 1>}
 upperarm_r_out_15 {type=Curve, loc=<0 0 0>, rot=<0 0 0 1>}
 upperarm_r_fwd_15 {type=Curve, loc=<0 0 0>, rot=<0 0 0 1>}
 upperarm_r_out_twist_a {type=Curve, loc=<0 0 0>, rot=<0 0 0 1>}
 upperarm_r_out110_twist_in_70 {type=Curve, loc=<0 0 0>, rot=<0 0 0 1>}
 upperarm_r_fwd_90 {type=Curve, loc=<0 0 0>, rot=<0 0 0 1>}
 upperarm_r_out110 {type=Curve, loc=<0 0 0>, rot=<0 0 0 1>}
 upperarm_r_back_10 {type=Curve, loc=<0 0 0>, rot=<0 0 0 1>}
 upperarm_r_out_85 {type=Curve, loc=<0 0 0>, rot=<0 0 0 1>}
 upperarm_r_in_10 {type=Curve, loc=<0 0 0>, rot=<0 0 0 1>}
 upperarm_r_back_45 {type=Curve, loc=<0 0 0>, rot=<0 0 0 1>}
 upperarm_r_out110_twist_out_70 {type=Curve, loc=<0 0 0>, rot=<0 0 0 1>}
 upperarm_r_out_55 {type=Curve, loc=<0 0 0>, rot=<0 0 0 1>}
 upperarm_r_fwd_twist_in_a {type=Curve, loc=<0 0 0>, rot=<0 0 0 1>}
 upperarm_r_fwd_110 {type=Curve, loc=<0 0 0>, rot=<0 0 0 1>}
 Distance {type=Curve, loc=<0 0 0>, rot=<0 0 0 1>}
 RemainingTurnYaw {type=Curve, loc=<0 0 0>, rot=<0 0 0 1>}
 TurnYawWeight {type=Curve, loc=<0 0 0>, rot=<0 0 0 1>}
 GroundDistance {type=Curve, loc=<0 0 0>, rot=<0 0 0 1>}
 DisableLHandIK {type=Curve, loc=<0 0 0>, rot=<0 0 0 1>}
 ScaleDownWeaponR {type=Curve, loc=<0 0 0>, rot=<0 0 0 1>}
 DisableLegIK {type=Curve, loc=<0 0 0>, rot=<0 0 0 1>}
 blendParent1 {type=Curve, loc=<0 0 0>, rot=<0 0 0 1>}
 DisableRHandIK {type=Curve, loc=<0 0 0>, rot=<0 0 0 1>}
 DisableLeftHandPoseOverride {type=Curve, loc=<0 0 0>, rot=<0 0 0 1>}
 applyHipfireOverridePose {type=Curve, loc=<0 0 0>, rot=<0 0 0 1>}
 DisableHandIKRetargeting {type=Curve, loc=<0 0 0>, rot=<0 0 0 1>}
 lowerarm_r_in_80 {type=Curve, loc=<0 0 0>, rot=<0 0 0 1>}
 lowerarm_l_in_80 {type=Curve, loc=<0 0 0>, rot=<0 0 0 1>}
 calf_r_back_130 {type=Curve, loc=<0 0 0>, rot=<0 0 0 1>}
 calf_l_back_130 {type=Curve, loc=<0 0 0>, rot=<0 0 0 1>}
 Pose_9 {type=Curve, loc=<0 0 0>, rot=<0 0 0 1>}
 root {loc=<0 0 0>, rot=<0 0 0 1>}
 pelvis {parent=root, loc=<0 2.2809 95.8968>, rot=<-0.0224 0.7068 0.0224 -0.7068>}
 spine_01 {parent=pelvis, loc=<3.6771 0 0>, rot=<0 0 -0.1258 -0.9921>}
 spine_02 {parent=spine_01, loc=<6.7951 0 0>, rot=<0 0 0.0302 -0.9995>}
 spine_03 {parent=spine_02, loc=<7.2382 0 0>, rot=<0 0 0.0954 -0.9954>}
 spine_04 {parent=spine_03, loc=<8.5239 0 0>, rot=<0 0 0.0512 -0.9987>}
 spine_05 {parent=spine_04, loc=<19.4398 0 0>, rot=<0 0 0.0059 -1>}
 neck_01 {parent=spine_05, loc=<11.8878 0 0>, rot=<0 0 -0.2073 -0.9783>}
 neck_02 {parent=neck_01, loc=<5.1103 0 0>, rot=<0 0 0.0167 -0.9999>}
 head {parent=neck_02, loc=<4.913 0 0>, rot=<0 0 0.1035 -0.9946>}
 clavicle_l {parent=spine_05, loc=<5.5163 -1.3148 -1.4279>, rot=<0.0833 -0.7546 -0.0413 -0.6496>}
 upperarm_l {parent=clavicle_l, loc=<17.8095 0 0>, rot=<-0.0199 -0.3917 -0.0202 -0.9196>}
 lowerarm_l {parent=upperarm_l, loc=<27.7711 0 0>, rot=<0 0 -0.3336 -0.9427>}
 lowerarm_twist_02_l {parent=lowerarm_l, loc=<9.0837 0 0>, rot=<0.0115 0.0149 0.0175 -0.9997>}
 lowerarm_twist_01_l {parent=lowerarm_l, loc=<18.1674 0 0>, rot=<0.0115 0.0149 0.0175 -0.9997>}
 hand_l {parent=lowerarm_l, loc=<27.2511 0 0>, rot=<-0.5576 -0.0017 0.0206 -0.8299>}
 index_metacarpal_l {parent=hand_l, loc=<3.4445 0.3847 -2.3793>, rot=<0.029 -0.0637 -0.0071 -0.9975>}
 index_01_l {parent=index_metacarpal_l, loc=<5.8771 -0.0432 0.2409>, rot=<0 0 0.2026 -0.9793>}
 index_02_l {parent=index_01_l, loc=<4.08 0 0>, rot=<0 0 0.1296 -0.9916>}
 index_03_l {parent=index_02_l, loc=<2.595 0 0>, rot=<0 0 0.109 -0.994>}
 middle_metacarpal_l {parent=hand_l, loc=<3.3758 0.7536 -0.1829>, rot=<-0.0373 0.0004 -0.0203 -0.9991>}
 middle_01_l {parent=middle_metacarpal_l, loc=<6.0982 0 0>, rot=<0 0 0.2721 -0.9623>}
 middle_02_l {parent=middle_01_l, loc=<5.169 0 0>, rot=<0 0 0.1803 -0.9836>}
 middle_03_l {parent=middle_02_l, loc=<2.474 0 0>, rot=<0 0 0.0872 -0.9962>}
 thumb_01_l {parent=hand_l, loc=<1.9924 -1.3566 -2.5815>, rot=<0.5052 -0.3691 -0.067 -0.7772>}
 thumb_02_l {parent=thumb_01_l, loc=<4.378 0 0>, rot=<0.0336 0.0103 0.2019 -0.9788>}
 thumb_03_l {parent=thumb_02_l, loc=<3.086 0 0>, rot=<0 0 0.0872 -0.9962>}
 pinky_metacarpal_l {parent=hand_l, loc=<3.3144 0.3059 2.3911>, rot=<-0.2182 0.1882 0.0583 -0.9558>}
 pinky_01_l {parent=pinky_metacarpal_l, loc=<4.9576 0.1431 -0.1988>, rot=<0.09 -0.017 0.1281 -0.9875>}
 pinky_02_l {parent=pinky_01_l, loc=<3.816 0 0>, rot=<0 0 0.1847 -0.9828>}
 pinky_03_l {parent=pinky_02_l, loc=<2.04 0 0>, rot=<0 0 0.0429 -0.9991>}
 ring_metacarpal_l {parent=hand_l, loc=<3.3743 0.5425 1.0918>, rot=<-0.1166 0.1006 -0.0257 -0.9877>}
 ring_01_l {parent=ring_metacarpal_l, loc=<5.6455 0.0416 -0.0207>, rot=<0.0542 -0.0132 0.2535 -0.9657>}
 ring_02_l {parent=ring_01_l, loc=<4.977 0 0>, rot=<0 0 0.1647 -0.9863>}
 ring_03_l {parent=ring_02_l, loc=<2.265 0 0>, rot=<0 0 0.0799 -0.9968>}
 upperarm_twist_01_l {parent=upperarm_l, loc=<9.257 0 0>, rot=<-0.0001 0.0188 -0.0028 -0.9998>}
 upperarm_twist_02_l {parent=upperarm_l, loc=<18.5141 0 0>, rot=<0 0 0 1>}
 clavicle_r {parent=spine_05, loc=<5.5162 -1.3148 1.4279>, rot=<0.7546 0.0833 -0.6496 0.0413>}
 upperarm_r {parent=clavicle_r, loc=<-17.8096 0 0.0004>, rot=<-0.0199 -0.3917 -0.0202 -0.9196>}
 lowerarm_r {parent=upperarm_r, loc=<-27.7707 0 0>, rot=<0 0 -0.3336 -0.9427>}
 lowerarm_twist_02_r {parent=lowerarm_r, loc=<-9.0837 0 0>, rot=<0.0115 0.0149 0.0175 -0.9997>}
 lowerarm_twist_01_r {parent=lowerarm_r, loc=<-18.1673 0 0>, rot=<0.0115 0.0149 0.0175 -0.9997>}
 hand_r {parent=lowerarm_r, loc=<-27.251 0 0>, rot=<-0.5576 -0.0017 0.0206 -0.8299>}
 pinky_metacarpal_r {parent=hand_r, loc=<-3.3147 -0.3059 -2.3913>, rot=<-0.2182 0.1882 0.0583 -0.9558>}
 pinky_01_r {parent=pinky_metacarpal_r, loc=<-4.9573 -0.1433 0.1989>, rot=<0.09 -0.017 0.1281 -0.9875>}
 pinky_02_r {parent=pinky_01_r, loc=<-3.816 0 0>, rot=<0 0 0.1847 -0.9828>}
 pinky_03_r {parent=pinky_02_r, loc=<-2.04 0 0>, rot=<0 0 0.0429 -0.9991>}
 ring_metacarpal_r {parent=hand_r, loc=<-3.3742 -0.543 -1.0918>, rot=<-0.1166 0.1006 -0.0257 -0.9877>}
 ring_01_r {parent=ring_metacarpal_r, loc=<-5.6457 -0.0414 0.0207>, rot=<0.0542 -0.0132 0.2535 -0.9657>}
 ring_02_r {parent=ring_01_r, loc=<-4.9771 0 0>, rot=<0 0 0.1647 -0.9863>}
 ring_03_r {parent=ring_02_r, loc=<-2.265 -0.0001 0>, rot=<0 0 0.0799 -0.9968>}
 middle_metacarpal_r {parent=hand_r, loc=<-3.3758 -0.754 0.1828>, rot=<-0.0373 0.0004 -0.0203 -0.9991>}
 middle_01_r {parent=middle_metacarpal_r, loc=<-6.0984 0.0001 0>, rot=<0 0 0.2721 -0.9623>}
 middle_02_r {parent=middle_01_r, loc=<-5.169 0.0001 0>, rot=<0 0 0.1803 -0.9836>}
 middle_03_r {parent=middle_02_r, loc=<-2.474 0 0>, rot=<0 0 0.0872 -0.9962>}
 index_metacarpal_r {parent=hand_r, loc=<-3.4445 -0.3852 2.3793>, rot=<0.029 -0.0637 -0.0071 -0.9975>}
 index_01_r {parent=index_metacarpal_r, loc=<-5.8772 0.0434 -0.241>, rot=<0 0 0.2026 -0.9793>}
 index_02_r {parent=index_01_r, loc=<-4.0799 0 0>, rot=<0 0 0.1296 -0.9916>}
 index_03_r {parent=index_02_r, loc=<-2.5951 0 0>, rot=<0 0 0.109 -0.994>}
 thumb_01_r {parent=hand_r, loc=<-1.9928 1.3567 2.5813>, rot=<0.5052 -0.3691 -0.067 -0.7772>}
 thumb_02_r {parent=thumb_01_r, loc=<-4.3778 -0.0004 -0.0001>, rot=<0.0336 0.0103 0.2019 -0.9788>}
 thumb_03_r {parent=thumb_02_r, loc=<-3.086 0 0>, rot=<0 0 0.0872 -0.9962>}
 upperarm_twist_01_r {parent=upperarm_r, loc=<-9.2569 0 -0.0001>, rot=<-0.0001 0.0188 -0.0028 -0.9998>}
 upperarm_twist_02_r {parent=upperarm_r, loc=<-18.5138 0 -0.0002>, rot=<0 0 0 1>}
 thigh_r {parent=pelvis, loc=<-2.3657 0.1195 9.9691>, rot=<0.0295 -0.0724 0.9965 -0.029>}
 calf_r {parent=thigh_r, loc=<43.3413 0 0>, rot=<0 0 -0.0437 -0.999>}
 foot_r {parent=calf_r, loc=<42.2179 0 0>, rot=<-0.0007 -0.0269 0.0232 -0.9994>}
 ball_r {parent=foot_r, loc=<7.0094 -15.2376 -0.5389>, rot=<0 0 -0.7071 -0.7071>}
 calf_twist_02_r {parent=calf_r, loc=<14.0726 0 0.0377>, rot=<-0.0001 -0.0023 0.0232 -0.9997>}
 calf_twist_01_r {parent=calf_r, loc=<28.1453 0 0.0755>, rot=<-0.0001 -0.0023 0.0232 -0.9997>}
 thigh_twist_01_r {parent=thigh_r, loc=<14.4471 0 0>, rot=<0 -0.0023 -0.0114 -0.9999>}
 thigh_twist_02_r {parent=thigh_r, loc=<28.8942 0 0>, rot=<0 -0.0023 -0.0114 -0.9999>}
 thigh_l {parent=pelvis, loc=<-2.3657 0.11 -9.9692>, rot=<0.0724 0.0295 -0.029 -0.9965>}
 calf_l {parent=thigh_l, loc=<-43.3413 0 0>, rot=<0 0 -0.0437 -0.999>}
 foot_l {parent=calf_l, loc=<-42.2179 0 0>, rot=<-0.0007 -0.0269 0.0232 -0.9994>}
 ball_l {parent=foot_l, loc=<-7.0094 15.2376 0.5389>, rot=<0 0 -0.7071 -0.7071>}
 calf_twist_02_l {parent=calf_l, loc=<-14.0726 0 0>, rot=<-0.0001 -0.0023 0.0232 -0.9997>}
 calf_twist_01_l {parent=calf_l, loc=<-28.1453 0 -0.0754>, rot=<-0.0001 -0.0023 0.0232 -0.9997>}
 thigh_twist_01_l {parent=thigh_l, loc=<-14.4471 0 0>, rot=<0 -0.0023 -0.0114 -0.9999>}
 thigh_twist_02_l {parent=thigh_l, loc=<-28.8942 0 0>, rot=<0 -0.0023 -0.0114 -0.9999>}
 ik_foot_root {parent=root, loc=<0 0 0>, rot=<0 0 0 1>}
 ik_foot_l {parent=ik_foot_root, loc=<14.0877 -0.9943 8.2376>, rot=<0.0636 0.7042 0.0414 -0.7059>}
 ik_foot_r {parent=ik_foot_root, loc=<-14.0876 -0.9848 8.2382>, rot=<0.7059 0.0414 -0.7042 0.0636>}
 ik_hand_root {parent=root, loc=<0 0 0>, rot=<0 0 0 1>}
 ik_hand_gun {parent=ik_hand_root, loc=<-47.7683 15.6988 104.4664>, rot=<0.5985 0.0096 0.5206 -0.6089>}
 ik_hand_l {parent=ik_hand_gun, loc=<43.7336 61.6698 58.4081>, rot=<-0.4578 -0.6455 -0.6114 0>}
 ik_hand_r {parent=ik_hand_gun, loc=<0 0 0>, rot=<0 0 0 1>}
 interaction {parent=root, loc=<0 0 0>, rot=<0 0 0 1>}
 center_of_mass {parent=root, loc=<0 0 0>, rot=<0 0 0 1>}
]
