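# Control Rig function library of math helpers for bone chains and IK setup.
# Get Chain Length sums distances between successive items; Get Node Twist Value extracts twist about an axis from local rotation relative to initial.
# Compute Pole Vector offsets the middle bone away from the A-to-C line and builds an orientation.
# Project Middle Bone to IK Plane flattens the initial mid onto a secondary-axis-derived plane, preserving start-to-mid length, with debug drawing.
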ControlRigBlueprint /EpicControlRig/Libraries/CRFL_Math_v001
graphs [
 ContainedGraph {nodes [
   n1 = RigVMFunctionEntryNode()
   RigUnit_SequenceAggregate_1 = &RigVMFunction_Sequence(ExecuteContext=@n1.ExecuteContext)
   n2 = RigVMDispatchNode(ExecuteContext=@RigUnit_SequenceAggregate_1.A, Array=RigVMVariableNode() {variable=Items}.Value) {template="DISPATCH_RigVMDispatch_ArrayIterator(in Array,out Element,out Index,out Count,out Ratio)"}
   MathIntEquals = RigVMDispatchNode(
    B=/Script/RigVM.RigVMFunction_MathIntSub(B=1, A=@n2.Count) {is=struct RigVMUnitNode, template="Subtract::Execute(in A,in B,out Result)"}.Result
    A=@n2.Index
   ) {template="DISPATCH_RigVMDispatch_CoreEquals(in A,in B,out Result)"}
   n3 = &RigVMFunction_ControlFlowBranch(ExecuteContext=@n2.ExecuteContext, Condition=&RigVMFunction_MathBoolNot(Value=@MathIntEquals.Result).Result)
   n4 = RigVMVariableNode() {variable=Initial}
   n5 = &RigUnit_GetTransform(
    Item=RigVMDispatchNode(
     Index=/Script/RigVM.RigVMFunction_MathIntAdd(B=1, A=@n2.Index) {is=struct RigVMUnitNode, template="Add::Execute(in A,in B,out Result)"}.Result
     Array=RigVMVariableNode() {variable=Items}.Value
    ) {template="DISPATCH_RigVMDispatch_ArrayGetAtIndex(in Array,in Index,out Element)"}.Element
    bInitial=@n4.Value
   )
   n6 = &RigUnit_GetTransform(bInitial=@n4.Value, Item=@n2.Element)
   RigVMVariableNode(
    ExecuteContext=@n3.True
    Value=&RigVMFunction_MathFloatAdd(
     B=&RigVMFunction_MathVectorDistance(B=@n5.Transform.Translation, A=@n6.Transform.Translation).Result
     A=RigVMVariableNode() {variable=distance}.Value
    ) {template="Add::Execute(in A,in B,out Result)"}.Result
   ) {variable=distance}
   RigVMFunctionReturnNode(ExecuteContext=@RigUnit_SequenceAggregate_1.B, Length=RigVMVariableNode() {variable=distance}.Value)
  ]}
 ContainedGraph {nodes [
   n1 = RigVMFunctionEntryNode()
   RigVMFunction_Sequence = &RigVMFunction_Sequence(ExecuteContext=@n1.ExecuteContext)
   RigVMFunction_Sequence_1 = RigVMAggregateNode(ExecuteContext=@RigVMFunction_Sequence.B)
   n2 = RigVMVariableNode() {variable=ac vector}
   n3 = /Script/RigVM.RigVMFunction_MathVectorDot(B=@n2.Value, A=RigVMVariableNode() {variable=ab vector}.Value) {is=struct RigVMUnitNode, template="Dot,|::Execute(in A,in B,out Result)"}
   n4 = /Script/RigVM.RigVMFunction_MathVectorLength(Value=@n2.Value) {is=struct RigVMUnitNode, template="Length::Execute(in Value,out Result)"}
   n5 = /Script/RigVM.RigVMFunction_MathVectorSub(
    B=/Script/RigVM.RigVMFunction_MathVectorScale(
     Factor=/Script/RigVM.RigVMFunction_MathFloatDiv(A=@n3.Result, B=@n4.Result) {is=struct RigVMUnitNode, template="Divide::Execute(in A,in B,out Result)"}.Result
     Value=/Script/RigVM.RigVMFunction_MathVectorUnit(Value=@n2.Value) {is=struct RigVMUnitNode, template="Unit::Execute(in Value,out Result)"}.Result
    ) {is=struct RigVMUnitNode, template="Scale::Execute(in Value,in Factor,out Result)"}.Result
    A=RigVMVariableNode() {variable=ab vector}.Value
   ) {is=struct RigVMUnitNode, template="Subtract::Execute(in A,in B,out Result)"}
   n6 = RigVMVariableNode() {variable=OffsetFactor}
   RigVMVariableNode(
    Value=/Script/RigVM.RigVMFunction_MathVectorAdd(
     B=&RigVMFunction_MathVectorScale(Value=@n5.Result, Factor=@n6.Value) {template="Scale::Execute(in Value,in Factor,out Result)"}.Result
     A=RigVMVariableNode() {variable=boneB translation}.Value
    ) {is=struct RigVMUnitNode, template="Add::Execute(in A,in B,out Result)"}.Result
    ExecuteContext=@RigVMFunction_Sequence_1.B
   ) {variable=translation output}
   n7 = RigVMVariableNode() {variable=ab vector}
   n8 = &RigVMFunction_MathVectorUnit(Value=/Script/RigVM.RigVMFunction_MathVectorCross(B=@n7.Value, A=RigVMVariableNode() {variable=ac vector}.Value) {is=struct RigVMUnitNode, template="Cross::Execute(in A,in B,out Result)"}.Result) {template="Unit::Execute(in Value,out Result)"}
   n9 = RigVMVariableNode() {variable=ac vector}
   n10 = &RigVMFunction_MathFloatDiv(
    A=&RigVMFunction_MathVectorDot(B=RigVMVariableNode() {variable=boneC translation}.Value, A=RigVMVariableNode() {variable=boneA translation}.Value) {template="Dot,|::Execute(in A,in B,out Result)"}.Result
    B=&RigVMFunction_MathVectorLength(Value=@n9.Value) {template="Length::Execute(in Value,out Result)"}.Result
   ) {template="Divide::Execute(in A,in B,out Result)"}
   n11 = &RigVMFunction_MathVectorUnit(Value=@n9.Value) {template="Unit::Execute(in Value,out Result)"}
   n12 = &RigVMFunction_MathVectorScale(
    Factor=0.5
    Value=&RigVMFunction_MathVectorSub(
     B=&RigVMFunction_MathVectorScale(Factor=@n10.Result, Value=@n11.Result) {template="Scale::Execute(in Value,in Factor,out Result)"}.Result
     A=@n7.Value
    ) {template="Subtract::Execute(in A,in B,out Result)"}.Result
   ) {template="Scale::Execute(in Value,in Factor,out Result)"}
   n13 = &RigVMFunction_MathVectorCross(A=@n8.Result, B=@n12.Result) {template="Cross::Execute(in A,in B,out Result)"}
   n14 = /Script/RigVM.RigVMFunction_MathMatrixToTransform(Value=/Script/RigVM.RigVMFunction_MathMatrixFromVectors(
     Z=&RigVMFunction_MathVectorUnit(Value=@n13.Result) {template="Unit::Execute(in Value,out Result)"}.Result
     Origin=RigVMVariableNode() {variable=Origin Scale}.Value
     Y=@n8.Result
     X=@n12.Result
    ) {is=struct RigVMUnitNode}.Result) {is=struct RigVMUnitNode, template="Cast::Execute(in Value,out Result)"}
   RigVMVariableNode(
    Value=&RigVMFunction_MathQuaternionFromEuler(RotationOrder=XYZ, Euler=&RigVMFunction_MathQuaternionToEuler(RotationOrder=XYZ, Value=@n14.Result.Rotation).Result).Result
    ExecuteContext=@RigVMFunction_Sequence_1.C
   ) {variable=rotation output}
   n15 = RigVMVariableNode() {variable=boneA translation}
   RigVMVariableNode(
    Value=&RigVMFunction_MathVectorSub(A=RigVMVariableNode() {variable=boneB translation}.Value, B=@n15.Value) {template="Subtract::Execute(in A,in B,out Result)"}.Result
    ExecuteContext=@RigVMFunction_Sequence_1.A
   ) {variable=ab vector}
   > RigVMVariableNode(Value=&RigVMFunction_MathVectorSub(A=RigVMVariableNode() {variable=boneC translation}.Value, B=@n15.Value) {template="Subtract::Execute(in A,in B,out Result)"}.Result) {variable=ac vector}
   DebugTransformMutableItemSpace = /Script/RigVM.RigVMFunction_DebugTransformMutableNoSpace(
    Color=(R=1,G=0,B=0,A=1)
    DebugDrawSettings=(DepthPriority=SDPG_Foreground,Lifetime=-1)
    Mode=Axes
    Scale=20
    Thickness=1
    Transform=(Rotation=(X=0,Y=0,Z=0,W=1),Translation=(X=0,Y=0,Z=0),Scale3D=(X=1,Y=1,Z=1))
    WorldOffset=(Rotation=(X=0,Y=0,Z=0,W=1),Translation=(X=0,Y=0,Z=0),Scale3D=(X=1,Y=1,Z=1))
    bEnabled=RigVMVariableNode() {variable=Draw}.Value
    Transform.Translation=RigVMVariableNode() {variable=translation output}.Value
    Transform.Rotation=RigVMVariableNode() {variable=rotation output}.Value
    ExecutePin=@RigVMFunction_Sequence_1.D
   ) {is=struct RigVMUnitNode}
   > RigVMFunctionReturnNode(
    Transform=(Rotation=(X=0,Y=0,Z=0,W=1),Translation=(X=0,Y=0,Z=0),Scale3D=(X=1,Y=1,Z=1))
    Transform.Translation=RigVMVariableNode() {variable=translation output}.Value
    Transform.Rotation=RigVMVariableNode() {variable=rotation output}.Value
   )
   n16 = RigVMVariableNode() {variable=Initial}
   RigVMVariableNode(
    Value=&RigUnit_GetTransform(Item=@n1.Bone A, bInitial=@n16.Value).Transform.Translation
    ExecuteContext=@RigVMFunction_Sequence.A
   ) {variable=boneA translation}
   > RigVMVariableNode(Value=&RigUnit_GetTransform(Item=@n1.Bone B, bInitial=@n16.Value).Transform.Translation) {variable=boneB translation}
   > RigVMVariableNode(Value=&RigUnit_GetTransform(Item=@n1.Bone C, bInitial=@n16.Value).Transform.Translation) {variable=boneC translation}
  ]}
 New Function_ContainedGraph {nodes [
   RigVMFunctionEntryNode()
   > RigVMFunction_Sequence = RigVMAggregateNode()
   RigVMVariableNode(
    Value=&RigUnit_GetTransform(bInitial=true, Item=RigVMVariableNode() {variable=Start}.Value).Transform
    ExecuteContext=@RigVMFunction_Sequence.A
   ) {variable=start transform}
   > RigVMVariableNode(Value=&RigUnit_GetTransform(bInitial=true, Item=RigVMVariableNode() {variable=Mid}.Value).Transform) {variable=mid transform}
   > RigVMVariableNode(Value=&RigUnit_GetTransform(bInitial=true, Item=RigVMVariableNode() {variable=End}.Value).Transform) {variable=end transform}
   n1 = RigVMVariableNode() {variable=mid transform}
   RigVMFunction_MathTransformTransformVector = /Script/RigVM.RigVMFunction_MathTransformTransformVector(Location=RigVMVariableNode() {variable=Secondary Axis}.Value, Transform=@n1.Value) {is=struct RigVMUnitNode}
   n2 = &RigVMFunction_MathVectorCross(
    A=&RigVMFunction_MathVectorSub(A=@RigVMFunction_MathTransformTransformVector.Result, B=@n1.Value.Translation) {template="Subtract::Execute(in A,in B,out Result)"}.Result
    B=&RigVMFunction_MathVectorSub(B=RigVMVariableNode() {variable=start transform}.Value.Translation, A=RigVMVariableNode() {variable=end transform}.Value.Translation) {template="Subtract::Execute(in A,in B,out Result)"}.Result
   ) {template="Cross::Execute(in A,in B,out Result)"}
   RigVMFunction_MathIntersectPlane = /Script/RigVM.RigVMFunction_MathIntersectPlane(
    Direction=@n2.Result
    PlaneNormal=@n2.Result
    PlanePoint=RigVMVariableNode() {variable=start transform}.Value.Translation
    Start=RigVMVariableNode() {variable=mid transform}.Value.Translation
   ) {is=struct RigVMUnitNode}
   Subtract = &RigVMFunction_MathVectorSub(A=@RigVMFunction_MathIntersectPlane.Result, B=RigVMVariableNode() {variable=start transform}.Value.Translation) {template="Subtract::Execute(in A,in B,out Result)"}
   RigVMFunction_MathVectorLength = &RigVMFunction_MathVectorLength(Value=@Subtract.Result) {template="Length::Execute(in Value,out Result)"}
   Divide = &RigVMFunction_MathDoubleDiv(
    A=&RigVMFunction_MathVectorDistance(A=RigVMVariableNode() {variable=start transform}.Value.Translation, B=RigVMVariableNode() {variable=mid transform}.Value.Translation).Result
    B=@RigVMFunction_MathVectorLength.Result
   )
   Multiply = /Script/RigVM.RigVMFunction_MathVectorMul(B=(X=1,Y=1,Z=1), B.X=@Divide.Result, B.Y=@Divide.Result, B.Z=@Divide.Result, A=@Subtract.Result) {is=struct RigVMUnitNode, template="Multiply::Execute(in A,in B,out Result)"}
   Add = &RigVMFunction_MathVectorAdd(B=@Multiply.Result, A=RigVMVariableNode() {variable=start transform}.Value.Translation) {template="Add::Execute(in A,in B,out Result)"}
   n3 = RigVMVariableNode() {variable=mid transform}
   RigVMFunction_MathTransformMake = /Script/RigVM.RigVMFunction_MathTransformMake(Translation=@Add.Result, Rotation=@n3.Value.Rotation, Scale=@n3.Value.Scale3D) {is=struct RigVMUnitNode}
   n4 = RigVMVariableNode(Value=@RigVMFunction_MathTransformMake.Result, ExecuteContext=@RigVMFunction_Sequence.B) {variable=out transform}
   RigVMFunction_ControlFlowBranch = &RigVMFunction_ControlFlowBranch(Condition=RigVMVariableNode() {variable=Debug}.Value, ExecuteContext=@n4.ExecuteContext)
   RigVMFunction_DebugLineNoSpace = /Script/RigVM.RigVMFunction_DebugLineNoSpace(
    Color=(R=1,G=0,B=0,A=1)
    DebugDrawSettings=(DepthPriority=SDPG_Foreground,Lifetime=-1)
    Thickness=0.2
    WorldOffset=(Rotation=(X=0,Y=0,Z=0,W=1),Translation=(X=0,Y=0,Z=0),Scale3D=(X=1,Y=1,Z=1))
    bEnabled=True
    ExecutePin=@RigVMFunction_ControlFlowBranch.True
    A=RigVMVariableNode() {variable=start transform}.Value.Translation
    B=@Add.Result
   ) {is=struct RigVMUnitNode}
   > RigVMFunction_DebugRectangleNoSpace = /Script/RigVM.RigVMFunction_DebugRectangleNoSpace(
    Color=(R=1,G=0,B=0,A=1)
    DebugDrawSettings=(DepthPriority=SDPG_Foreground,Lifetime=-1)
    Scale=1
    Thickness=2
    WorldOffset=(Rotation=(X=0,Y=0,Z=0,W=1),Translation=(X=0,Y=0,Z=0),Scale3D=(X=1,Y=1,Z=1))
    bEnabled=True
    Transform=RigVMVariableNode() {variable=out transform}.Value
   ) {is=struct RigVMUnitNode}
   RigVMFunction_MathVectorDistance = &RigVMFunction_MathVectorDistance(A=RigVMVariableNode() {variable=start transform}.Value.Translation, B=RigVMVariableNode() {variable=out transform}.Value.Translation)
   Less = /Script/RigVM.RigVMFunction_MathDoubleLess(B=0.001, A=@RigVMFunction_MathVectorDistance.Result) {is=struct RigVMUnitNode, template="Less::Execute(in A,in B,out Result)"}
   Make_Array = RigVMDispatchNode(
    Values=((Rotation=(X=0,Y=0,Z=0,W=1),Translation=(X=0,Y=0,Z=0),Scale3D=(X=1,Y=1,Z=1)),(Rotation=(X=0,Y=0,Z=0,W=1),Translation=(X=0,Y=0,Z=0),Scale3D=(X=1,Y=1,Z=1)),(Rotation=(X=0,Y=0,Z=0,W=1),Translation=(X=0,Y=0,Z=0),Scale3D=(X=1,Y=1,Z=1)))
    Values.0=RigVMVariableNode() {variable=start transform}.Value
    Values.1=RigVMVariableNode() {variable=out transform}.Value
    Values.2=RigVMVariableNode() {variable=end transform}.Value
   ) {template="DISPATCH_RigVMDispatch_ArrayMake(in Values,out Array)"}
   RigVMFunctionReturnNode(Same As Source=@Less.Result, ExecuteContext=@RigVMFunction_Sequence.D, outputs=@Make_Array.Array)
  ]}
 New Function_ContainedGraph {nodes [
   n1 = RigVMFunctionEntryNode()
   RigVMFunction_MathQuaternionSwingTwist = &RigVMFunction_MathQuaternionSwingTwist(
    Input=&RigVMFunction_MathTransformMakeRelative(
     Parent=&RigUnit_GetTransform(Space=LocalSpace, bInitial=true, Item=@n1.Item).Transform
     Global=&RigUnit_GetTransform(Space=LocalSpace, bInitial=false, Item=@n1.Item).Transform
    ).Local.Rotation
    TwistAxis=RigVMVariableNode() {variable=Axis}.Value
   )
   > RigVMFunctionReturnNode(Twist=@RigVMFunction_MathQuaternionSwingTwist.Twist)
  ]}
 RigVMFunctionLibrary {nodes [
   Compute Pole Vector = RigVMCollapseNode(
    Bone A=(Type=Bone)
    Bone B=(Type=Bone)
    Bone C=(Type=Bone)
    Draw=true
    ExecuteContext=()
    Initial=false
    OffsetFactor=2
    Origin Scale=(X=0,Y=0,Z=0)
   )
   Get Chain Length = RigVMCollapseNode(
    ExecuteContext=()
    Initial=true
    Items=((Type=Bone,Name="torso_1"),(Type=Bone,Name="torso_2"),(Type=Bone,Name="torso_3"),(Type=Bone,Name="torso_4"))
   )
   Get Node Twist Value = RigVMCollapseNode(Axis=(X=1,Y=0,Z=0), ExecuteContext=(), Item=(Type=None,Name="None"))
   Project Middle Bone to IK Plane = RigVMCollapseNode(
    Debug=true
    End=(Type=None,Name="None")
    ExecuteContext=()
    Mid=(Type=None,Name="None")
    Secondary Axis=(X=0,Y=1,Z=0)
    Start=(Type=None,Name="None")
   )
  ]}
 RigVMFunction_Sequence_1_1_ContainedGraph {nodes [
   n1 = RigVMFunctionEntryNode()
   n2 = &RigVMFunction_Sequence(ExecuteContext=@n1.ExecuteContext)
   n3 = &RigVMFunction_Sequence(ExecuteContext=@n2.B)
   n4 = &RigVMFunction_Sequence(ExecuteContext=@n3.B)
   RigVMFunctionReturnNode(A=@n2.A, B=@n3.A, C=@n4.A, D=@n4.B)
  ]}
 RigVMFunction_Sequence_1_ContainedGraph {nodes [
   n1 = RigVMFunctionEntryNode()
   RigVMFunction_Sequence = &RigVMFunction_Sequence(ExecuteContext=@n1.ExecuteContext)
   n2 = &RigVMFunction_Sequence(ExecuteContext=@RigVMFunction_Sequence.B)
   n3 = &RigVMFunction_Sequence(ExecuteContext=@n2.B)
   RigVMFunctionReturnNode(A=@RigVMFunction_Sequence.A, B=@n2.A, C=@n3.A, D=@n3.B)
  ]}
 RigVMModel {nodes [
   RigUnit_BeginExecution = &RigUnit_BeginExecution()
   Project Middle Bone to IK Plane = RigVMFunctionReferenceNode(
    Debug=true
    End=(Type=Control,Name="root_ctrl_2_2")
    Mid=(Type=Control,Name="root_ctrl_2")
    Secondary Axis=(X=0,Y=0,Z=0)
    Start=(Type=Control,Name="root_ctrl")
    ExecuteContext=@RigUnit_BeginExecution.ExecutePin
   )
  ]}
 RigVMModel Rig {nodes [&RigUnit_PrepareForExecution()]}
]
hierarchy [
 root {loc=<0 0 0>, rot=<0 0 0 1>}
 pelvis {parent=root, loc=<0 2.2809 95.8968>, rot=<-0.0224 0.7068 0.0224 -0.7068>}
 spine_01 {parent=pelvis, loc=<3.6771 0 0>, rot=<0 0 -0.1258 -0.9921>}
 spine_02 {parent=spine_01, loc=<6.7951 0 0>, rot=<0 0 0.0302 -0.9995>}
 spine_03 {parent=spine_02, loc=<7.2382 0 0>, rot=<0 0 0.0954 -0.9954>}
 spine_04 {parent=spine_03, loc=<8.5239 0 0>, rot=<0 0 0.0512 -0.9987>}
 spine_05 {parent=spine_04, loc=<19.4398 0 0>, rot=<0 0 0.0059 -1>}
 neck_01 {parent=spine_05, loc=<11.8878 0 0>, rot=<0 0 -0.2073 -0.9783>}
 neck_02 {parent=neck_01, loc=<5.1103 0 0>, rot=<0 0 0.0167 -0.9999>}
 head {parent=neck_02, loc=<4.913 0 0>, rot=<0 0 0.1035 -0.9946>}
 clavicle_l {parent=spine_05, loc=<5.5163 -1.3148 -1.4279>, rot=<0.0833 -0.7546 -0.0413 -0.6496>}
 upperarm_l {parent=clavicle_l, loc=<17.8095 0 0>, rot=<-0.0199 -0.3917 -0.0202 -0.9196>}
 lowerarm_l {parent=upperarm_l, loc=<27.7711 0 0>, rot=<0 0 -0.3336 -0.9427>}
 lowerarm_twist_02_l {parent=lowerarm_l, loc=<9.0837 0 0>, rot=<0.0115 0.0149 0.0175 -0.9997>}
 lowerarm_twist_01_l {parent=lowerarm_l, loc=<18.1674 0 0>, rot=<0.0115 0.0149 0.0175 -0.9997>}
 hand_l {parent=lowerarm_l, loc=<27.2511 0 0>, rot=<-0.5576 -0.0017 0.0206 -0.8299>}
 middle_metacarpal_l {parent=hand_l, loc=<3.3758 0.7536 -0.1829>, rot=<-0.0373 0.0004 -0.0203 -0.9991>}
 middle_01_l {parent=middle_metacarpal_l, loc=<6.0982 0 0>, rot=<0 0 0.2721 -0.9623>}
 middle_02_l {parent=middle_01_l, loc=<5.169 0 0>, rot=<0 0 0.1803 -0.9836>}
 middle_03_l {parent=middle_02_l, loc=<2.474 0 0>, rot=<0 0 0.0872 -0.9962>}
 pinky_metacarpal_l {parent=hand_l, loc=<3.3144 0.3059 2.3911>, rot=<-0.2182 0.1882 0.0583 -0.9558>}
 pinky_01_l {parent=pinky_metacarpal_l, loc=<4.9576 0.1431 -0.1988>, rot=<0.09 -0.017 0.1281 -0.9875>}
 pinky_02_l {parent=pinky_01_l, loc=<3.816 0 0>, rot=<0 0 0.1847 -0.9828>}
 pinky_03_l {parent=pinky_02_l, loc=<2.04 0 0>, rot=<0 0 0.0429 -0.9991>}
 ring_metacarpal_l {parent=hand_l, loc=<3.3743 0.5425 1.0918>, rot=<-0.1166 0.1006 -0.0257 -0.9877>}
 ring_01_l {parent=ring_metacarpal_l, loc=<5.6455 0.0416 -0.0207>, rot=<0.0542 -0.0132 0.2535 -0.9657>}
 ring_02_l {parent=ring_01_l, loc=<4.977 0 0>, rot=<0 0 0.1647 -0.9863>}
 ring_03_l {parent=ring_02_l, loc=<2.265 0 0>, rot=<0 0 0.0799 -0.9968>}
 thumb_01_l {parent=hand_l, loc=<1.9924 -1.3566 -2.5815>, rot=<0.5052 -0.3691 -0.067 -0.7772>}
 thumb_02_l {parent=thumb_01_l, loc=<4.378 0 0>, rot=<0.0336 0.0103 0.2019 -0.9788>}
 thumb_03_l {parent=thumb_02_l, loc=<3.086 0 0>, rot=<0 0 0.0872 -0.9962>}
 index_metacarpal_l {parent=hand_l, loc=<3.4445 0.3847 -2.3793>, rot=<0.029 -0.0637 -0.0071 -0.9975>}
 index_01_l {parent=index_metacarpal_l, loc=<5.8771 -0.0432 0.2409>, rot=<0 0 0.2026 -0.9793>}
 index_02_l {parent=index_01_l, loc=<4.08 0 0>, rot=<0 0 0.1296 -0.9916>}
 index_03_l {parent=index_02_l, loc=<2.595 0 0>, rot=<0 0 0.109 -0.994>}
 upperarm_twist_01_l {parent=upperarm_l, loc=<9.257 0 0>, rot=<-0.0001 0.0188 -0.0028 -0.9998>}
 upperarm_twist_02_l {parent=upperarm_l, loc=<18.5141 0 0>, rot=<0 0 0 1>}
 clavicle_r {parent=spine_05, loc=<5.5162 -1.3148 1.4279>, rot=<0.7546 0.0833 -0.6496 0.0413>}
 upperarm_r {parent=clavicle_r, loc=<-17.8096 0 0.0004>, rot=<-0.0199 -0.3917 -0.0202 -0.9196>}
 lowerarm_r {parent=upperarm_r, loc=<-27.7707 0 0>, rot=<0 0 -0.3336 -0.9427>}
 lowerarm_twist_02_r {parent=lowerarm_r, loc=<-9.0837 0 0>, rot=<0.0115 0.0149 0.0175 -0.9997>}
 lowerarm_twist_01_r {parent=lowerarm_r, loc=<-18.1673 0 0>, rot=<0.0115 0.0149 0.0175 -0.9997>}
 hand_r {parent=lowerarm_r, loc=<-27.251 0 0>, rot=<-0.5576 -0.0017 0.0206 -0.8299>}
 middle_metacarpal_r {parent=hand_r, loc=<-3.3758 -0.754 0.1828>, rot=<-0.0373 0.0004 -0.0203 -0.9991>}
 middle_01_r {parent=middle_metacarpal_r, loc=<-6.0984 0.0001 0>, rot=<0 0 0.2721 -0.9623>}
 middle_02_r {parent=middle_01_r, loc=<-5.169 0.0001 0>, rot=<0 0 0.1803 -0.9836>}
 middle_03_r {parent=middle_02_r, loc=<-2.474 0 0>, rot=<0 0 0.0872 -0.9962>}
 pinky_metacarpal_r {parent=hand_r, loc=<-3.3147 -0.3059 -2.3913>, rot=<-0.2182 0.1882 0.0583 -0.9558>}
 pinky_01_r {parent=pinky_metacarpal_r, loc=<-4.9573 -0.1433 0.1989>, rot=<0.09 -0.017 0.1281 -0.9875>}
 pinky_02_r {parent=pinky_01_r, loc=<-3.816 0 0>, rot=<0 0 0.1847 -0.9828>}
 pinky_03_r {parent=pinky_02_r, loc=<-2.04 0 0>, rot=<0 0 0.0429 -0.9991>}
 ring_metacarpal_r {parent=hand_r, loc=<-3.3742 -0.543 -1.0918>, rot=<-0.1166 0.1006 -0.0257 -0.9877>}
 ring_01_r {parent=ring_metacarpal_r, loc=<-5.6457 -0.0414 0.0207>, rot=<0.0542 -0.0132 0.2535 -0.9657>}
 ring_02_r {parent=ring_01_r, loc=<-4.9771 0 0>, rot=<0 0 0.1647 -0.9863>}
 ring_03_r {parent=ring_02_r, loc=<-2.265 -0.0001 0>, rot=<0 0 0.0799 -0.9968>}
 thumb_01_r {parent=hand_r, loc=<-1.9928 1.3567 2.5813>, rot=<0.5052 -0.3691 -0.067 -0.7772>}
 thumb_02_r {parent=thumb_01_r, loc=<-4.3778 -0.0004 -0.0001>, rot=<0.0336 0.0103 0.2019 -0.9788>}
 thumb_03_r {parent=thumb_02_r, loc=<-3.086 0 0>, rot=<0 0 0.0872 -0.9962>}
 index_metacarpal_r {parent=hand_r, loc=<-3.4445 -0.3852 2.3793>, rot=<0.029 -0.0637 -0.0071 -0.9975>}
 index_01_r {parent=index_metacarpal_r, loc=<-5.8772 0.0434 -0.241>, rot=<0 0 0.2026 -0.9793>}
 index_02_r {parent=index_01_r, loc=<-4.0799 0 0>, rot=<0 0 0.1296 -0.9916>}
 index_03_r {parent=index_02_r, loc=<-2.5951 0 0>, rot=<0 0 0.109 -0.994>}
 upperarm_twist_01_r {parent=upperarm_r, loc=<-9.2569 0 -0.0001>, rot=<-0.0001 0.0188 -0.0028 -0.9998>}
 upperarm_twist_02_r {parent=upperarm_r, loc=<-18.5138 0 -0.0002>, rot=<0 0 0 1>}
 thigh_r {parent=pelvis, loc=<-2.3657 0.1195 9.9691>, rot=<0.0295 -0.0724 0.9965 -0.029>}
 calf_r {parent=thigh_r, loc=<43.3413 0 0>, rot=<0 0 -0.0437 -0.999>}
 foot_r {parent=calf_r, loc=<42.2179 0 0>, rot=<-0.0007 -0.0269 0.0232 -0.9994>}
 ball_r {parent=foot_r, loc=<7.0094 -15.2376 -0.5389>, rot=<0 0 -0.7071 -0.7071>}
 calf_twist_02_r {parent=calf_r, loc=<14.0726 0 0.0377>, rot=<-0.0001 -0.0023 0.0232 -0.9997>}
 calf_twist_01_r {parent=calf_r, loc=<28.1453 0 0.0755>, rot=<-0.0001 -0.0023 0.0232 -0.9997>}
 thigh_twist_01_r {parent=thigh_r, loc=<14.4471 0 0>, rot=<0 -0.0023 -0.0114 -0.9999>}
 thigh_twist_02_r {parent=thigh_r, loc=<28.8942 0 0>, rot=<0 -0.0023 -0.0114 -0.9999>}
 thigh_l {parent=pelvis, loc=<-2.3657 0.11 -9.9692>, rot=<0.0724 0.0295 -0.029 -0.9965>}
 calf_l {parent=thigh_l, loc=<-43.3413 0 0>, rot=<0 0 -0.0437 -0.999>}
 foot_l {parent=calf_l, loc=<-42.2179 0 0>, rot=<-0.0007 -0.0269 0.0232 -0.9994>}
 ball_l {parent=foot_l, loc=<-7.0094 15.2376 0.5389>, rot=<0 0 -0.7071 -0.7071>}
 calf_twist_02_l {parent=calf_l, loc=<-14.0726 0 0>, rot=<-0.0001 -0.0023 0.0232 -0.9997>}
 calf_twist_01_l {parent=calf_l, loc=<-28.1453 0 -0.0754>, rot=<-0.0001 -0.0023 0.0232 -0.9997>}
 thigh_twist_01_l {parent=thigh_l, loc=<-14.4471 0 0>, rot=<0 -0.0023 -0.0114 -0.9999>}
 thigh_twist_02_l {parent=thigh_l, loc=<-28.8942 0 0>, rot=<0 -0.0023 -0.0114 -0.9999>}
 ik_foot_root {parent=root, loc=<0 0 0>, rot=<0 0 0 1>}
 ik_foot_l {parent=ik_foot_root, loc=<14.0877 -0.9943 8.2376>, rot=<0.0636 0.7042 0.0414 -0.7059>}
 ik_foot_r {parent=ik_foot_root, loc=<-14.0876 -0.9848 8.2382>, rot=<0.7059 0.0414 -0.7042 0.0636>}
 ik_hand_root {parent=root, loc=<0 0 0>, rot=<0 0 0 1>}
 ik_hand_gun {parent=ik_hand_root, loc=<-47.7683 15.6988 104.4664>, rot=<0.5985 0.0096 0.5206 -0.6089>}
 ik_hand_l {parent=ik_hand_gun, loc=<43.7336 61.6698 58.4081>, rot=<-0.4578 -0.6455 -0.6114 0>}
 ik_hand_r {parent=ik_hand_gun, loc=<0 0 0>, rot=<0 0 0 1>}
 interaction {parent=root, loc=<0 0 0>, rot=<0 0 0 1>}
 center_of_mass {parent=root, loc=<0 0 0>, rot=<0 0 0 1>}
 MoveData_Speed {type=Curve, loc=<0 0 0>, rot=<0 0 0 1>}
 base_pose {type=Curve, loc=<0 0 0>, rot=<0 0 0 1>}
 calf_l_back_50 {type=Curve, loc=<0 0 0>, rot=<0 0 0 1>}
 calf_l_back_120 {type=Curve, loc=<0 0 0>, rot=<0 0 0 1>}
 calf_l_back_90 {type=Curve, loc=<0 0 0>, rot=<0 0 0 1>}
 calf_l_back_150 {type=Curve, loc=<0 0 0>, rot=<0 0 0 1>}
 calf_r_back_50 {type=Curve, loc=<0 0 0>, rot=<0 0 0 1>}
 calf_r_back_120 {type=Curve, loc=<0 0 0>, rot=<0 0 0 1>}
 calf_r_back_90 {type=Curve, loc=<0 0 0>, rot=<0 0 0 1>}
 calf_r_back_150 {type=Curve, loc=<0 0 0>, rot=<0 0 0 1>}
 clavicle_l_fwd_30 {type=Curve, loc=<0 0 0>, rot=<0 0 0 1>}
 clavicle_l_down_20 {type=Curve, loc=<0 0 0>, rot=<0 0 0 1>}
 clavicle_l_back_30 {type=Curve, loc=<0 0 0>, rot=<0 0 0 1>}
 clavicle_l_up_40 {type=Curve, loc=<0 0 0>, rot=<0 0 0 1>}
 clavicle_r_fwd_30 {type=Curve, loc=<0 0 0>, rot=<0 0 0 1>}
 clavicle_r_down_20 {type=Curve, loc=<0 0 0>, rot=<0 0 0 1>}
 clavicle_r_back_30 {type=Curve, loc=<0 0 0>, rot=<0 0 0 1>}
 clavicle_r_up_40 {type=Curve, loc=<0 0 0>, rot=<0 0 0 1>}
 clavicle_r_back_310 {type=Curve, loc=<0 0 0>, rot=<0 0 0 1>}
 clavicle_r_up_401 {type=Curve, loc=<0 0 0>, rot=<0 0 0 1>}
 foot_l_down_60 {type=Curve, loc=<0 0 0>, rot=<0 0 0 1>}
 foot_l_up_35 {type=Curve, loc=<0 0 0>, rot=<0 0 0 1>}
 foot_r_up_35 {type=Curve, loc=<0 0 0>, rot=<0 0 0 1>}
 foot_r_down_60 {type=Curve, loc=<0 0 0>, rot=<0 0 0 1>}
 hand_l_down_90 {type=Curve, loc=<0 0 0>, rot=<0 0 0 1>}
 hand_l_up_20 {type=Curve, loc=<0 0 0>, rot=<0 0 0 1>}
 hand_l_up_90 {type=Curve, loc=<0 0 0>, rot=<0 0 0 1>}
 hand_r_up_90 {type=Curve, loc=<0 0 0>, rot=<0 0 0 1>}
 hand_r_down_90 {type=Curve, loc=<0 0 0>, rot=<0 0 0 1>}
 hand_r_up_20 {type=Curve, loc=<0 0 0>, rot=<0 0 0 1>}
 Pose_0 {type=Curve, loc=<0 0 0>, rot=<0 0 0 1>}
 Pose_1 {type=Curve, loc=<0 0 0>, rot=<0 0 0 1>}
 Pose_2 {type=Curve, loc=<0 0 0>, rot=<0 0 0 1>}
 Pose_3 {type=Curve, loc=<0 0 0>, rot=<0 0 0 1>}
 Pose_4 {type=Curve, loc=<0 0 0>, rot=<0 0 0 1>}
 Pose_5 {type=Curve, loc=<0 0 0>, rot=<0 0 0 1>}
 Pose_6 {type=Curve, loc=<0 0 0>, rot=<0 0 0 1>}
 Pose_7 {type=Curve, loc=<0 0 0>, rot=<0 0 0 1>}
 Pose_8 {type=Curve, loc=<0 0 0>, rot=<0 0 0 1>}
 lowerarm_l_in_110 {type=Curve, loc=<0 0 0>, rot=<0 0 0 1>}
 lowerarm_l_in_35 {type=Curve, loc=<0 0 0>, rot=<0 0 0 1>}
 lowerarm_l_out_35 {type=Curve, loc=<0 0 0>, rot=<0 0 0 1>}
 lowerarm_l_in_50 {type=Curve, loc=<0 0 0>, rot=<0 0 0 1>}
 lowerarm_l_in_75 {type=Curve, loc=<0 0 0>, rot=<0 0 0 1>}
 lowerarm_l_in_10 {type=Curve, loc=<0 0 0>, rot=<0 0 0 1>}
 lowerarm_l_out_10 {type=Curve, loc=<0 0 0>, rot=<0 0 0 1>}
 lowerarm_l_in_90 {type=Curve, loc=<0 0 0>, rot=<0 0 0 1>}
 lowerarm_r_in_10 {type=Curve, loc=<0 0 0>, rot=<0 0 0 1>}
 lowerarm_r_out_10 {type=Curve, loc=<0 0 0>, rot=<0 0 0 1>}
 lowerarm_r_out_35 {type=Curve, loc=<0 0 0>, rot=<0 0 0 1>}
 lowerarm_r_in_75 {type=Curve, loc=<0 0 0>, rot=<0 0 0 1>}
 lowerarm_r_in_50 {type=Curve, loc=<0 0 0>, rot=<0 0 0 1>}
 lowerarm_r_in_110 {type=Curve, loc=<0 0 0>, rot=<0 0 0 1>}
 lowerarm_r_in_35 {type=Curve, loc=<0 0 0>, rot=<0 0 0 1>}
 lowerarm_r_in_90 {type=Curve, loc=<0 0 0>, rot=<0 0 0 1>}
 thigh_l_bck_10 {type=Curve, loc=<0 0 0>, rot=<0 0 0 1>}
 thigh_l_fwd_10 {type=Curve, loc=<0 0 0>, rot=<0 0 0 1>}
 thigh_l_in_45_out_90 {type=Curve, loc=<0 0 0>, rot=<0 0 0 1>}
 thigh_l_fwd_90 {type=Curve, loc=<0 0 0>, rot=<0 0 0 1>}
 thigh_l_bck_90 {type=Curve, loc=<0 0 0>, rot=<0 0 0 1>}
 thigh_l_bck_50 {type=Curve, loc=<0 0 0>, rot=<0 0 0 1>}
 thigh_l_out_110 {type=Curve, loc=<0 0 0>, rot=<0 0 0 1>}
 thigh_l_out_10 {type=Curve, loc=<0 0 0>, rot=<0 0 0 1>}
 thigh_l_fwd_45 {type=Curve, loc=<0 0 0>, rot=<0 0 0 1>}
 thigh_l_in_50 {type=Curve, loc=<0 0 0>, rot=<0 0 0 1>}
 thigh_l_fwd_110 {type=Curve, loc=<0 0 0>, rot=<0 0 0 1>}
 thigh_l_out_55 {type=Curve, loc=<0 0 0>, rot=<0 0 0 1>}
 thigh_l_out_85 {type=Curve, loc=<0 0 0>, rot=<0 0 0 1>}
 thigh_r_fwd_10 {type=Curve, loc=<0 0 0>, rot=<0 0 0 1>}
 thigh_r_out_55 {type=Curve, loc=<0 0 0>, rot=<0 0 0 1>}
 thigh_r_fwd_90 {type=Curve, loc=<0 0 0>, rot=<0 0 0 1>}
 thigh_r_fwd_45 {type=Curve, loc=<0 0 0>, rot=<0 0 0 1>}
 thigh_r_bck_10 {type=Curve, loc=<0 0 0>, rot=<0 0 0 1>}
 thigh_r_out_85 {type=Curve, loc=<0 0 0>, rot=<0 0 0 1>}
 thigh_r_in_50 {type=Curve, loc=<0 0 0>, rot=<0 0 0 1>}
 thigh_r_out_10 {type=Curve, loc=<0 0 0>, rot=<0 0 0 1>}
 thigh_r_out_110 {type=Curve, loc=<0 0 0>, rot=<0 0 0 1>}
 thigh_r_bck_50 {type=Curve, loc=<0 0 0>, rot=<0 0 0 1>}
 thigh_r_bck_90 {type=Curve, loc=<0 0 0>, rot=<0 0 0 1>}
 thigh_r_in_45_out_90 {type=Curve, loc=<0 0 0>, rot=<0 0 0 1>}
 thigh_r_fwd_110 {type=Curve, loc=<0 0 0>, rot=<0 0 0 1>}
 upperarm_l_back_45 {type=Curve, loc=<0 0 0>, rot=<0 0 0 1>}
 upperarm_l_out_55 {type=Curve, loc=<0 0 0>, rot=<0 0 0 1>}
 upperarm_l_out110_twist_in_70 {type=Curve, loc=<0 0 0>, rot=<0 0 0 1>}
 upperarm_l_out_85 {type=Curve, loc=<0 0 0>, rot=<0 0 0 1>}
 upperarm_l_fwd_110 {type=Curve, loc=<0 0 0>, rot=<0 0 0 1>}
 upperarm_l_in_10 {type=Curve, loc=<0 0 0>, rot=<0 0 0 1>}
 upperarm_l_out110_twist_out_70 {type=Curve, loc=<0 0 0>, rot=<0 0 0 1>}
 upperarm_l_out_twist_a {type=Curve, loc=<0 0 0>, rot=<0 0 0 1>}
 upperarm_l_fwd_15 {type=Curve, loc=<0 0 0>, rot=<0 0 0 1>}
 upperarm_l_out_15 {type=Curve, loc=<0 0 0>, rot=<0 0 0 1>}
 upperarm_l_back_10 {type=Curve, loc=<0 0 0>, rot=<0 0 0 1>}
 upperarm_l_out110 {type=Curve, loc=<0 0 0>, rot=<0 0 0 1>}
 upperarm_l_fwd_twist_in_a {type=Curve, loc=<0 0 0>, rot=<0 0 0 1>}
 upperarm_l_fwd_90 {type=Curve, loc=<0 0 0>, rot=<0 0 0 1>}
 upperarm_r_out_15 {type=Curve, loc=<0 0 0>, rot=<0 0 0 1>}
 upperarm_r_fwd_15 {type=Curve, loc=<0 0 0>, rot=<0 0 0 1>}
 upperarm_r_out_twist_a {type=Curve, loc=<0 0 0>, rot=<0 0 0 1>}
 upperarm_r_out110_twist_in_70 {type=Curve, loc=<0 0 0>, rot=<0 0 0 1>}
 upperarm_r_fwd_90 {type=Curve, loc=<0 0 0>, rot=<0 0 0 1>}
 upperarm_r_out110 {type=Curve, loc=<0 0 0>, rot=<0 0 0 1>}
 upperarm_r_back_10 {type=Curve, loc=<0 0 0>, rot=<0 0 0 1>}
 upperarm_r_out_85 {type=Curve, loc=<0 0 0>, rot=<0 0 0 1>}
 upperarm_r_in_10 {type=Curve, loc=<0 0 0>, rot=<0 0 0 1>}
 upperarm_r_back_45 {type=Curve, loc=<0 0 0>, rot=<0 0 0 1>}
 upperarm_r_out110_twist_out_70 {type=Curve, loc=<0 0 0>, rot=<0 0 0 1>}
 upperarm_r_out_55 {type=Curve, loc=<0 0 0>, rot=<0 0 0 1>}
 upperarm_r_fwd_twist_in_a {type=Curve, loc=<0 0 0>, rot=<0 0 0 1>}
 upperarm_r_fwd_110 {type=Curve, loc=<0 0 0>, rot=<0 0 0 1>}
 Distance {type=Curve, loc=<0 0 0>, rot=<0 0 0 1>}
 RemainingTurnYaw {type=Curve, loc=<0 0 0>, rot=<0 0 0 1>}
 TurnYawWeight {type=Curve, loc=<0 0 0>, rot=<0 0 0 1>}
 GroundDistance {type=Curve, loc=<0 0 0>, rot=<0 0 0 1>}
 DisableLHandIK {type=Curve, loc=<0 0 0>, rot=<0 0 0 1>}
 ScaleDownWeaponR {type=Curve, loc=<0 0 0>, rot=<0 0 0 1>}
 DisableLegIK {type=Curve, loc=<0 0 0>, rot=<0 0 0 1>}
 blendParent1 {type=Curve, loc=<0 0 0>, rot=<0 0 0 1>}
 DisableRHandIK {type=Curve, loc=<0 0 0>, rot=<0 0 0 1>}
 DisableLeftHandPoseOverride {type=Curve, loc=<0 0 0>, rot=<0 0 0 1>}
 applyHipfireOverridePose {type=Curve, loc=<0 0 0>, rot=<0 0 0 1>}
 DisableHandIKRetargeting {type=Curve, loc=<0 0 0>, rot=<0 0 0 1>}
 lowerarm_r_in_80 {type=Curve, loc=<0 0 0>, rot=<0 0 0 1>}
 lowerarm_l_in_80 {type=Curve, loc=<0 0 0>, rot=<0 0 0 1>}
 calf_r_back_130 {type=Curve, loc=<0 0 0>, rot=<0 0 0 1>}
 calf_l_back_130 {type=Curve, loc=<0 0 0>, rot=<0 0 0 1>}
 Pose_9 {type=Curve, loc=<0 0 0>, rot=<0 0 0 1>}
]
